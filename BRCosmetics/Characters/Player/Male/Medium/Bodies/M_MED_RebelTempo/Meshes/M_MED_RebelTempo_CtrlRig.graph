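# Control Rig for the M_MED_RebelTempo player body that drives corrective deformation of muscles, skirt, loincloth and side armor bones.
# Spherical pose readers and a swing/twist DeltaRotation function turn limb, clavicle, spine and head rotations into weights for additive bone offsets.
# Gameplay states like aiming, equipped weapon, jumping, skydiving and parachuting gate armor fixes, often smoothed by AlphaInterp.
# Weighted multi-parent constraints position helper bones, and isFrontEnd disables the bicep/tricep offsets.
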ControlRigBlueprint /BRCosmetics/Characters/Player/Male/Medium/Bodies/M_MED_RebelTempo/Meshes/M_MED_RebelTempo_CtrlRig
graphs [
 CollapseNode_ContainedGraph {nodes [
   n1 = RigVMFunctionEntryNode()
   RigVMFunction_Sequence = RigVMAggregateNode(ExecuteContext=@n1.ExecutePin)
   &RigUnit_ParentConstraint(
    AdvancedSettings=(RotationOrderForFilter=XZY)
    Child=(Type=Bone,Name="upperarm_twist_02_l")
    Parents="((Item=(Type=Bone,Name=\"lowerarm_twist_02_l\"),Weight=0.0024),(Item=(Type=Bone,Name=\"lowerarm_l\"),Weight=0.0063),(Item=(Type=Bone,Name=\"clavicle_l\"),Weight=0.0252),(Item=(Type=Bone,Name=\"upperarm_twist_02_l\"),Weight=0.1602),(Item=(Type=Bone,Name=\"upperarm_l ...(truncated)"
    Weight=1
    ExecutePin=@RigVMFunction_Sequence.A
   )
   > n2 = &RigUnit_SphericalPoseReader(
    ActiveRegionScaleFactors=(PositiveWidth=1,NegativeWidth=1,PositiveHeight=1,NegativeHeight=1)
    ActiveRegionSize=0.1
    Debug=(bDrawDebug=false,bDraw2D=False,bDrawLocalAxes=False,DebugScale=15,DebugSegments=20,DebugThickness=0.1)
    DriverAxis=(X=-1,Y=0,Z=0)
    DriverItem=(Type=Bone,Name="lowerarm_r")
    FalloffRegionScaleFactors=(PositiveWidth=1,NegativeWidth=1,PositiveHeight=1,NegativeHeight=1)
    FalloffSize=0.5
    FlipHeightScaling=False
    FlipWidthScaling=False
    OptionalParentItem=(Type=Bone,Name="upperarm_twist_02_r")
    RotationOffset=(X=-90,Y=0,Z=0)
   )
   RigVMFunction_MathBoolToFloat = &RigVMFunction_MathBoolToFloat(Value=RigVMVariableNode() {variable=isFrontEnd}.Value)
   Remap = &RigVMFunction_MathDoubleRemap(SourceMaximum=1, SourceMinimum=0, TargetMaximum=0, TargetMinimum=1, bClamp=False, Value=@RigVMFunction_MathBoolToFloat.Result)
   Multiply = &RigVMFunction_MathDoubleMul(B=@Remap.Result, A=@n2.OutputParam)
   > &RigUnit_OffsetTransformForItem(
    Item=(Type=Bone,Name="upperarm_bicep_r")
    OffsetTransform=(Rotation=(X=0,Y=0,Z=0,W=1),Translation=(X=3,Y=4,Z=0),Scale3D=(X=1,Y=1,Z=1))
    bPropagateToChildren=True
    Weight=@Multiply.Result
   )
   > &RigUnit_OffsetTransformForItem(
    Item=(Type=Bone,Name="upperarm_tricep_r")
    OffsetTransform=(Rotation=(X=0,Y=0,Z=0,W=1),Translation=(X=-4,Y=-5,Z=0),Scale3D=(X=1,Y=1,Z=1))
    bPropagateToChildren=True
    Weight=@Multiply.Result
   )
   > n3 = &RigUnit_SphericalPoseReader(
    ActiveRegionScaleFactors=(PositiveWidth=1,NegativeWidth=1,PositiveHeight=1,NegativeHeight=1)
    ActiveRegionSize=0.1
    Debug=(bDrawDebug=false,bDraw2D=False,bDrawLocalAxes=False,DebugScale=15,DebugSegments=20,DebugThickness=0.1)
    DriverAxis=(X=1,Y=0,Z=0)
    DriverItem=(Type=Bone,Name="lowerarm_l")
    FalloffRegionScaleFactors=(PositiveWidth=1,NegativeWidth=1,PositiveHeight=1,NegativeHeight=1)
    FalloffSize=0.5
    FlipHeightScaling=False
    FlipWidthScaling=False
    OptionalParentItem=(Type=Bone,Name="upperarm_twist_02_l")
    RotationOffset=(X=90,Y=0,Z=0)
   )
   n4 = RigVMVariableNode() {variable=isFrontEnd}
   n5 = &RigVMFunction_MathDoubleMul(
    B=&RigVMFunction_MathDoubleRemap(
     SourceMaximum=1
     SourceMinimum=0
     TargetMaximum=0
     TargetMinimum=1
     bClamp=False
     Value=&RigVMFunction_MathBoolToFloat(Value=@n4.Value).Result
    ).Result
    A=@n3.OutputParam
   )
   > &RigUnit_OffsetTransformForItem(
    Item=(Type=Bone,Name="upperarm_bicep_l")
    OffsetTransform=(Rotation=(X=0,Y=0,Z=0,W=1),Translation=(X=3,Y=4,Z=0),Scale3D=(X=1,Y=1,Z=1))
    bPropagateToChildren=True
    Weight=@n5.Result
   )
   > &RigUnit_OffsetTransformForItem(
    Item=(Type=Bone,Name="upperarm_tricep_l")
    OffsetTransform=(Rotation=(X=0,Y=0,Z=0,W=1),Translation=(X=-4,Y=-5,Z=0),Scale3D=(X=1,Y=1,Z=1))
    bPropagateToChildren=True
    Weight=@n5.Result
   )
   > n6 = &RigUnit_SphericalPoseReader(
    ActiveRegionScaleFactors=(PositiveWidth=1,NegativeWidth=1,PositiveHeight=1,NegativeHeight=1)
    ActiveRegionSize=0.1
    Debug=(bDrawDebug=false,bDraw2D=False,bDrawLocalAxes=False,DebugScale=15,DebugSegments=20,DebugThickness=0.1)
    DriverAxis=(X=-1,Y=0,Z=0)
    DriverItem=(Type=Bone,Name="upperarm_l")
    FalloffRegionScaleFactors=(PositiveWidth=1,NegativeWidth=1,PositiveHeight=1,NegativeHeight=1)
    FalloffSize=0.5
    FlipHeightScaling=False
    FlipWidthScaling=False
    OptionalParentItem=(Type=Bone,Name="upperarm_twist_02_r")
    RotationOffset=(X=-90,Y=0,Z=0)
   )
   > &RigUnit_ModifyTransforms(
    ItemToModify=((Item=(Type=Bone,Name="deform_pec_l"),Transform=(Rotation=(X=0,Y=0.1736,Z=0,W=0.9848),Translation=(X=-2,Y=0,Z=3),Scale3D=(X=1,Y=1,Z=1))))
    WeightMinimum=0
    Weight=@n6.OutputParam
   )
   > n7 = &RigUnit_SphericalPoseReader(
    ActiveRegionScaleFactors=(PositiveWidth=1,NegativeWidth=1,PositiveHeight=1,NegativeHeight=1)
    ActiveRegionSize=0.1
    Debug=(bDrawDebug=false,bDraw2D=False,bDrawLocalAxes=False,DebugScale=15,DebugSegments=20,DebugThickness=0.1)
    DriverAxis=(X=1,Y=0,Z=0)
    DriverItem=(Type=Bone,Name="upperarm_r")
    FalloffRegionScaleFactors=(PositiveWidth=1,NegativeWidth=1,PositiveHeight=1,NegativeHeight=1)
    FalloffSize=0.5
    FlipHeightScaling=False
    FlipWidthScaling=False
    OptionalParentItem=(Type=Bone,Name="upperarm_twist_02_l")
    RotationOffset=(X=90,Y=0,Z=0)
   )
   > &RigUnit_ModifyTransforms(
    ItemToModify=((Item=(Type=Bone,Name="deform_pec_r"),Transform=(Rotation=(X=0,Y=0.1736,Z=0,W=0.9848),Translation=(X=2,Y=0,Z=-3),Scale3D=(X=1,Y=1,Z=1))))
    WeightMinimum=0
    Weight=@n7.OutputParam
   )
   RigVMFunctionReturnNode(ExecutePin=@RigVMFunction_Sequence.C)
   n8 = &RigUnit_SphericalPoseReader(
    ActiveRegionScaleFactors=(PositiveWidth=1,NegativeWidth=1,PositiveHeight=1,NegativeHeight=1)
    ActiveRegionSize=0.1
    Debug=(bDrawDebug=false,bDraw2D=False,bDrawLocalAxes=False,DebugScale=15,DebugSegments=20,DebugThickness=0.1)
    DriverAxis=(X=-1,Y=0,Z=0)
    DriverItem=(Type=Bone,Name="calf_l")
    FalloffRegionScaleFactors=(PositiveWidth=1,NegativeWidth=1,PositiveHeight=1,NegativeHeight=1)
    FalloffSize=0.5
    FlipHeightScaling=False
    FlipWidthScaling=False
    OptionalParentItem=(Type=Bone,Name="upperarm_twist_02_r")
    RotationOffset=(X=-90,Y=0,Z=0)
    ExecutePin=@RigVMFunction_Sequence.B
   )
   > &RigUnit_OffsetTransformForItem(
    Item=(Type=Bone,Name="thigh_quads_l")
    OffsetTransform=(Rotation=(X=0,Y=0,Z=0,W=1),Translation=(X=3,Y=0,Z=0),Scale3D=(X=1,Y=1,Z=1))
    bPropagateToChildren=True
    Weight=@n8.OutputParam
   )
   > &RigUnit_OffsetTransformForItem(
    Item=(Type=Bone,Name="thigh_hamstrings_l")
    OffsetTransform=(Rotation=(X=0,Y=0,Z=0,W=1),Translation=(X=-3.5,Y=0,Z=0),Scale3D=(X=1,Y=1,Z=1))
    bPropagateToChildren=True
    Weight=@n8.OutputParam
   )
   > n9 = &RigUnit_SphericalPoseReader(
    ActiveRegionScaleFactors=(PositiveWidth=1,NegativeWidth=1,PositiveHeight=1,NegativeHeight=1)
    ActiveRegionSize=0.1
    Debug=(bDrawDebug=false,bDraw2D=False,bDrawLocalAxes=False,DebugScale=15,DebugSegments=20,DebugThickness=0.1)
    DriverAxis=(X=1,Y=0,Z=0)
    DriverItem=(Type=Bone,Name="calf_r")
    FalloffRegionScaleFactors=(PositiveWidth=1,NegativeWidth=1,PositiveHeight=1,NegativeHeight=1)
    FalloffSize=0.5
    FlipHeightScaling=False
    FlipWidthScaling=False
    OptionalParentItem=(Type=Bone,Name="upperarm_twist_02_l")
    RotationOffset=(X=90,Y=0,Z=0)
   )
   > &RigUnit_OffsetTransformForItem(
    Item=(Type=Bone,Name="thigh_quads_r")
    OffsetTransform=(Rotation=(X=0,Y=0,Z=0,W=1),Translation=(X=3,Y=0,Z=0),Scale3D=(X=1,Y=1,Z=1))
    bPropagateToChildren=True
    Weight=@n9.OutputParam
   )
   > &RigUnit_OffsetTransformForItem(
    Item=(Type=Bone,Name="thigh_hamstrings_r")
    OffsetTransform=(Rotation=(X=0,Y=0,Z=0,W=1),Translation=(X=-3.5,Y=0,Z=0),Scale3D=(X=1,Y=1,Z=1))
    bPropagateToChildren=True
    Weight=@n9.OutputParam
   )
  ]}
 CollapseNode_ContainedGraph {nodes [
   RigVMFunctionEntryNode()
   n1 = &RigVMFunction_AlphaInterp(
    InterpSpeedDecreasing=15
    InterpSpeedIncreasing=15
    Scale=1
    bClampResult=False
    Value=&RigVMFunction_MathBoolToFloat(Value=RigVMVariableNode() {variable=pub isTargeting}.Value).Result
   )
   > RigVMVariableNode(Value=@n1.Result) {variable=isTargeting lerp}
   RigVMFunction_MathBoolOr = RigVMAggregateNode(
    C=RigVMVariableNode() {variable=SMG_P90}.Value
    A=RigVMVariableNode() {variable=pub isAR Equipped}.Value
    B=RigVMVariableNode() {variable=is_OutLaw_SG}.Value
   )
   n2 = &RigVMFunction_AlphaInterp(
    InterpSpeedDecreasing=15
    InterpSpeedIncreasing=15
    Scale=1
    bClampResult=False
    Value=&RigVMFunction_MathBoolToFloat(Value=@RigVMFunction_MathBoolOr.Result).Result
   )
   > RigVMVariableNode(Value=@n2.Result) {variable=isAR Equipped lerp}
   n3 = &RigVMFunction_AlphaInterp(
    InterpSpeedDecreasing=15
    InterpSpeedIncreasing=15
    Scale=1
    bClampResult=False
    Value=&RigVMFunction_MathBoolToFloat(Value=RigVMVariableNode() {variable=pub isPistol Equipped}.Value).Result
   )
   > RigVMVariableNode(Value=@n3.Result) {variable=isPistol Equipped lerp}
   > RigVMFunctionReturnNode()
  ]}
 Multiply_11_ContainedGraph {nodes [
   n1 = RigVMFunctionEntryNode()
   RigVMFunctionReturnNode(Result=&RigVMFunction_MathDoubleMul(B=@n1.C, A=&RigVMFunction_MathDoubleMul(A=@n1.A, B=@n1.B).Result).Result)
  ]}
 Multiply_11_ContainedGraph {nodes [
   n1 = RigVMFunctionEntryNode()
   RigVMFunctionReturnNode(Result=&RigVMFunction_MathDoubleMul(B=@n1.C, A=&RigVMFunction_MathDoubleMul(A=@n1.A, B=@n1.B).Result).Result)
  ]}
 Multiply_11_ContainedGraph {nodes [
   n1 = RigVMFunctionEntryNode()
   RigVMFunctionReturnNode(Result=&RigVMFunction_MathDoubleMul(B=@n1.C, A=&RigVMFunction_MathDoubleMul(A=@n1.A, B=@n1.B).Result).Result)
  ]}
 Multiply_17_ContainedGraph {nodes [
   n1 = RigVMFunctionEntryNode()
   RigVMFunctionReturnNode(Result=&RigVMFunction_MathDoubleMul(B=@n1.C, A=&RigVMFunction_MathDoubleMul(A=@n1.A, B=@n1.B).Result).Result)
  ]}
 Multiply_17_ContainedGraph {nodes [
   n1 = RigVMFunctionEntryNode()
   RigVMFunctionReturnNode(Result=&RigVMFunction_MathDoubleMul(B=@n1.C, A=&RigVMFunction_MathDoubleMul(A=@n1.A, B=@n1.B).Result).Result)
  ]}
 Multiply_17_ContainedGraph {nodes [
   n1 = RigVMFunctionEntryNode()
   RigVMFunctionReturnNode(Result=&RigVMFunction_MathDoubleMul(B=@n1.C, A=&RigVMFunction_MathDoubleMul(A=@n1.A, B=@n1.B).Result).Result)
  ]}
 New Function_ContainedGraph {nodes [
   n1 = RigVMFunctionEntryNode()
   RigVMFunction_MathBoolToInteger = &RigVMFunction_MathBoolToInteger(Value=@n1.onlyTwist)
   RigVMFunction_MathQuaternionSwingTwist = &RigVMFunction_MathQuaternionSwingTwist(
    TwistAxis=@n1.twistAxis
    Input=&RigUnit_GetRelativeTransformForItem(bChildInitial=False, bParentInitial=False, Child=@n1.child, Parent=@n1.parent).RelativeTransform.Rotation
   )
   n2 = RigVMDispatchNode(
    Values=((),())
    Index=@RigVMFunction_MathBoolToInteger.Result
    Values.0=@RigVMFunction_MathQuaternionSwingTwist.Swing
    Values.1=@RigVMFunction_MathQuaternionSwingTwist.Twist
   ) {template="DISPATCH_RigVMDispatch_SelectInt32(in Index,in Values,out Result)"}
   RigVMFunction_MathQuaternionToEuler = &RigVMFunction_MathQuaternionToEuler(Value=@n2.Result, RotationOrder=@n1.Rotation Order)
   > DISPATCH_RigVMDispatch_Print = RigVMDispatchNode(ScreenColor=(R=1,G=1,B=1,A=1), ScreenDuration=1.5, Enabled=@n1.debugRotation, Value=@RigVMFunction_MathQuaternionToEuler.Result) {template="DISPATCH_RigVMDispatch_Print(in Prefix,in Value,in Enabled,in ScreenDuration,in ScreenColor)"}
   Cast = RigVMDispatchNode(Value=@n1.triggerAxis) {template="DISPATCH_RigVMDispatch_CastEnumToInt(in Value,out Result)"}
   DISPATCH_RigVMDispatch_SelectInt32 = RigVMDispatchNode(
    Values=(0,0,0)
    Index=@Cast.Result
    Values.0=@RigVMFunction_MathQuaternionToEuler.Result.X
    Values.1=@RigVMFunction_MathQuaternionToEuler.Result.Y
    Values.2=@RigVMFunction_MathQuaternionToEuler.Result.Z
   ) {template="DISPATCH_RigVMDispatch_SelectInt32(in Index,in Values,out Result)"}
   Remap = &RigVMFunction_MathDoubleRemap(
    bClamp=False
    Value=@DISPATCH_RigVMDispatch_SelectInt32.Result
    TargetMinimum=@n1.outputMinMax.X
    TargetMaximum=@n1.outputMinMax.Y
    SourceMinimum=@n1.rotationMinMax.X
    SourceMaximum=@n1.rotationMinMax.Y
   )
   Absolute = &RigVMFunction_MathDoubleAbs(Value=@Remap.Result)
   Multiply = &RigVMFunction_MathDoubleMul(
    A=&RigVMFunction_MathDoubleClamp(Value=RigVMDispatchNode(True=@Absolute.Result, Condition=@n1.absoluteValue, False=@Remap.Result) {template="DISPATCH_RigVMDispatch_If(in Condition,in True,in False,out Result)"}.Result).Result
    B=@n1.premult
   )
   > RigVMFunctionReturnNode(outputQuat=@n2.Result, outputWeight=@Multiply.Result)
  ]}
 New Function_ContainedGraph {nodes [
   RigVMFunctionEntryNode()
   > RigVMFunction_Sequence = RigVMAggregateNode()
   DeltaRotation_10 = RigVMFunctionReferenceNode(
    Rotation Order=ZYX
    child=(Type=Bone,Name="thigh_l")
    debugRotation=false
    outputMinMax=(X=0,Y=1)
    parent=(Type=Null,Name="thigh_l_null")
    premult=1
    rotationMinMax=(X=0,Y=90)
    triggerAxis=Z
    twistAxis=(X=1,Y=0,Z=0)
    ExecuteContext=@RigVMFunction_Sequence.A
   )
   > RigVMVariableNode(Value=@DeltaRotation_10.outputWeight) {variable=Leg_Back_R}
   > DeltaRotation_11 = RigVMFunctionReferenceNode(
    Rotation Order=ZYX
    child=(Type=Bone,Name="thigh_r")
    debugRotation=false
    outputMinMax=(X=0,Y=1)
    parent=(Type=Null,Name="thigh_r_null")
    premult=1
    rotationMinMax=(X=0,Y=90)
    triggerAxis=Z
    twistAxis=(X=1,Y=0,Z=0)
   )
   > RigVMVariableNode(Value=@DeltaRotation_11.outputWeight) {variable=Leg_Front_L}
   RigVMFunctionReturnNode(ExecuteContext=@RigVMFunction_Sequence.D)
   DeltaRotation_5 = RigVMFunctionReferenceNode(
    Rotation Order=ZYX
    child=(Type=Bone,Name="clavicle_l")
    debugRotation=false
    outputMinMax=(X=0,Y=1)
    parent=(Type=Null,Name="clavicle_l_null")
    premult=1
    rotationMinMax=(X=0,Y=-35)
    triggerAxis=Y
    twistAxis=(X=1,Y=0,Z=0)
    ExecuteContext=@RigVMFunction_Sequence.B
   )
   > RigVMVariableNode(Value=@DeltaRotation_5.outputWeight) {variable=SPR_Clavicle_l_Up}
   > DeltaRotation_6 = RigVMFunctionReferenceNode(
    Rotation Order=ZYX
    child=(Type=Bone,Name="clavicle_l")
    debugRotation=false
    outputMinMax=(X=0,Y=1)
    parent=(Type=Null,Name="clavicle_l_null")
    premult=1
    rotationMinMax=(X=0,Y=30)
    triggerAxis=Z
    twistAxis=(X=1,Y=0,Z=0)
   )
   > RigVMVariableNode(Value=@DeltaRotation_6.outputWeight) {variable=SPR_Clavicle_l_Fwd}
   > DeltaRotation_7 = RigVMFunctionReferenceNode(
    Rotation Order=ZYX
    child=(Type=Bone,Name="clavicle_r")
    debugRotation=false
    outputMinMax=(X=0,Y=1)
    parent=(Type=Null,Name="clavicle_r_null")
    premult=1
    rotationMinMax=(X=0,Y=-35)
    triggerAxis=Y
    twistAxis=(X=1,Y=0,Z=0)
   )
   > RigVMVariableNode(Value=@DeltaRotation_7.outputWeight) {variable=SPR_Clavicle_r_Up}
   > DeltaRotation_8 = RigVMFunctionReferenceNode(
    Rotation Order=ZYX
    child=(Type=Bone,Name="clavicle_r")
    debugRotation=false
    outputMinMax=(X=0,Y=1)
    parent=(Type=Null,Name="clavicle_r_null")
    premult=1
    rotationMinMax=(X=0,Y=30)
    triggerAxis=Z
    twistAxis=(X=1,Y=0,Z=0)
   )
   > RigVMVariableNode(Value=@DeltaRotation_8.outputWeight) {variable=SPR_Clavicle_r_Fwd}
   > DeltaRotation_12 = RigVMFunctionReferenceNode(
    Rotation Order=ZYX
    child=(Type=Bone,Name="clavicle_r")
    debugRotation=false
    outputMinMax=(X=0,Y=1)
    parent=(Type=Null,Name="clavicle_r_null")
    premult=1
    rotationMinMax=(X=0,Y=-30)
    triggerAxis=Y
    twistAxis=(X=1,Y=0,Z=0)
   )
   > RigVMVariableNode(Value=@DeltaRotation_12.outputWeight) {variable=SPR_Clavicle_r_Back}
   > DeltaRotation_14 = RigVMFunctionReferenceNode(
    Rotation Order=ZYX
    child=(Type=Bone,Name="clavicle_l")
    debugRotation=false
    outputMinMax=(X=0,Y=1)
    parent=(Type=Null,Name="clavicle_l_null")
    premult=1
    rotationMinMax=(X=0,Y=-30)
    triggerAxis=Z
    twistAxis=(X=1,Y=0,Z=0)
   )
   > RigVMVariableNode(Value=@DeltaRotation_14.outputWeight) {variable=SPR_Clavicle_l_Back}
   DeltaRotation_15 = RigVMFunctionReferenceNode(
    Rotation Order=ZYX
    child=(Type=Bone,Name="upperarm_l")
    debugRotation=false
    outputMinMax=(X=0,Y=1)
    parent=(Type=Null,Name="upperarm_l_null")
    premult=1
    rotationMinMax=(X=0,Y=-30)
    triggerAxis=Y
    twistAxis=(X=1,Y=0,Z=0)
    ExecuteContext=@RigVMFunction_Sequence.C
   )
   > RigVMVariableNode(Value=@DeltaRotation_15.outputWeight) {variable=SPR_Clavicle_l_Up_Parachuting}
   > DeltaRotation_13 = RigVMFunctionReferenceNode(
    Rotation Order=ZYX
    child=(Type=Bone,Name="upperarm_r")
    debugRotation=false
    outputMinMax=(X=0,Y=1)
    parent=(Type=Null,Name="upperarm_r_null")
    premult=1
    rotationMinMax=(X=0,Y=-30)
    triggerAxis=Y
    twistAxis=(X=1,Y=0,Z=0)
   )
   > RigVMVariableNode(Value=@DeltaRotation_13.outputWeight) {variable=SPR_Clavicle_r_Up_Parachuting}
   > DeltaRotation_16 = RigVMFunctionReferenceNode(
    Rotation Order=ZYX
    child=(Type=Bone,Name="clavicle_r")
    debugRotation=false
    outputMinMax=(X=0,Y=1)
    parent=(Type=Null,Name="clavicle_r_null")
    premult=1
    rotationMinMax=(X=0,Y=-25)
    triggerAxis=Y
    twistAxis=(X=1,Y=0,Z=0)
   )
   > RigVMVariableNode(Value=@DeltaRotation_16.outputWeight) {variable=right_Upperarm}
  ]}
 New Function_ContainedGraph {nodes [
   RigVMFunctionEntryNode()
   > RigVMFunction_Sequence = RigVMAggregateNode()
   &RigUnit_ModifyTransforms(
    ItemToModify="((Item=(Type=Bone,Name=\"base_side_01_skirt_l\"),Transform=(Rotation=(X=-0.1976,Y=-0.8019,Z=0.193,W=0.5298),Translation=(X=-0.9747,Y=1.2522,Z=-0.0637),Scale3D=(X=1,Y=1,Z=1))),(Item=(Type=Bone,Name=\"base_front_01_skirt_l\"),Transform=(Rotation=(X=0.0247,Y=0.29 ...(truncated)"
    WeightMinimum=0
    ExecutePin=@RigVMFunction_Sequence.B
    Weight=RigVMVariableNode() {variable=Leg_Side_R}.Value
   )
   > &RigUnit_ModifyTransforms(
    ItemToModify="((Item=(Type=Bone,Name=\"base_side_01_skirt_r\"),Transform=(Rotation=(X=0.0276,Y=-0.7268,Z=-0.1139,W=0.6767),Translation=(X=-1.8416,Y=0.435,Z=0.3876),Scale3D=(X=1,Y=1,Z=1))),(Item=(Type=Bone,Name=\"base_front_01_skirt_r\"),Transform=(Rotation=(X=-0.0501,Y=-0.1 ...(truncated)"
    WeightMinimum=0
    Weight=RigVMVariableNode() {variable=Leg_BackSide_R}.Value
   )
   > &RigUnit_ModifyTransforms(
    ItemToModify="((Item=(Type=Bone,Name=\"base_bk_01_skirt_l\"),Transform=(Rotation=(X=-0.0643,Y=-0.8636,Z=0.0371,W=0.4986),Translation=(X=-10.4607,Y=1.5133,Z=0.9666),Scale3D=(X=1,Y=1,Z=1))),(Item=(Type=Bone,Name=\"base_bk_01_skirt_c\"),Transform=(Rotation=(X=-0.0001,Y=-0.7691 ...(truncated)"
    WeightMinimum=0
    Weight=RigVMVariableNode() {variable=Leg_Back_R}.Value
   )
   > &RigUnit_ModifyTransforms(
    ItemToModify="((Item=(Type=Bone,Name=\"base_bk_01_skirt_r\"),Transform=(Rotation=(X=-0.0643,Y=-0.8636,Z=0.0371,W=0.4986),Translation=(X=-10.4607,Y=1.5133,Z=0.9666),Scale3D=(X=1,Y=1,Z=1))),(Item=(Type=Bone,Name=\"base_bk_01_skirt_c\"),Transform=(Rotation=(X=-0.0001,Y=-0.7691 ...(truncated)"
    WeightMinimum=0
    Weight=RigVMVariableNode() {variable=Leg_Front_L}.Value
   )
   > &RigUnit_ModifyTransforms(
    ItemToModify=((Item=(Type=Bone,Name="base_bk_01_skirt_c"),Transform=(Rotation=(X=0,Y=0.7556,Z=0,W=0.655),Translation=(X=-4.6846,Y=0,Z=-3.3692),Scale3D=(X=1,Y=1,Z=1))))
    WeightMinimum=0
    Weight=&RigVMFunction_MathDoubleMul(A=RigVMVariableNode() {variable=Leg_Back_R}.Value, B=RigVMVariableNode() {variable=Leg_Front_L}.Value).Result
   )
   DeltaRotation = RigVMFunctionReferenceNode(
    Rotation Order=ZYX
    child=(Type=Bone,Name="spine_05")
    onlyTwist=true
    outputMinMax=(X=0,Y=1)
    parent=(Type=Bone,Name="pelvis")
    premult=1
    rotationMinMax=(X=0,Y=-45)
    twistAxis=(X=1,Y=0,Z=0)
    ExecuteContext=@RigVMFunction_Sequence.C
   )
   AlphaInterp = &RigVMFunction_AlphaInterp(
    InterpSpeedDecreasing=2.5
    InterpSpeedIncreasing=1
    Scale=1
    bClampResult=False
    Value=&RigVMFunction_MathDoubleMul(B=1, A=@DeltaRotation.outputWeight).Result
   )
   > &RigUnit_ModifyTransforms(
    ItemToModify="((Item=(Type=Bone,Name=\"base_front_01_loin_c\"),Transform=(Rotation=(X=-0.0887,Y=0.0199,Z=0.0717,W=0.9933),Translation=(X=0,Y=4,Z=5),Scale3D=(X=1,Y=1,Z=1))),(Item=(Type=Bone,Name=\"base_front_01_skirt_l\"),Transform=(Rotation=(X=-0.2549,Y=-0.0471,Z=0.1151,W=0 ...(truncated)"
    WeightMinimum=0
    Weight=@AlphaInterp.Result
   )
   n1 = &RigVMFunction_AlphaInterp(InterpSpeedDecreasing=3, InterpSpeedIncreasing=3, Scale=1, Value=0, bClampResult=False)
   > &RigUnit_ModifyTransforms(
    ItemToModify="((Item=(Type=Bone,Name=\"base_front_01_loin_c\"),Transform=(Rotation=(X=-0.2245,Y=0.1542,Z=0.1046,W=0.9565),Translation=(X=-8,Y=1.5,Z=-6),Scale3D=(X=1,Y=1,Z=1))),(Item=(Type=Bone,Name=\"base_front_01_skirt_r\"),Transform=(Rotation=(X=0.0842,Y=-0.0226,Z=-0.2578 ...(truncated)"
    WeightMinimum=0
    Weight=@n1.Result
   )
   > n2 = &RigUnit_SphericalPoseReader(
    ActiveRegionScaleFactors=(PositiveWidth=1,NegativeWidth=1,PositiveHeight=1,NegativeHeight=0.8)
    ActiveRegionSize=0.1
    Debug=(bDrawDebug=true,bDraw2D=False,bDrawLocalAxes=False,DebugScale=25,DebugSegments=20,DebugThickness=0.25)
    DriverAxis=(X=1,Y=0,Z=0)
    DriverItem=(Type=Bone,Name="upperarm_r")
    FalloffRegionScaleFactors=(PositiveWidth=1,NegativeWidth=1,PositiveHeight=1.7,NegativeHeight=0.5)
    FalloffSize=0.2
    FlipHeightScaling=False
    FlipWidthScaling=False
    OptionalParentItem=(Type=Bone,Name="None")
    RotationOffset=(X=51.9395,Y=8.7023,Z=22.4256)
   )
   > &RigUnit_ModifyTransforms(
    ItemToModify="((Item=(Type=Bone,Name=\"None\"),Transform=(Rotation=(X=0,Y=0,Z=0,W=1),Translation=(X=0,Y=0,Z=0),Scale3D=(X=1,Y=1,Z=1))),(Item=(Type=Bone,Name=\"clavicle_r_helper\"),Transform=(Rotation=(X=0.0871,Y=-0.0015,Z=-0.0174,W=0.996),Translation=(X=0,Y=0,Z=0),Scale3D=( ...(truncated)"
    WeightMinimum=0
    Weight=@n2.OutputParam
   )
   RigVMFunctionReturnNode(ExecuteContext=@RigVMFunction_Sequence.D)
   n3 = RigVMVariableNode() {variable=Leg_Front_R}
   &RigUnit_ModifyTransforms(
    ItemToModify="((Item=(Type=Bone,Name=\"base_front_01_skirt_l\"),Transform=(Rotation=(X=0.1658,Y=-0.6276,Z=0.1391,W=0.7479),Translation=(X=-11,Y=0,Z=3),Scale3D=(X=1,Y=1,Z=1))),(Item=(Type=Bone,Name=\"base_side_01_skirt_l\"),Transform=(Rotation=(X=0.0665,Y=-0.1604,Z=0.3769,W= ...(truncated)"
    WeightMinimum=0
    Weight=@n3.Value
    ExecutePin=@RigVMFunction_Sequence.A
   )
   > &RigUnit_ModifyTransforms(
    ItemToModify=((Item=(Type=Bone,Name="base_front_01_loin_c"),Transform=(Rotation=(X=-0.0933,Y=-0.8306,Z=-0.1465,W=0.5291),Translation=(X=-3,Y=0,Z=0),Scale3D=(X=1,Y=1,Z=1))))
    WeightMinimum=0
    Weight=@n3.Value
   )
   > &RigUnit_ModifyTransforms(
    ItemToModify="((Item=(Type=Bone,Name=\"base_front_01_skirt_l\"),Transform=(Rotation=(X=0,Y=-0.4226,Z=0,W=0.9063),Translation=(X=-3,Y=0,Z=2),Scale3D=(X=1,Y=1,Z=1))),(Item=(Type=Bone,Name=\"base_side_01_skirt_l\"),Transform=(Rotation=(X=0,Y=0,Z=0.1305,W=0.9914),Translation=(X ...(truncated)"
    WeightMinimum=0
    Weight=RigVMVariableNode() {variable=Leg_Side_L}.Value
   )
   n4 = RigVMVariableNode() {variable=Leg_FrontSide_R}
   > &RigUnit_ModifyTransforms(
    ItemToModify="((Item=(Type=Bone,Name=\"base_front_01_skirt_r\"),Transform=(Rotation=(X=0,Y=-0.6756,Z=0,W=0.7373),Translation=(X=-11,Y=0,Z=2),Scale3D=(X=1,Y=1,Z=1))),(Item=(Type=Bone,Name=\"base_side_01_skirt_r\"),Transform=(Rotation=(X=0,Y=0,Z=-0.1305,W=0.9914),Translation= ...(truncated)"
    WeightMinimum=0
    Weight=@n4.Value
   )
   > &RigUnit_ModifyTransforms(
    ItemToModify=((Item=(Type=Bone,Name="base_front_01_loin_c"),Transform=(Rotation=(X=0,Y=0.9063,Z=0,W=-0.4226),Translation=(X=-1,Y=0,Z=-0.8078),Scale3D=(X=1,Y=1,Z=1))))
    WeightMinimum=0
    Weight=@n4.Value
   )
   > &RigUnit_ModifyTransforms(
    ItemToModify="((Item=(Type=Bone,Name=\"base_front_01_skirt_r\"),Transform=(Rotation=(X=0,Y=-0.3155,Z=0,W=0.9489),Translation=(X=-4.4578,Y=0,Z=4.24),Scale3D=(X=1,Y=1,Z=1))),(Item=(Type=Bone,Name=\"base_side_01_skirt_r\"),Transform=(Rotation=(X=0.0099,Y=0.075,Z=0.1302,W=-0.98 ...(truncated)"
    WeightMinimum=0
    Weight=RigVMVariableNode() {variable=SPR_Leg_R_Fwd_Ext}.Value
   )
   Multiply = &RigVMFunction_MathDoubleMul(A=RigVMVariableNode() {variable=Leg_Front_R}.Value, B=RigVMVariableNode() {variable=Leg_FrontSide_R}.Value)
   > &RigUnit_ModifyTransforms(
    ItemToModify=((Item=(Type=Bone,Name="base_front_01_loin_c"),Transform=(Rotation=(X=0,Y=0.9848,Z=0,W=0.1736),Translation=(X=5,Y=3,Z=0),Scale3D=(X=1,Y=1,Z=1))))
    WeightMinimum=0
    Weight=@Multiply.Result
   )
   &RigUnit_ModifyTransforms(
    ExecutePin=()
    ItemToModify="((Item=(Type=Bone,Name=\"base_side_01_skirt_l\"),Transform=(Rotation=(X=-0.1976,Y=-0.8019,Z=0.193,W=0.5298),Translation=(X=-0.9747,Y=1.2522,Z=-0.0637),Scale3D=(X=1,Y=1,Z=1))),(Item=(Type=Bone,Name=\"base_front_01_skirt_l\"),Transform=(Rotation=(X=0.0247,Y=0.29 ...(truncated)"
    WeightMinimum=0
    Weight=RigVMVariableNode() {variable=Leg_Side_R}.Value
   )
   > &RigUnit_ModifyTransforms(
    ItemToModify="((Item=(Type=Bone,Name=\"base_side_01_skirt_r\"),Transform=(Rotation=(X=0.0276,Y=-0.7268,Z=-0.1139,W=0.6767),Translation=(X=-1.8416,Y=0.435,Z=0.3876),Scale3D=(X=1,Y=1,Z=1))),(Item=(Type=Bone,Name=\"base_front_01_skirt_r\"),Transform=(Rotation=(X=-0.0501,Y=-0.1 ...(truncated)"
    WeightMinimum=0
    Weight=RigVMVariableNode() {variable=Leg_BackSide_R}.Value
   )
   n5 = RigVMRerouteNode(Value=())
   n6 = RigVMVariableNode() {variable=Leg_Front_R}
   &RigUnit_ModifyTransforms(
    ItemToModify="((Item=(Type=Bone,Name=\"base_front_01_skirt_l\"),Transform=(Rotation=(X=0,Y=-0.6756,Z=0,W=0.7372),Translation=(X=-11,Y=0,Z=2),Scale3D=(X=1,Y=1,Z=1))),(Item=(Type=Bone,Name=\"base_side_01_skirt_l\"),Transform=(Rotation=(X=0,Y=0,Z=0.1305,W=0.9914),Translation=( ...(truncated)"
    WeightMinimum=0
    ExecutePin=@n5.Value
    Weight=@n6.Value
   )
   > &RigUnit_ModifyTransforms(
    ItemToModify=((Item=(Type=Bone,Name="base_front_01_loin_c"),Transform=(Rotation=(X=0,Y=-0.887,Z=0,W=0.4617),Translation=(X=0,Y=0,Z=0),Scale3D=(X=1,Y=1,Z=1))))
    WeightMinimum=0
    Weight=@n6.Value
   )
   > &RigUnit_ModifyTransforms(
    ItemToModify="((Item=(Type=Bone,Name=\"base_front_01_skirt_l\"),Transform=(Rotation=(X=0,Y=-0.4226,Z=0,W=0.9063),Translation=(X=-3,Y=0,Z=2),Scale3D=(X=1,Y=1,Z=1))),(Item=(Type=Bone,Name=\"base_side_01_skirt_l\"),Transform=(Rotation=(X=0,Y=0,Z=0.1305,W=0.9914),Translation=(X ...(truncated)"
    WeightMinimum=0
    Weight=RigVMVariableNode() {variable=Leg_Side_L}.Value
   )
   n7 = RigVMVariableNode() {variable=Leg_FrontSide_R}
   > &RigUnit_ModifyTransforms(
    ItemToModify="((Item=(Type=Bone,Name=\"base_front_01_skirt_r\"),Transform=(Rotation=(X=0,Y=-0.6756,Z=0,W=0.7373),Translation=(X=-11,Y=0,Z=2),Scale3D=(X=1,Y=1,Z=1))),(Item=(Type=Bone,Name=\"base_side_01_skirt_r\"),Transform=(Rotation=(X=0,Y=0,Z=-0.1305,W=0.9914),Translation= ...(truncated)"
    WeightMinimum=0
    Weight=@n7.Value
   )
   > &RigUnit_ModifyTransforms(
    ItemToModify=((Item=(Type=Bone,Name="base_front_01_loin_c"),Transform=(Rotation=(X=0,Y=0.9063,Z=0,W=-0.4226),Translation=(X=-1,Y=0,Z=-0.8078),Scale3D=(X=1,Y=1,Z=1))))
    WeightMinimum=0
    Weight=@n7.Value
   )
   > &RigUnit_ModifyTransforms(
    ItemToModify=((Item=(Type=Bone,Name="base_front_01_loin_c"),Transform=(Rotation=(X=0,Y=0.929,Z=0,W=0.3701),Translation=(X=2.4857,Y=0,Z=-2.625),Scale3D=(X=1,Y=1,Z=1))))
    WeightMinimum=0
    Weight=&RigVMFunction_MathDoubleMul(B=RigVMVariableNode() {variable=Leg_FrontSide_R}.Value, A=RigVMVariableNode() {variable=Leg_Front_R}.Value).Result
   )
  ]}
 New Function_ContainedGraph {nodes [
   RigVMFunctionEntryNode()
   > RigVMFunction_Sequence = RigVMAggregateNode()
   &RigUnit_ModifyTransforms(
    ItemToModify="((Item=(Type=Bone,Name=\"base_SideArmor_l\"),Transform=(Rotation=(X=0.0368,Y=-0.421,Z=0.079,W=0.9029),Translation=(X=-1,Y=2,Z=-2),Scale3D=(X=1,Y=1.1,Z=1))),(Item=(Type=Bone,Name=\"base_SideArmor_l_front\"),Transform=(Rotation=(X=0,Y=-0.1736,Z=0,W=0.9848),Trans ...(truncated)"
    WeightMinimum=0
    ExecutePin=@RigVMFunction_Sequence.A
    Weight=RigVMVariableNode() {variable=SPR_Clavicle_l_Fwd}.Value
   )
   > &RigUnit_ModifyTransforms(
    ItemToModify="((Item=(Type=Bone,Name=\"base_SideArmor_l\"),Transform=(Rotation=(X=0.0468,Y=-0.5353,Z=0.0735,W=0.8402),Translation=(X=-5,Y=4,Z=-1),Scale3D=(X=0.5,Y=0.75,Z=1.5))),(Item=(Type=Bone,Name=\"dyn_SideArmor_l_front\"),Transform=(Rotation=(X=-0.2346,Y=-0.4082,Z=-0.10 ...(truncated)"
    WeightMinimum=0
    Weight=RigVMVariableNode() {variable=SPR_Clavicle_l_Up}.Value
   )
   > &RigUnit_ModifyTransforms(
    ItemToModify="((Item=(Type=Bone,Name=\"base_SideArmor_r\"),Transform=(Rotation=(X=0.0368,Y=-0.421,Z=0.079,W=0.9029),Translation=(X=-2,Y=-1,Z=2),Scale3D=(X=1,Y=1.1,Z=1))),(Item=(Type=Bone,Name=\"base_SideArmor_r_front\"),Transform=(Rotation=(X=0,Y=0.1736,Z=0,W=0.9848),Transl ...(truncated)"
    WeightMinimum=0
    Weight=RigVMVariableNode() {variable=SPR_Clavicle_r_Fwd}.Value
   )
   n1 = RigVMVariableNode() {variable=onGround}
   > &RigUnit_ModifyTransforms(
    ItemToModify="((Item=(Type=Bone,Name=\"base_SideArmor_r\"),Transform=(Rotation=(X=0.0468,Y=-0.5353,Z=0.0735,W=0.8402),Translation=(X=-5,Y=-8,Z=-1),Scale3D=(X=0.5,Y=0.75,Z=1.5))),(Item=(Type=Bone,Name=\"dyn_SideArmor_r_front\"),Transform=(Rotation=(X=-0.2346,Y=-0.4082,Z=-0.1 ...(truncated)"
    WeightMinimum=0
    Weight=&RigVMFunction_MathDoubleMul(A=&RigVMFunction_MathBoolToFloat(Value=@n1.Value).Result, B=RigVMVariableNode() {variable=SPR_Clavicle_r_Up}.Value).Result
   )
   n2 = &RigVMFunction_AlphaInterp(InterpSpeedDecreasing=10, InterpSpeedIncreasing=10, Scale=1, bClampResult=False, Value=RigVMVariableNode() {variable=isJumping}.Value)
   > &RigUnit_ModifyTransforms(
    ItemToModify="((Item=(Type=Bone,Name=\"base_SideArmor_r\"),Transform=(Rotation=(X=0,Y=-0.3035,Z=0,W=0.9528),Translation=(X=-3,Y=0,Z=-1),Scale3D=(X=1,Y=1,Z=1))),(Item=(Type=Bone,Name=\"dyn_SideArmor_r_rear\"),Transform=(Rotation=(X=-0.0044,Y=0.0002,Z=0.0495,W=0.9988),Transla ...(truncated)"
    WeightMinimum=0
    Weight=&RigVMFunction_MathDoubleMul(A=@n2.Result, B=RigVMVariableNode() {variable=right_Upperarm}.Value).Result
   )
   > &RigUnit_ModifyTransforms(
    ItemToModify="((Item=(Type=Bone,Name=\"base_SideArmor_l\"),Transform=(Rotation=(X=0,Y=0,Z=0,W=1),Translation=(X=-2,Y=0,Z=0),Scale3D=(X=1,Y=1,Z=1))),(Item=(Type=Bone,Name=\"base_SideArmor_l_rear\"),Transform=(Rotation=(X=0,Y=0,Z=0,W=1),Translation=(X=-8,Y=0,Z=0),Scale3D=(X=1 ...(truncated)"
    WeightMinimum=0
    Weight=RigVMVariableNode() {variable=SPR_Clavicle_l_Back}.Value
   )
   > &RigUnit_ModifyTransforms(
    ItemToModify="((Item=(Type=Bone,Name=\"base_SideArmor_r\"),Transform=(Rotation=(X=0,Y=0,Z=0,W=1),Translation=(X=-3,Y=0,Z=0),Scale3D=(X=1,Y=1,Z=1))),(Item=(Type=Bone,Name=\"base_SideArmor_r_rear\"),Transform=(Rotation=(X=0,Y=0,Z=0,W=1),Translation=(X=-8,Y=0,Z=2),Scale3D=(X=1 ...(truncated)"
    WeightMinimum=0
    Weight=&RigVMFunction_MathDoubleMul(A=RigVMVariableNode() {variable=Skydiving}.Value, B=RigVMVariableNode() {variable=SPR_Clavicle_r_Back}.Value).Result
   )
   > DeltaRotation_2 = RigVMFunctionReferenceNode(
    Rotation Order=ZYX
    child=(Type=Bone,Name="upperarm_l")
    debugRotation=false
    outputMinMax=(X=0,Y=1)
    parent=(Type=Null,Name="upperarm_l_null")
    premult=1
    rotationMinMax=(X=20,Y=5)
    triggerAxis=Z
    twistAxis=(X=1,Y=0,Z=0)
   )
   n3 = RigVMVariableNode() {variable=Parachute_opened}
   n4 = &RigVMFunction_AlphaInterp(
    InterpSpeedDecreasing=3
    InterpSpeedIncreasing=6
    Scale=1
    bClampResult=False
    Value=&RigVMFunction_MathDoubleMul(A=@DeltaRotation_2.outputWeight, B=&RigVMFunction_MathBoolToFloat(Value=@n3.Value).Result).Result
   )
   > &RigUnit_ModifyTransforms(
    ItemToModify="((Item=(Type=Bone,Name=\"base_SideArmor_l\"),Transform=(Rotation=(X=0,Y=-0.1908,Z=0,W=0.9816),Translation=(X=-4,Y=2,Z=3),Scale3D=(X=1,Y=1.25,Z=1.5))),(Item=(Type=Bone,Name=\"base_SideArmor_l_rear\"),Transform=(Rotation=(X=0,Y=0,Z=0,W=1),Translation=(X=-5,Y=0,Z ...(truncated)"
    WeightMinimum=0
    Weight=@n4.Result
   )
   > DeltaRotation_3 = RigVMFunctionReferenceNode(
    Rotation Order=ZYX
    child=(Type=Bone,Name="upperarm_r")
    debugRotation=false
    outputMinMax=(X=0,Y=1)
    parent=(Type=Null,Name="upperarm_r_null")
    premult=1
    rotationMinMax=(X=15.5,Y=-1)
    triggerAxis=Z
    twistAxis=(X=1,Y=0,Z=0)
   )
   n5 = RigVMVariableNode() {variable=Parachute_opened}
   n6 = &RigVMFunction_AlphaInterp(
    InterpSpeedDecreasing=3
    InterpSpeedIncreasing=7
    Scale=1
    bClampResult=False
    Value=&RigVMFunction_MathDoubleMul(A=@DeltaRotation_3.outputWeight, B=&RigVMFunction_MathBoolToFloat(Value=@n5.Value).Result).Result
   )
   > &RigUnit_ModifyTransforms(
    ItemToModify="((Item=(Type=Bone,Name=\"base_SideArmor_r\"),Transform=(Rotation=(X=0,Y=-0.1305,Z=0,W=0.9914),Translation=(X=-4,Y=4,Z=1),Scale3D=(X=1,Y=2.5,Z=1.3))),(Item=(Type=Bone,Name=\"clavicle_r_helper\"),Transform=(Rotation=(X=0,Y=0.0872,Z=0,W=0.9962),Translation=(X=0,Y ...(truncated)"
    WeightMinimum=0
    Weight=@n6.Result
   )
   RigVMFunctionReturnNode(ExecuteContext=@RigVMFunction_Sequence.C)
   &RigUnit_ModifyTransforms(
    ItemToModify="((Item=(Type=Bone,Name=\"dyn_SideArmor_r_rear\"),Transform=(Rotation=(X=0,Y=0,Z=0,W=1),Translation=(X=-8,Y=0,Z=0),Scale3D=(X=1,Y=1,Z=1))),(Item=(Type=Bone,Name=\"dyn_SideArmor_r\"),Transform=(Rotation=(X=0,Y=0,Z=0,W=1),Translation=(X=-1,Y=0,Z=0),Scale3D=(X=2,Y ...(truncated)"
    WeightMinimum=0
    Weight=&RigVMFunction_MathDoubleMul(A=RigVMVariableNode() {variable=Skydiving}.Value, B=RigVMVariableNode() {variable=SPR_Clavicle_r_Back}.Value).Result
    ExecutePin=@RigVMFunction_Sequence.B
   )
   n7 = RigVMVariableNode() {variable=Parachute_opened}
   AlphaInterp = &RigVMFunction_AlphaInterp(
    InterpSpeedDecreasing=3
    InterpSpeedIncreasing=0.5
    Scale=1
    bClampResult=False
    Value=&RigVMFunction_MathDoubleMul(B=&RigVMFunction_MathBoolToFloat(Value=@n7.Value).Result, A=RigVMVariableNode() {variable=SPR_Clavicle_r_Up}.Value).Result
   )
   > &RigUnit_ModifyTransforms(
    ItemToModify="((Item=(Type=Bone,Name=\"base_SideArmor_r\"),Transform=(Rotation=(X=0.05,Y=-0.5714,Z=0.0714,W=0.816),Translation=(X=-5.5,Y=-4,Z=-1),Scale3D=(X=0.5,Y=0.75,Z=1.5))),(Item=(Type=Bone,Name=\"dyn_SideArmor_r_front\"),Transform=(Rotation=(X=-0.2346,Y=0.4082,Z=0.1094 ...(truncated)"
    WeightMinimum=0
    Weight=@AlphaInterp.Result
   )
   RigVMFunction_MathBoolToFloat = &RigVMFunction_MathBoolToFloat(Value=RigVMVariableNode() {variable=Parachute_opened}.Value)
   Multiply = &RigVMFunction_MathDoubleMul(B=@RigVMFunction_MathBoolToFloat.Result, A=RigVMVariableNode() {variable=SPR_Clavicle_r_Up_Parachuting}.Value)
   n8 = &RigVMFunction_AlphaInterp(InterpSpeedDecreasing=5, InterpSpeedIncreasing=5, Scale=1, bClampResult=False, Value=@Multiply.Result)
   > &RigUnit_ModifyTransforms(
    ItemToModify="((Item=(Type=Bone,Name=\"base_SideArmor_r\"),Transform=(Rotation=(X=0,Y=-0.1305,Z=0,W=0.9914),Translation=(X=4,Y=8,Z=-1),Scale3D=(X=1.1,Y=1,Z=1.1))),(Item=(Type=Bone,Name=\"dyn_SideArmor_r_rear\"),Transform=(Rotation=(X=0,Y=0,Z=0,W=1),Translation=(X=0,Y=0,Z=0) ...(truncated)"
    WeightMinimum=0
    Weight=@n8.Result
   )
   n9 = RigVMVariableNode() {variable=Parachute_opened}
   n10 = &RigVMFunction_AlphaInterp(
    InterpSpeedDecreasing=5
    InterpSpeedIncreasing=5
    Scale=1
    bClampResult=False
    Value=&RigVMFunction_MathDoubleMul(B=&RigVMFunction_MathBoolToFloat(Value=@n9.Value).Result, A=RigVMVariableNode() {variable=SPR_Clavicle_l_Up_Parachuting}.Value).Result
   )
   > &RigUnit_ModifyTransforms(
    ItemToModify="((Item=(Type=Bone,Name=\"base_SideArmor_l\"),Transform=(Rotation=(X=0,Y=-0.1736,Z=0,W=0.9848),Translation=(X=2,Y=-8,Z=0.5),Scale3D=(X=1,Y=1,Z=1.1))),(Item=(Type=Bone,Name=\"base_SideArmor_l_rear\"),Transform=(Rotation=(X=0,Y=0,Z=0,W=1),Translation=(X=0,Y=0,Z=0 ...(truncated)"
    WeightMinimum=0
    Weight=@n10.Result
   )
   > n11 = &RigUnit_SphericalPoseReader(
    ActiveRegionScaleFactors=(PositiveWidth=1,NegativeWidth=1,PositiveHeight=1,NegativeHeight=0.8)
    ActiveRegionSize=0.1
    Debug=(bDrawDebug=false,bDraw2D=False,bDrawLocalAxes=False,DebugScale=25,DebugSegments=20,DebugThickness=0.25)
    DriverAxis=(X=1,Y=0,Z=0)
    DriverItem=(Type=Bone,Name="upperarm_r")
    FalloffRegionScaleFactors=(PositiveWidth=1,NegativeWidth=1,PositiveHeight=1.7,NegativeHeight=0.5)
    FalloffSize=0.2
    FlipHeightScaling=False
    FlipWidthScaling=False
    OptionalParentItem=(Type=Bone,Name="None")
    RotationOffset=(X=51.9395,Y=8.7023,Z=22.4256)
   )
   > &RigUnit_ModifyTransforms(
    ItemToModify="((Item=(Type=Bone,Name=\"deform_r_back_helper\"),Transform=(Rotation=(X=-0.0568,Y=-0.2652,Z=0.0781,W=0.9593),Translation=(X=4,Y=-5,Z=-2),Scale3D=(X=1,Y=1,Z=1))),(Item=(Type=Bone,Name=\"clavicle_r_helper\"),Transform=(Rotation=(X=0.0871,Y=-0.0015,Z=-0.0174,W=0. ...(truncated)"
    WeightMinimum=0
    Weight=@n11.OutputParam
   )
   > DeltaRotation = RigVMFunctionReferenceNode(
    Rotation Order=ZYX
    child=(Type=Bone,Name="head")
    debugRotation=false
    outputMinMax=(X=0,Y=1)
    parent=(Type=Null,Name="head_null")
    premult=1
    rotationMinMax=(X=3,Y=-15)
    triggerAxis=Z
    twistAxis=(X=1,Y=0,Z=0)
   )
   Multiply_9 = RigVMAggregateNode(
    A=@DeltaRotation.outputWeight
    B=RigVMVariableNode() {variable=isTargeting lerp}.Value
    C=RigVMVariableNode() {variable=isAR Equipped lerp}.Value
   )
   n12 = &RigVMFunction_AlphaInterp(InterpSpeedDecreasing=10, InterpSpeedIncreasing=20, Scale=1, bClampResult=False, Value=@Multiply_9.Result)
   > &RigUnit_ModifyTransforms(
    ItemToModify="((Item=(Type=Bone,Name=\"base_SideArmor_r\"),Transform=(Rotation=(X=-0.0199,Y=0.1776,Z=-0.1209,W=0.9764),Translation=(X=8.5,Y=2,Z=5),Scale3D=(X=1,Y=1,Z=1))),(Item=(Type=Bone,Name=\"base_SideArmor_l_rear\"),Transform=(Rotation=(X=0,Y=0,Z=0,W=1),Translation=(X=0 ...(truncated)"
    WeightMinimum=0
    Weight=@n12.Result
   )
   > DeltaRotation_1 = RigVMFunctionReferenceNode(
    Rotation Order=ZYX
    child=(Type=Bone,Name="head")
    debugRotation=false
    outputMinMax=(X=0,Y=1)
    parent=(Type=Null,Name="head_null")
    premult=1
    rotationMinMax=(X=3,Y=-15)
    triggerAxis=Z
    twistAxis=(X=1,Y=0,Z=0)
   )
   Multiply_10 = RigVMAggregateNode(
    A=@DeltaRotation_1.outputWeight
    B=RigVMVariableNode() {variable=isTargeting lerp}.Value
    C=RigVMVariableNode() {variable=isPistol Equipped lerp}.Value
   )
   n13 = &RigVMFunction_AlphaInterp(InterpSpeedDecreasing=10, InterpSpeedIncreasing=10, Scale=1, bClampResult=False, Value=@Multiply_10.Result)
   > &RigUnit_ModifyTransforms(
    ItemToModify="((Item=(Type=Bone,Name=\"base_SideArmor_r\"),Transform=(Rotation=(X=0,Y=0,Z=-0.061,W=0.9981),Translation=(X=4,Y=-2,Z=0),Scale3D=(X=1,Y=1,Z=1))),(Item=(Type=Bone,Name=\"base_SideArmor_l_rear\"),Transform=(Rotation=(X=0,Y=0,Z=0,W=1),Translation=(X=0,Y=0,Z=0),Sca ...(truncated)"
    WeightMinimum=0
    Weight=@n13.Result
   )
  ]}
 New Function_ContainedGraph {nodes [
   RigVMFunctionEntryNode()
   > &RigUnit_ParentConstraint(
    AdvancedSettings=(RotationOrderForFilter=XZY)
    Child=(Type=Bone,Name="base_SideArmor_l")
    Parents="((Item=(Type=Bone,Name=\"clavicle_r\"),Weight=0.0007),(Item=(Type=Bone,Name=\"upperarm_twist_01_l\"),Weight=0.0267),(Item=(Type=Bone,Name=\"neck_01\"),Weight=0.0267),(Item=(Type=Bone,Name=\"spine_04\"),Weight=0.0325),(Item=(Type=Bone,Name=\"neck_02\"),Weight=0.0344) ...(truncated)"
    Weight=1
   )
   > &RigUnit_ParentConstraint(
    AdvancedSettings=(RotationOrderForFilter=XZY)
    Child=(Type=Bone,Name="base_SideArmor_l_rear")
    Parents="((Item=(Type=Bone,Name=\"clavicle_r\"),Weight=0.0027),(Item=(Type=Bone,Name=\"upperarm_twist_01_l\"),Weight=0.012),(Item=(Type=Bone,Name=\"neck_02\"),Weight=0.013),(Item=(Type=Bone,Name=\"neck_01\"),Weight=0.017),(Item=(Type=Bone,Name=\"spine_04\"),Weight=0.0674),(I ...(truncated)"
    Weight=1
   )
   > &RigUnit_ParentConstraint(
    AdvancedSettings=(RotationOrderForFilter=XZY)
    Child=(Type=Bone,Name="base_SideArmor_l_front")
    Parents="((Item=(Type=Bone,Name=\"clavicle_r\"),Weight=0.0008),(Item=(Type=Bone,Name=\"neck_02\"),Weight=0.01),(Item=(Type=Bone,Name=\"upperarm_twist_01_l\"),Weight=0.0229),(Item=(Type=Bone,Name=\"spine_03\"),Weight=0.0485),(Item=(Type=Bone,Name=\"neck_01\"),Weight=0.0504),( ...(truncated)"
    Weight=1
   )
   > &RigUnit_ParentConstraint(
    AdvancedSettings=(RotationOrderForFilter=XZY)
    Child=(Type=Bone,Name="base_SideArmor_r_front")
    Parents="((Item=(Type=Bone,Name=\"clavicle_l\"),Weight=0.0008),(Item=(Type=Bone,Name=\"neck_02\"),Weight=0.01),(Item=(Type=Bone,Name=\"upperarm_twist_01_r\"),Weight=0.0229),(Item=(Type=Bone,Name=\"spine_03\"),Weight=0.0485),(Item=(Type=Bone,Name=\"neck_01\"),Weight=0.0504),( ...(truncated)"
    Weight=1
   )
   > &RigUnit_ParentConstraint(
    AdvancedSettings=(RotationOrderForFilter=XZY)
    Child=(Type=Bone,Name="base_SideArmor_r")
    Parents="((Item=(Type=Bone,Name=\"clavicle_l\"),Weight=0.0007),(Item=(Type=Bone,Name=\"upperarm_twist_01_r\"),Weight=0.0267),(Item=(Type=Bone,Name=\"neck_01\"),Weight=0.0267),(Item=(Type=Bone,Name=\"spine_04\"),Weight=0.0325),(Item=(Type=Bone,Name=\"neck_02\"),Weight=0.0344) ...(truncated)"
    Weight=1
   )
   > &RigUnit_ParentConstraint(
    AdvancedSettings=(RotationOrderForFilter=XZY)
    Child=(Type=Bone,Name="base_SideArmor_r_rear")
    Parents="((Item=(Type=Bone,Name=\"clavicle_l\"),Weight=0.0027),(Item=(Type=Bone,Name=\"upperarm_twist_01_r\"),Weight=0.012),(Item=(Type=Bone,Name=\"neck_02\"),Weight=0.013),(Item=(Type=Bone,Name=\"neck_01\"),Weight=0.017),(Item=(Type=Bone,Name=\"spine_04\"),Weight=0.0674),(I ...(truncated)"
    Weight=1
   )
   > &RigUnit_ParentConstraint(
    AdvancedSettings=(RotationOrderForFilter=XZY)
    Child=(Type=Bone,Name="base_deform_r_back_helper")
    Parents="((Item=(Type=Bone,Name=\"upperarm_r\"),Weight=0.0339),(Item=(Type=Bone,Name=\"upperarm_twist_02_r\"),Weight=0.0397),(Item=(Type=Bone,Name=\"spine_02\"),Weight=0.0499),(Item=(Type=Bone,Name=\"spine_05\"),Weight=0.1361),(Item=(Type=Bone,Name=\"clavicle_r_helper\"),Wei ...(truncated)"
    Weight=1
   )
   > &RigUnit_ParentConstraint(
    AdvancedSettings=(RotationOrderForFilter=XZY)
    Child=(Type=Bone,Name="Base_r_leg_inner")
    Parents="((Item=(Type=Bone,Name=\"calf_l\"),Weight=0.0001),(Item=(Type=Bone,Name=\"thigh_twist_01_r\"),Weight=0.0025),(Item=(Type=Bone,Name=\"deform_knee_in_r\"),Weight=0.0034),(Item=(Type=Bone,Name=\"calf_twist_02_r\"),Weight=0.0151),(Item=(Type=Bone,Name=\"thigh_r\"),Weigh ...(truncated)"
    Weight=1
   )
   > &RigUnit_ParentConstraint(
    AdvancedSettings=(RotationOrderForFilter=XZY)
    Child=(Type=Bone,Name="Base_r_leg_outer")
    Parents=((Item=(Type=Bone,Name="thigh_twist_01_r"),Weight=0.0004),(Item=(Type=Bone,Name="calf_twist_02_r"),Weight=0.0022),(Item=(Type=Bone,Name="calf_r"),Weight=0.4675),(Item=(Type=Bone,Name="thigh_r"),Weight=0.5299))
    Weight=1
   )
   > &RigUnit_ParentConstraint(
    AdvancedSettings=(RotationOrderForFilter=XZY)
    Child=(Type=Bone,Name="Base_l_leg_inner")
    Parents=((Item=(Type=Bone,Name="calf_r"),Weight=0.0007),(Item=(Type=Bone,Name="thigh_twist_01_l"),Weight=0.0028),(Item=(Type=Bone,Name="calf_twist_02_l"),Weight=0.0164),(Item=(Type=Bone,Name="calf_l"),Weight=0.4444),(Item=(Type=Bone,Name="thigh_l"),Weight=0.5357))
    Weight=1
   )
   > &RigUnit_ParentConstraint(
    AdvancedSettings=(RotationOrderForFilter=XZY)
    Child=(Type=Bone,Name="Base_l_leg_outer")
    Parents=((Item=(Type=Bone,Name="thigh_twist_01_l"),Weight=0.0004),(Item=(Type=Bone,Name="calf_twist_02_l"),Weight=0.0022),(Item=(Type=Bone,Name="calf_l"),Weight=0.4675),(Item=(Type=Bone,Name="thigh_l"),Weight=0.5299))
    Weight=1
   )
   > RigVMFunctionReturnNode()
  ]}
 New Function_ContainedGraph {nodes [
   RigVMFunctionEntryNode()
   > lerp bools = RigVMCollapseNode()
   > RigVMFunction_Sequence = RigVMAggregateNode()
   &RigUnit_ModifyTransforms(
    ItemToModify="((Item=(Type=Bone,Name=\"base_skirt_fr_l\"),Transform=(Rotation=(X=0,Y=0.64,Z=0,W=-0.7684),Translation=(X=-9.2865,Y=0.0002,Z=5.872),Scale3D=(X=1,Y=1,Z=1))),(Item=(Type=Bone,Name=\"base_skirt_fr_c\"),Transform=(Rotation=(X=-0.1373,Y=-0.5635,Z=0.1399,W=0.8025),T ...(truncated)"
    WeightMinimum=0
    Weight=&RigVMFunction_MathDoubleMul(B=RigVMVariableNode() {variable=Leg_Front_R}.Value, A=RigVMVariableNode() {variable=Leg_Front_L}.Value).Result
    ExecutePin=@RigVMFunction_Sequence.D
   )
   > &RigUnit_ModifyTransforms(
    ItemToModify="((Item=(Type=Bone,Name=\"dyn_l_leg_outer\"),Transform=(Rotation=(X=0,Y=0,Z=0,W=1),Translation=(X=2.5,Y=0,Z=0),Scale3D=(X=1,Y=1,Z=1))),(Item=(Type=Bone,Name=\"dyn_l_leg_inner\"),Transform=(Rotation=(X=0,Y=0,Z=0,W=1),Translation=(X=2.2,Y=0,Z=0),Scale3D=(X=1,Y=1, ...(truncated)"
    WeightMinimum=0
    Weight=RigVMVariableNode() {variable=Crouching}.Value
   )
   Multiply = &RigVMFunction_MathDoubleMul(A=RigVMVariableNode() {variable=Leg_Front_L}.Value, B=RigVMVariableNode() {variable=Crouching}.Value)
   > &RigUnit_ModifyTransforms(
    ItemToModify="((Item=(Type=Bone,Name=\"base_skirt_fr_l\"),Transform=(Rotation=(X=0.1555,Y=0.1853,Z=0.0545,W=0.9688),Translation=(X=5,Y=-5,Z=0),Scale3D=(X=1,Y=1,Z=1))),(Item=(Type=Bone,Name=\"base_skirt_sd_l\"),Transform=(Rotation=(X=0.2495,Y=0.1063,Z=0.0614,W=0.9606),Transl ...(truncated)"
    WeightMinimum=0
    Weight=@Multiply.Result
   )
   > &RigUnit_ModifyTransforms(
    ItemToModify="((Item=(Type=Bone,Name=\"base_skirt_fr_r\"),Transform=(Rotation=(X=-0.1392,Y=-0.3094,Z=0.0298,W=0.9402),Translation=(X=5,Y=0,Z=5),Scale3D=(X=1,Y=1,Z=1))),(Item=(Type=Bone,Name=\"base_skirt_fr_c\"),Transform=(Rotation=(X=0,Y=0,Z=0,W=1),Translation=(X=0,Y=0,Z=0) ...(truncated)"
    WeightMinimum=0
    Weight=&RigVMFunction_MathDoubleMul(B=RigVMVariableNode() {variable=Crouching}.Value, A=RigVMVariableNode() {variable=Leg_Front_R}.Value).Result
   )
   > &RigUnit_ModifyTransforms(
    ItemToModify="((Item=(Type=Bone,Name=\"base_skirt_fr_c\"),Transform=(Rotation=(X=-0.1663,Y=0.1106,Z=0.4098,W=-0.89),Translation=(X=0,Y=0,Z=0),Scale3D=(X=1,Y=1,Z=1))),(Item=(Type=Bone,Name=\"base_skirt_fr_l\"),Transform=(Rotation=(X=-0.2518,Y=0.1788,Z=0.1221,W=-0.9432),Trans ...(truncated)"
    WeightMinimum=0
    Weight=&RigVMFunction_MathDoubleMul(B=RigVMVariableNode() {variable=Leg_FrontInside_R}.Value, A=RigVMVariableNode() {variable=Crouching}.Value).Result
   )
   > &RigUnit_ModifyTransforms(
    ItemToModify="((Item=(Type=Bone,Name=\"base_skirt_fr_l\"),Transform=(Rotation=(X=0.0009,Y=0.0063,Z=0.0414,W=0.9991),Translation=(X=0,Y=0,Z=0),Scale3D=(X=1,Y=1,Z=1))),(Item=(Type=Bone,Name=\"base_skirt_fr_r\"),Transform=(Rotation=(X=-0.514,Y=-0.1869,Z=-0.1147,W=0.8292),Trans ...(truncated)"
    WeightMinimum=0
    Weight=&RigVMFunction_MathDoubleMul(A=RigVMVariableNode() {variable=Leg_FrontInside_L}.Value, B=RigVMVariableNode() {variable=Crouching}.Value).Result
   )
   > &RigUnit_ModifyTransforms(
    ItemToModify="((Item=(Type=Bone,Name=\"base_skirt_fr_r\"),Transform=(Rotation=(X=0,Y=-0.0872,Z=0,W=0.9962),Translation=(X=0,Y=0,Z=0),Scale3D=(X=1,Y=1,Z=1))),(Item=(Type=Bone,Name=\"base_skirt_fr_l\"),Transform=(Rotation=(X=0,Y=-0.0436,Z=0,W=0.999),Translation=(X=0,Y=0,Z=0), ...(truncated)"
    Weight=1
    WeightMinimum=0
   )
   RigVMFunctionReturnNode(ExecuteContext=@RigVMFunction_Sequence.G)
   n1 = &RigUnit_SphericalPoseReader(
    ActiveRegionScaleFactors=(PositiveWidth=1,NegativeWidth=1,PositiveHeight=1,NegativeHeight=1)
    ActiveRegionSize=0.05
    Debug=(bDrawDebug=false,bDraw2D=False,bDrawLocalAxes=False,DebugScale=25,DebugSegments=20,DebugThickness=0.25)
    DriverAxis=(X=1,Y=0,Z=0)
    DriverItem=(Type=Bone,Name="thigh_r")
    FalloffRegionScaleFactors=(PositiveWidth=1,NegativeWidth=1.1,PositiveHeight=1,NegativeHeight=1)
    FalloffSize=0.55
    FlipHeightScaling=true
    FlipWidthScaling=true
    OptionalParentItem=(Type=Bone,Name="None")
    RotationOffset=(X=-90,Y=0,Z=-20)
    ExecutePin=@RigVMFunction_Sequence.A
   )
   n2 = RigVMVariableNode(Value=@n1.OutputParam, ExecuteContext=@RigVMFunction_Sequence.B) {variable=Leg_Front_R}
   n3 = &RigUnit_SphericalPoseReader(
    ActiveRegionScaleFactors=(PositiveWidth=1,NegativeWidth=1,PositiveHeight=1,NegativeHeight=1)
    ActiveRegionSize=0.05
    Debug=(bDrawDebug=false,bDraw2D=False,bDrawLocalAxes=False,DebugScale=25,DebugSegments=20,DebugThickness=0.25)
    DriverAxis=(X=1,Y=0,Z=0)
    DriverItem=(Type=Bone,Name="thigh_r")
    FalloffRegionScaleFactors=(PositiveWidth=1,NegativeWidth=1,PositiveHeight=1,NegativeHeight=1)
    FalloffSize=0.25
    FlipHeightScaling=false
    FlipWidthScaling=false
    OptionalParentItem=(Type=Bone,Name="None")
    RotationOffset=(X=-40,Y=30,Z=0)
    ExecutePin=@n1.ExecutePin
   )
   n4 = RigVMVariableNode(Value=@n3.OutputParam, ExecuteContext=@n2.ExecuteContext) {variable=Leg_FrontSide_R}
   n5 = &RigUnit_SphericalPoseReader(
    ActiveRegionScaleFactors=(PositiveWidth=1,NegativeWidth=1,PositiveHeight=1,NegativeHeight=1)
    ActiveRegionSize=0.05
    Debug=(bDrawDebug=false,bDraw2D=False,bDrawLocalAxes=False,DebugScale=25,DebugSegments=20,DebugThickness=0.25)
    DriverAxis=(X=1,Y=0,Z=0)
    DriverItem=(Type=Bone,Name="thigh_r")
    FalloffRegionScaleFactors=(PositiveWidth=1,NegativeWidth=1.2,PositiveHeight=1,NegativeHeight=1)
    FalloffSize=0.55
    FlipHeightScaling=true
    FlipWidthScaling=true
    OptionalParentItem=(Type=Bone,Name="None")
    RotationOffset=(X=0,Y=20,Z=0)
    ExecutePin=@n3.ExecutePin
   )
   n6 = RigVMVariableNode(Value=@n5.OutputParam, ExecuteContext=@n4.ExecuteContext) {variable=Leg_Side_R}
   n7 = &RigUnit_SphericalPoseReader(
    ActiveRegionScaleFactors=(PositiveWidth=1,NegativeWidth=1,PositiveHeight=1,NegativeHeight=1)
    ActiveRegionSize=0.05
    Debug=(bDrawDebug=false,bDraw2D=False,bDrawLocalAxes=False,DebugScale=25,DebugSegments=20,DebugThickness=0.25)
    DriverAxis=(X=1,Y=0,Z=0)
    DriverItem=(Type=Bone,Name="thigh_r")
    FalloffRegionScaleFactors=(PositiveWidth=1,NegativeWidth=1,PositiveHeight=1,NegativeHeight=1.25)
    FalloffSize=0.2
    FlipHeightScaling=true
    FlipWidthScaling=true
    OptionalParentItem=(Type=Bone,Name="None")
    RotationOffset=(X=40,Y=20,Z=0)
    ExecutePin=@n5.ExecutePin
   )
   n8 = RigVMVariableNode(Value=@n7.OutputParam, ExecuteContext=@n6.ExecuteContext) {variable=Leg_BackSide_R}
   n9 = &RigUnit_SphericalPoseReader(
    ActiveRegionScaleFactors=(PositiveWidth=1,NegativeWidth=1,PositiveHeight=1,NegativeHeight=1)
    ActiveRegionSize=0.05
    Debug=(bDrawDebug=false,bDraw2D=false,bDrawLocalAxes=False,DebugScale=25,DebugSegments=20,DebugThickness=0.25)
    DriverAxis=(X=1,Y=0,Z=0)
    DriverItem=(Type=Bone,Name="thigh_r")
    FalloffRegionScaleFactors=(PositiveWidth=1,NegativeWidth=1,PositiveHeight=1,NegativeHeight=1)
    FalloffSize=0.55
    FlipHeightScaling=true
    FlipWidthScaling=true
    OptionalParentItem=(Type=Bone,Name="None")
    RotationOffset=(X=90,Y=0,Z=20)
    ExecutePin=@n7.ExecutePin
   )
   n10 = RigVMVariableNode(Value=@n9.OutputParam, ExecuteContext=@n8.ExecuteContext) {variable=Leg_Back_R}
   n11 = &RigUnit_SphericalPoseReader(
    ActiveRegionScaleFactors=(PositiveWidth=1,NegativeWidth=1,PositiveHeight=1,NegativeHeight=1)
    ActiveRegionSize=0.05
    Debug=(bDrawDebug=false,bDraw2D=false,bDrawLocalAxes=False,DebugScale=25,DebugSegments=20,DebugThickness=0.25)
    DriverAxis=(X=1,Y=0,Z=0)
    DriverItem=(Type=Bone,Name="thigh_r")
    FalloffRegionScaleFactors=(PositiveWidth=2,NegativeWidth=1.5,PositiveHeight=1,NegativeHeight=1)
    FalloffSize=0.22
    FlipHeightScaling=true
    FlipWidthScaling=true
    OptionalParentItem=(Type=Bone,Name="None")
    RotationOffset=(X=-130,Y=0,Z=0)
    ExecutePin=@n9.ExecutePin
   )
   n12 = RigVMVariableNode(Value=@n11.OutputParam, ExecuteContext=@n10.ExecuteContext) {variable=Leg_FrontInside_R}
   n13 = &RigUnit_SphericalPoseReader(
    ActiveRegionScaleFactors=(PositiveWidth=1,NegativeWidth=1,PositiveHeight=1,NegativeHeight=1)
    ActiveRegionSize=0.05
    Debug=(bDrawDebug=false,bDraw2D=False,bDrawLocalAxes=False,DebugScale=25,DebugSegments=20,DebugThickness=0.25)
    DriverAxis=(X=-1,Y=0,Z=0)
    DriverItem=(Type=Bone,Name="thigh_l")
    FalloffRegionScaleFactors=(PositiveWidth=1,NegativeWidth=1.1,PositiveHeight=1,NegativeHeight=1)
    FalloffSize=0.55
    FlipHeightScaling=False
    FlipWidthScaling=False
    OptionalParentItem=(Type=Bone,Name="None")
    RotationOffset=(X=90,Y=0,Z=-20)
    ExecutePin=@n11.ExecutePin
   )
   n14 = RigVMVariableNode(Value=@n13.OutputParam, ExecuteContext=@n12.ExecuteContext) {variable=Leg_Front_L}
   n15 = &RigUnit_SphericalPoseReader(
    ActiveRegionScaleFactors=(PositiveWidth=1,NegativeWidth=1,PositiveHeight=1,NegativeHeight=1)
    ActiveRegionSize=0.05
    Debug=(bDrawDebug=false,bDraw2D=False,bDrawLocalAxes=False,DebugScale=25,DebugSegments=20,DebugThickness=0.25)
    DriverAxis=(X=-1,Y=0,Z=0)
    DriverItem=(Type=Bone,Name="thigh_l")
    FalloffRegionScaleFactors=(PositiveWidth=1,NegativeWidth=1,PositiveHeight=1,NegativeHeight=1)
    FalloffSize=0.25
    FlipHeightScaling=False
    FlipWidthScaling=False
    OptionalParentItem=(Type=Bone,Name="None")
    RotationOffset=(X=140,Y=30,Z=0)
    ExecutePin=@n13.ExecutePin
   )
   n16 = RigVMVariableNode(Value=@n15.OutputParam, ExecuteContext=@n14.ExecuteContext) {variable=Leg_FrontSide_L}
   n17 = &RigUnit_SphericalPoseReader(
    ActiveRegionScaleFactors=(PositiveWidth=1,NegativeWidth=1,PositiveHeight=1,NegativeHeight=1)
    ActiveRegionSize=0.05
    Debug=(bDrawDebug=false,bDraw2D=False,bDrawLocalAxes=False,DebugScale=25,DebugSegments=20,DebugThickness=0.25)
    DriverAxis=(X=-1,Y=0,Z=0)
    DriverItem=(Type=Bone,Name="thigh_l")
    FalloffRegionScaleFactors=(PositiveWidth=1,NegativeWidth=1.12,PositiveHeight=1,NegativeHeight=1)
    FalloffSize=0.55
    FlipHeightScaling=False
    FlipWidthScaling=False
    OptionalParentItem=(Type=Bone,Name="None")
    RotationOffset=(X=180,Y=20,Z=0)
    ExecutePin=@n15.ExecutePin
   )
   n18 = RigVMVariableNode(Value=@n17.OutputParam, ExecuteContext=@n16.ExecuteContext) {variable=Leg_Side_L}
   n19 = &RigUnit_SphericalPoseReader(
    ActiveRegionScaleFactors=(PositiveWidth=1,NegativeWidth=1,PositiveHeight=1,NegativeHeight=1)
    ActiveRegionSize=0.05
    Debug=(bDrawDebug=false,bDraw2D=False,bDrawLocalAxes=False,DebugScale=25,DebugSegments=20,DebugThickness=0.25)
    DriverAxis=(X=-1,Y=0,Z=0)
    DriverItem=(Type=Bone,Name="thigh_l")
    FalloffRegionScaleFactors=(PositiveWidth=1,NegativeWidth=1,PositiveHeight=1,NegativeHeight=1.25)
    FalloffSize=0.2
    FlipHeightScaling=true
    FlipWidthScaling=false
    OptionalParentItem=(Type=Bone,Name="None")
    RotationOffset=(X=-140,Y=20,Z=0)
    ExecutePin=@n17.ExecutePin
   )
   n20 = RigVMVariableNode(Value=@n19.OutputParam, ExecuteContext=@n18.ExecuteContext) {variable=Leg_BackSide_L}
   n21 = &RigUnit_SphericalPoseReader(
    ActiveRegionScaleFactors=(PositiveWidth=1,NegativeWidth=1,PositiveHeight=1,NegativeHeight=1)
    ActiveRegionSize=0.05
    Debug=(bDrawDebug=false,bDraw2D=False,bDrawLocalAxes=False,DebugScale=25,DebugSegments=20,DebugThickness=0.25)
    DriverAxis=(X=-1,Y=0,Z=0)
    DriverItem=(Type=Bone,Name="thigh_l")
    FalloffRegionScaleFactors=(PositiveWidth=1,NegativeWidth=1,PositiveHeight=1,NegativeHeight=1)
    FalloffSize=0.55
    FlipHeightScaling=False
    FlipWidthScaling=False
    OptionalParentItem=(Type=Bone,Name="None")
    RotationOffset=(X=-90,Y=0,Z=20)
    ExecutePin=@n19.ExecutePin
   )
   n22 = RigVMVariableNode(Value=@n21.OutputParam, ExecuteContext=@n20.ExecuteContext) {variable=Leg_Back_L}
   n23 = &RigUnit_SphericalPoseReader(
    ActiveRegionScaleFactors=(PositiveWidth=1,NegativeWidth=1,PositiveHeight=1,NegativeHeight=1)
    ActiveRegionSize=0.05
    Debug=(bDrawDebug=false,bDraw2D=False,bDrawLocalAxes=False,DebugScale=25,DebugSegments=20,DebugThickness=0.25)
    DriverAxis=(X=-1,Y=0,Z=0)
    DriverItem=(Type=Bone,Name="thigh_l")
    FalloffRegionScaleFactors=(PositiveWidth=2,NegativeWidth=2,PositiveHeight=1,NegativeHeight=2)
    FalloffSize=0.2
    FlipHeightScaling=true
    FlipWidthScaling=true
    OptionalParentItem=(Type=Bone,Name="None")
    RotationOffset=(X=40,Y=0,Z=0)
    ExecutePin=@n21.ExecutePin
   )
   RigVMVariableNode(Value=@n23.OutputParam, ExecuteContext=@n22.ExecuteContext) {variable=Leg_FrontInside_L}
   Remap = &RigVMFunction_MathDoubleRemap(SourceMaximum=1, SourceMinimum=0, TargetMaximum=0, TargetMinimum=1, bClamp=False, Value=RigVMVariableNode() {variable=Leg_Front_L}.Value)
   &RigUnit_ModifyTransforms(
    ItemToModify="((Item=(Type=Bone,Name=\"base_skirt_fr_r\"),Transform=(Rotation=(X=0.0668,Y=-0.6403,Z=0.056,W=0.7631),Translation=(X=-12.0393,Y=5,Z=0),Scale3D=(X=0.8659,Y=1,Z=1))),(Item=(Type=Bone,Name=\"base_skirt_sd_r\"),Transform=(Rotation=(X=0.0193,Y=0.172,Z=0.4832,W=-0.8 ...(truncated)"
    WeightMinimum=0
    Weight=&RigVMFunction_MathDoubleMul(A=@Remap.Result, B=RigVMVariableNode() {variable=Leg_Front_R}.Value).Result
    ExecutePin=@RigVMFunction_Sequence.C
   )
   > &RigUnit_ModifyTransforms(
    ItemToModify="((Item=(Type=Bone,Name=\"base_skirt_fr_r\"),Transform=(Rotation=(X=0.0866,Y=-0.3141,Z=0.5489,W=0.7698),Translation=(X=-1.1876,Y=-3.1231,Z=4.3185),Scale3D=(X=1,Y=1,Z=1))),(Item=(Type=Bone,Name=\"base_skirt_sd_r\"),Transform=(Rotation=(X=-0.1865,Y=-0.0009,Z=0.34 ...(truncated)"
    WeightMinimum=0
    Weight=RigVMVariableNode() {variable=Leg_FrontSide_R}.Value
   )
   > &RigUnit_ModifyTransforms(
    ItemToModify="((Item=(Type=Bone,Name=\"base_skirt_fr_r\"),Transform=(Rotation=(X=0.2296,Y=0.0273,Z=0.4388,W=0.8683),Translation=(X=-2.037,Y=-0.1065,Z=2.4601),Scale3D=(X=0.9107,Y=1,Z=1))),(Item=(Type=Bone,Name=\"base_skirt_sd_r\"),Transform=(Rotation=(X=-0.1906,Y=0.6559,Z=0. ...(truncated)"
    WeightMinimum=0
    Weight=RigVMVariableNode() {variable=Leg_Side_R}.Value
   )
   > &RigUnit_ModifyTransforms(
    ItemToModify="((Item=(Type=Bone,Name=\"base_skirt_fr_r\"),Transform=(Rotation=(X=0,Y=0,Z=0,W=1),Translation=(X=0,Y=0,Z=0),Scale3D=(X=1,Y=1,Z=1))),(Item=(Type=Bone,Name=\"base_skirt_sd_r\"),Transform=(Rotation=(X=0.0289,Y=-0.2165,Z=0.0764,W=0.9729),Translation=(X=0,Y=0,Z=0), ...(truncated)"
    WeightMinimum=0
    Weight=RigVMVariableNode() {variable=Leg_BackSide_R}.Value
   )
   > &RigUnit_ModifyTransforms(
    ItemToModify=((Item=(Type=Bone,Name="base_skirt_sd_r"),Transform=(Rotation=(X=0,Y=0,Z=0.3928,W=0.9196),Translation=(X=0,Y=0,Z=0),Scale3D=(X=1,Y=1,Z=1))))
    WeightMinimum=0
    Weight=RigVMVariableNode() {variable=Leg_Back_R}.Value
   )
   n24 = RigVMVariableNode() {variable=Leg_Front_R}
   > &RigUnit_ModifyTransforms(
    ItemToModify="((Item=(Type=Bone,Name=\"base_skirt_fr_l\"),Transform=(Rotation=(X=-0.3598,Y=-0.834,Z=0.0971,W=0.407),Translation=(X=-4.7349,Y=0.934,Z=-1.2451),Scale3D=(X=1,Y=1,Z=1))),(Item=(Type=Bone,Name=\"base_skirt_sd_l\"),Transform=(Rotation=(X=0.0498,Y=0.0708,Z=-0.4977, ...(truncated)"
    WeightMinimum=0
    Weight=&RigVMFunction_MathDoubleMul(
     A=&RigVMFunction_MathDoubleRemap(SourceMaximum=1, SourceMinimum=0, TargetMaximum=0, TargetMinimum=1, bClamp=False, Value=@n24.Value).Result
     B=RigVMVariableNode() {variable=Leg_Front_L}.Value
    ).Result
   )
   > &RigUnit_ModifyTransforms(
    ItemToModify="((Item=(Type=Bone,Name=\"base_skirt_fr_l\"),Transform=(Rotation=(X=0.2142,Y=-0.0491,Z=0.1285,W=-0.9671),Translation=(X=-0.2188,Y=0.412,Z=3.9752),Scale3D=(X=1,Y=1,Z=1))),(Item=(Type=Bone,Name=\"base_skirt_sd_l\"),Transform=(Rotation=(X=-0.2101,Y=-0.159,Z=0.2123 ...(truncated)"
    WeightMinimum=0
    Weight=RigVMVariableNode() {variable=Leg_FrontSide_L}.Value
   )
   > &RigUnit_ModifyTransforms(
    ItemToModify="((Item=(Type=Bone,Name=\"base_skirt_fr_l\"),Transform=(Rotation=(X=0.4224,Y=0.1027,Z=0.4124,W=-0.8006),Translation=(X=-0.7335,Y=0.4704,Z=0.1133),Scale3D=(X=1,Y=1,Z=1))),(Item=(Type=Bone,Name=\"base_skirt_sd_l\"),Transform=(Rotation=(X=0.0198,Y=-0.6081,Z=-0.025 ...(truncated)"
    WeightMinimum=0
    Weight=RigVMVariableNode() {variable=Leg_Side_L}.Value
   )
   > &RigUnit_ModifyTransforms(
    ItemToModify="((Item=(Type=Bone,Name=\"base_skirt_fr_l\"),Transform=(Rotation=(X=0,Y=0,Z=0,W=1),Translation=(X=0,Y=0,Z=0),Scale3D=(X=1,Y=1,Z=1))),(Item=(Type=Bone,Name=\"base_skirt_sd_l\"),Transform=(Rotation=(X=-0.0477,Y=-0.0993,Z=-0.2173,W=0.9699),Translation=(X=0,Y=0,Z=0 ...(truncated)"
    WeightMinimum=0
    Weight=RigVMVariableNode() {variable=Leg_BackSide_L}.Value
   )
   > &RigUnit_ModifyTransforms(
    ItemToModify="((Item=(Type=Bone,Name=\"base_skirt_sd_l\"),Transform=(Rotation=(X=0,Y=0,Z=0,W=1),Translation=(X=0,Y=0,Z=0),Scale3D=(X=1,Y=1,Z=1))),(Item=(Type=Bone,Name=\"base_skirt_bk_c\"),Transform=(Rotation=(X=-0.0162,Y=-0.6905,Z=0.0168,W=0.723),Translation=(X=0,Y=0,Z=0), ...(truncated)"
    WeightMinimum=0
    Weight=RigVMVariableNode() {variable=Leg_Back_L}.Value
   )
   > &RigUnit_ModifyTransforms(
    ItemToModify="((Item=(Type=Bone,Name=\"base_skirt_fr_l\"),Transform=(Rotation=(X=0.2189,Y=-0.2125,Z=0.2498,W=0.919),Translation=(X=-3,Y=5,Z=5),Scale3D=(X=1,Y=1,Z=1))),(Item=(Type=Bone,Name=\"base_skirt_fr_r\"),Transform=(Rotation=(X=-0.0955,Y=0.2325,Z=0.087,W=-0.964),Transl ...(truncated)"
    WeightMinimum=0
    Weight=&RigVMFunction_MathDoubleMul(A=RigVMVariableNode() {variable=Leg_FrontInside_L}.Value, B=RigVMVariableNode() {variable=Crouching}.Value).Result
   )
   &RigUnit_ModifyTransforms(
    ItemToModify="((Item=(Type=Bone,Name=\"base_bk_01_skirt_l\"),Transform=(Rotation=(X=0,Y=-1,Z=0,W=0),Translation=(X=-5.8501,Y=0.0001,Z=3.8431),Scale3D=(X=1,Y=1,Z=1))),(Item=(Type=Bone,Name=\"base_bk_01_skirt_c\"),Transform=(Rotation=(X=0.0144,Y=-0.9272,Z=0.0208,W=0.3738),Tra ...(truncated)"
    WeightMinimum=0
    Weight=RigVMVariableNode() {variable=Leg_Back_L}.Value
    ExecutePin=@RigVMFunction_Sequence.E
   )
   > &RigUnit_ModifyTransforms(
    ItemToModify="((Item=(Type=Bone,Name=\"base_bk_01_skirt_r\"),Transform=(Rotation=(X=0.0047,Y=-0.9987,Z=-0.0004,W=0.0508),Translation=(X=-8.1878,Y=-0.0004,Z=2.5349),Scale3D=(X=1,Y=1,Z=1))),(Item=(Type=Bone,Name=\"base_bk_01_skirt_c\"),Transform=(Rotation=(X=0,Y=-0.9063,Z=0,W ...(truncated)"
    WeightMinimum=0
    Weight=RigVMVariableNode() {variable=Leg_Back_R}.Value
   )
   > DeltaRotation = RigVMFunctionReferenceNode(
    Rotation Order=ZYX
    child=(Type=Bone,Name="thigh_l")
    debugRotation=false
    onlyTwist=false
    outputMinMax=(X=0,Y=1)
    parent=(Type=Null,Name="thigh_l_null")
    premult=1
    rotationMinMax=(X=-27,Y=5)
    triggerAxis=Y
    twistAxis=(X=1,Y=0,Z=0)
   )
   Multiply_15 = RigVMAggregateNode(
    A=@DeltaRotation.outputWeight
    C=&RigVMFunction_MathBoolToFloat(Value=RigVMVariableNode() {variable=swingPickaxe}.Value).Result
    B=RigVMVariableNode() {variable=Crouching}.Value
   )
   > &RigUnit_ModifyTransforms(
    ItemToModify="((Item=(Type=Bone,Name=\"base_skirt_sd_l\"),Transform=(Rotation=(X=-0.0819,Y=-0.0298,Z=0.3407,W=0.9361),Translation=(X=-8,Y=2,Z=-1),Scale3D=(X=1,Y=1,Z=1))),(Item=(Type=Bone,Name=\"base_skirt_fr_c\"),Transform=(Rotation=(X=-0.6428,Y=0,Z=0,W=0.766),Translation=( ...(truncated)"
    WeightMinimum=0
    Weight=@Multiply_15.Result
   )
   > DeltaRotation_1 = RigVMFunctionReferenceNode(
    Rotation Order=ZYX
    child=(Type=Bone,Name="spine_05")
    debugRotation=false
    onlyTwist=true
    outputMinMax=(X=0,Y=1)
    parent=(Type=Bone,Name="pelvis")
    premult=1
    rotationMinMax=(X=20,Y=12)
    twistAxis=(X=1,Y=0,Z=0)
   )
   Multiply_13 = RigVMAggregateNode(A=@DeltaRotation_1.outputWeight, B=RigVMVariableNode() {variable=isTargeting lerp}.Value, C=RigVMVariableNode() {variable=Crouching}.Value)
   n25 = &RigVMFunction_AlphaInterp(InterpSpeedDecreasing=3, InterpSpeedIncreasing=1.5, Scale=1, bClampResult=False, Value=@Multiply_13.Result)
   > &RigUnit_ModifyTransforms(
    ItemToModify="((Item=(Type=Bone,Name=\"base_skirt_fr_l\"),Transform=(Rotation=(X=0,Y=0,Z=0.1305,W=0.9914),Translation=(X=0,Y=0,Z=0),Scale3D=(X=1,Y=1,Z=1))),(Item=(Type=Bone,Name=\"base_skirt_fr_c\"),Transform=(Rotation=(X=0,Y=0,Z=0,W=1),Translation=(X=0,Y=0.5,Z=0),Scale3D=( ...(truncated)"
    WeightMinimum=0
    Weight=@n25.Result
   )
   > DeltaRotation_2 = RigVMFunctionReferenceNode(
    Rotation Order=ZYX
    child=(Type=Bone,Name="spine_05")
    debugRotation=false
    onlyTwist=true
    outputMinMax=(X=0,Y=1)
    parent=(Type=Bone,Name="pelvis")
    premult=1
    rotationMinMax=(X=30,Y=20)
    twistAxis=(X=1,Y=0,Z=0)
   )
   Multiply_16 = RigVMAggregateNode(A=@DeltaRotation_2.outputWeight, B=RigVMVariableNode() {variable=isTargeting lerp}.Value, C=RigVMVariableNode() {variable=Crouching}.Value)
   n26 = &RigVMFunction_AlphaInterp(InterpSpeedDecreasing=3, InterpSpeedIncreasing=1.5, Scale=1, bClampResult=False, Value=@Multiply_16.Result)
   > &RigUnit_ModifyTransforms(
    ItemToModify="((Item=(Type=Bone,Name=\"base_skirt_fr_r\"),Transform=(Rotation=(X=0,Y=0,Z=-0.1736,W=0.9848),Translation=(X=0,Y=0,Z=0),Scale3D=(X=1,Y=1,Z=1))),(Item=(Type=Bone,Name=\"base_skirt_fr_c\"),Transform=(Rotation=(X=-0.0872,Y=0,Z=0,W=0.9962),Translation=(X=0,Y=-1,Z=0 ...(truncated)"
    WeightMinimum=0
    Weight=@n26.Result
   )
   Multiply_11 = RigVMAggregateNode(B=1, A=RigVMVariableNode() {variable=Leg_Front_R}.Value, C=RigVMVariableNode() {variable=Crouching}.Value)
   n27 = &RigVMFunction_AnimEvalRichCurve(
    Curve=(EditorCurveData=(Keys=((),(InterpMode=RCIM_Cubic,Time=0.2661,Value=0.4716,ArriveTangent=1,LeaveTangent=1),(Time=1,Value=1))))
    SourceMinimum=0
    TargetMinimum=0
    Value=@Multiply_11.Result
   )
   &RigUnit_ModifyTransforms(
    ItemToModify="((Item=(Type=Bone,Name=\"base_skirt_fr_r\"),Transform=(Rotation=(X=0,Y=0.1564,Z=0,W=0.9877),Translation=(X=0,Y=0,Z=0),Scale3D=(X=1,Y=1,Z=1))),(Item=(Type=Bone,Name=\"base_skirt_fr_l\"),Transform=(Rotation=(X=0,Y=-0.0175,Z=0,W=0.9998),Translation=(X=0,Y=0,Z=0), ...(truncated)"
    WeightMinimum=0
    Weight=@n27.Result
    ExecutePin=@RigVMFunction_Sequence.F
   )
   > &RigUnit_ModifyTransforms(
    ItemToModify="((Item=(Type=Bone,Name=\"base_skirt_sd_l\"),Transform=(Rotation=(X=0,Y=0.0872,Z=0,W=0.9962),Translation=(X=0,Y=0,Z=0),Scale3D=(X=1,Y=1,Z=1))),(Item=(Type=Bone,Name=\"base_skirt_fr_l\"),Transform=(Rotation=(X=-0.101,Y=0.2574,Z=0.0271,W=0.9606),Translation=(X=0, ...(truncated)"
    WeightMinimum=0
    Weight=RigVMVariableNode() {variable=Baseball Sliding}.Value
   )
   > &RigUnit_ModifyTransforms(
    ItemToModify="((Item=(Type=Bone,Name=\"base_skirt_fr_l\"),Transform=(Rotation=(X=0,Y=-0.0436,Z=0,W=0.999),Translation=(X=0,Y=0,Z=0),Scale3D=(X=1,Y=1,Z=1))),(Item=(Type=Bone,Name=\"base_skirt_sd_l\"),Transform=(Rotation=(X=0,Y=-0.061,Z=0,W=0.9981),Translation=(X=0,Y=0,Z=0),S ...(truncated)"
    WeightMinimum=0
    Weight=&RigVMFunction_MathDoubleRemap(SourceMaximum=1, SourceMinimum=0, TargetMaximum=0, TargetMinimum=1, bClamp=False, Value=RigVMVariableNode() {variable=Crouching}.Value).Result
   )
   &RigUnit_ModifyTransforms(
    ExecutePin=()
    ItemToModify="((Item=(Type=Bone,Name=\"base_skirt_fr_r\"),Transform=(Rotation=(X=0,Y=0,Z=0,W=1),Translation=(X=0,Y=0,Z=0),Scale3D=(X=1,Y=1,Z=1))),(Item=(Type=Bone,Name=\"base_skirt_fr_l\"),Transform=(Rotation=(X=-0.0507,Y=-0.5507,Z=0.0749,W=0.8298),Translation=(X=0,Y=0,Z=0) ...(truncated)"
    WeightMinimum=0
    Weight=RigVMVariableNode() {variable=Leg_FrontInside_R}.Value
   )
   &RigUnit_ModifyTransforms(
    ExecutePin=()
    ItemToModify="((Item=(Type=Bone,Name=\"base_skirt_fr_r\"),Transform=(Rotation=(X=0.1395,Y=-0.6281,Z=-0.4053,W=0.6495),Translation=(X=-12.0393,Y=4.9147,Z=-3.3198),Scale3D=(X=0.8659,Y=1,Z=1))),(Item=(Type=Bone,Name=\"base_skirt_sd_r\"),Transform=(Rotation=(X=0.0193,Y=0.172,Z= ...(truncated)"
    Weight=1
    WeightMinimum=0
   )
   RigVMVariableNode() {variable=Leg_Front_L}
  ]}
 RigVMFunctionLibrary {nodes [
   DeltaRotation = RigVMCollapseNode(
    ExecuteContext=()
    Rotation Order=ZYX
    child=(Type=Bone,Name="None")
    outputMinMax=(X=0,Y=1)
    parent=(Type=Bone,Name="None")
    premult=1
    rotationMinMax=(X=0,Y=0)
    twistAxis=(X=1,Y=0,Z=0)
   )
   LoinCloth_Fixes = RigVMCollapseNode(ExecuteContext=())
   Loincloth_Setup = RigVMCollapseNode(ExecuteContext=())
   SideArmor_Straps = RigVMCollapseNode(ExecuteContext=())
   Surface_Constraints = RigVMCollapseNode(ExecuteContext=())
   Variable = RigVMCollapseNode(ExecuteContext=())
  ]}
 RigVMFunction_MathBoolOr_1_ContainedGraph {nodes [
   n1 = RigVMFunctionEntryNode()
   RigVMFunction_MathBoolOr = &RigVMFunction_MathBoolOr(A=@n1.A, B=@n1.B)
   RigVMFunctionReturnNode(Result=&RigVMFunction_MathBoolOr(B=@n1.C, A=@RigVMFunction_MathBoolOr.Result).Result)
  ]}
 RigVMFunction_Sequence_1_ContainedGraph {nodes [
   n1 = RigVMFunctionEntryNode()
   RigVMFunction_Sequence = &RigVMFunction_Sequence(ExecuteContext=@n1.ExecuteContext)
   n2 = &RigVMFunction_Sequence(ExecuteContext=@RigVMFunction_Sequence.B)
   RigVMFunctionReturnNode(A=@RigVMFunction_Sequence.A, B=@n2.A, C=@n2.B)
  ]}
 RigVMFunction_Sequence_1_ContainedGraph {nodes [
   n1 = RigVMFunctionEntryNode()
   RigVMFunction_Sequence = &RigVMFunction_Sequence(ExecuteContext=@n1.ExecuteContext)
   n2 = &RigVMFunction_Sequence(ExecuteContext=@RigVMFunction_Sequence.B)
   n3 = &RigVMFunction_Sequence(ExecuteContext=@n2.B)
   RigVMFunctionReturnNode(A=@RigVMFunction_Sequence.A, B=@n2.A, C=@n3.A, D=@n3.B)
  ]}
 RigVMFunction_Sequence_1_ContainedGraph {nodes [
   n1 = RigVMFunctionEntryNode()
   RigVMFunction_Sequence = &RigVMFunction_Sequence(ExecuteContext=@n1.ExecuteContext)
   n2 = &RigVMFunction_Sequence(ExecuteContext=@RigVMFunction_Sequence.B)
   n3 = &RigVMFunction_Sequence(ExecuteContext=@n2.B)
   RigVMFunctionReturnNode(A=@RigVMFunction_Sequence.A, B=@n2.A, C=@n3.A, D=@n3.B)
  ]}
 RigVMFunction_Sequence_1_ContainedGraph {nodes [
   n1 = RigVMFunctionEntryNode()
   RigVMFunction_Sequence = &RigVMFunction_Sequence(ExecuteContext=@n1.ExecuteContext)
   n2 = &RigVMFunction_Sequence(ExecuteContext=@RigVMFunction_Sequence.B)
   RigVMFunctionReturnNode(A=@RigVMFunction_Sequence.A, B=@n2.A, C=@n2.B)
  ]}
 RigVMFunction_Sequence_1_ContainedGraph {nodes [
   n1 = RigVMFunctionEntryNode()
   RigVMFunction_Sequence = &RigVMFunction_Sequence(ExecuteContext=@n1.ExecuteContext)
   n2 = &RigVMFunction_Sequence(ExecuteContext=@RigVMFunction_Sequence.B)
   n3 = &RigVMFunction_Sequence(ExecuteContext=@n2.B)
   n4 = &RigVMFunction_Sequence(ExecuteContext=@n3.B)
   n5 = &RigVMFunction_Sequence(ExecuteContext=@n4.B)
   n6 = &RigVMFunction_Sequence(ExecuteContext=@n5.B)
   RigVMFunctionReturnNode(A=@RigVMFunction_Sequence.A, B=@n2.A, C=@n3.A, D=@n4.A, E=@n5.A, F=@n6.A, G=@n6.B)
  ]}
 RigVMModel {nodes [
   n1 = &RigUnit_PrepareForExecution()
   addFKCtrl = RigVMFunctionReferenceNode(
    Custom_FKCtrl=()
    Options_FKCtrl="(pelvis_1_#1=false,spine_3_#2=false,clavicle_5_#3=true,arm_7_#4=true,leg_9_#5=true,head_11_46DC15EF465CD29 ...(truncated)"
    Shape_Settings="(pelvis_2_#6=(bVisible=True,Name=\"Hexagon_Thin\",Color=(R=0.0117,G=0.8281,B=1,A=1),Transform=(Rotation=(X=0,Y=-0.7071,Z=0,W=0.7071),Translation=(X=2,Y=0,Z=0),Scale3D=(X=5,Y=5,Z=7))),spine_4_#7=(bVi ...(truncated)"
    ExecuteContext=@n1.ExecutePin
   )
   RigUnit_BeginExecution = &RigUnit_BeginExecution()
   FN_DeformRig_FNC = RigVMFunctionReferenceNode(
    DeformRig_ Curve="(deformRig_l_upperarm_up_alpha_72_#8=true,deformRig_r_upperarm_up_alpha_73_#9=true,deformRig_l_upperarm_fr_alpha_74_#10=true,deformRig_r_upperarm_fr_alpha_75_CB51B80C4 ...(truncated)"
    DeformRig_Bool="(upperarm_1_#11=True,elbow_3_#12=True,wrist_7_#13=True,groin_8_#14=True,glute_17_#15=True,knee_9_40C408324CA261 ...(truncated)"
    DeformRig_Debug=(upperarm_debug_18_#11=False,wrist_debug_20_#13=False,lat_debug_26_#16=False,groinGlute_debug_27_#14=False)
    DeformRig_Value="(upperarm_up_18_#17=(Rotation=(X=0,Y=0,Z=0,W=1),Translation=(X=0,Y=0,Z=8),Scale3D=(X=0.7,Y=1,Z=1)),upperarm_fwd_fr_19_#18=(Rotation=(X=0,Y=0,Z=0,W=1),Translation=(X=0,Y=10,Z=0),Scale3D=(X=0.5,Y=1,Z= ...(truncated)"
    ExecuteContext=@RigUnit_BeginExecution.ExecutePin
   )
   > Surface_Constraints = RigVMFunctionReferenceNode()
   > Variable = RigVMFunctionReferenceNode()
   > Loincloth_Setup = RigVMFunctionReferenceNode()
   > SideArmor_Straps = RigVMFunctionReferenceNode()
   Arm muscles = RigVMCollapseNode(ExecutePin=@SideArmor_Straps.ExecuteContext)
   setFKCtrl = RigVMFunctionReferenceNode(ExecuteContext=(), setFK_Ctrl=false)
  ]}
]
hierarchy [
 root {loc=<0 0 0>, rot=<0 0 0 1>}
 pelvis {parent=root, loc=<0 -0.7071 93.1191>, rot=<0.0127 0.707 -0.0127 -0.707>}
 spine_01 {parent=pelvis, loc=<6.7494 -0.6388 0>, rot=<0 0 0.034 -0.9994>}
 spine_02 {parent=spine_01, loc=<7.2518 -0.035 0>, rot=<0 0 0.0238 -0.9997>}
 spine_03 {parent=spine_02, loc=<6.0104 0 0>, rot=<0 0 0.0149 -0.9999>}
 spine_04 {parent=spine_03, loc=<8.1024 0 0>, rot=<0 0 -0.0132 -0.9999>}
 spine_05 {parent=spine_04, loc=<10.0146 0 0>, rot=<0 0 -0.0015 -1>}
 neck_01 {parent=spine_05, loc=<11.3138 0.1529 0>, rot=<0 0 -0.2003 -0.9797>}
 neck_02 {parent=neck_01, loc=<4.5627 0 0>, rot=<0 0 0.0504 -0.9987>}
 head {parent=neck_02, loc=<5.3914 0 0>, rot=<0 0 0.115 -0.9934>}
 head_null {parent=neck_02, type=Null, loc=<5.3914 0 0>, rot=<0 0 0.115 -0.9934>}
 attach_cape {parent=spine_05, loc=<9.017 -11.714 0>, rot=<0.0616 -0.7044 -0.0616 -0.7044>}
 cape {parent=attach_cape, type=Null, loc=<0 -3.9677 -0.0162>, rot=<0.0371 0 0 0.9993>}
 attach_backpack {parent=spine_05, loc=<-0.123 -13.473 0>, rot=<-0.0093 -0.707 0.0093 -0.707>}
 Backpack_BR {parent=attach_backpack, type=Null, loc=<0 -2.8456 0.0051>, rot=<-0.015 0 0 0.9999>}
 clavicle_l {parent=spine_05, loc=<4.3575 -0.1466 -2.3223>, rot=<0.0465 -0.7289 -0.009 -0.6829>}
 upperarm_l {parent=clavicle_l, loc=<15.5111 0 0>, rot=<0.0226 -0.3234 -0.0088 -0.9459>}
 lowerarm_l {parent=upperarm_l, loc=<26.6514 0 0>, rot=<0.0391 -0.0099 -0.2453 -0.9686>}
 hand_l {parent=lowerarm_l, loc=<22.3491 0 0>, rot=<-0.7081 -0.0507 0.0975 -0.6975>}
 thumb_01_l {parent=hand_l, loc=<1.3118 -0.3191 -2.459>, rot=<0.6546 -0.4042 0.0664 -0.6354>}
 thumb_02_l {parent=thumb_01_l, loc=<5.9336 -0.0448 0.0401>, rot=<-0.0037 0.0213 0.298 -0.9543>}
 thumb_03_l {parent=thumb_02_l, loc=<2.2223 0.3682 -0.0907>, rot=<0.0118 -0.013 0.0066 -0.9998>}
 index_metacarpal_l {parent=hand_l, loc=<1.6535 0.0183 -2.96>, rot=<0.0158 -0.0124 -0.0173 -0.9996>}
 index_01_l {parent=index_metacarpal_l, loc=<8.3008 0.0357 -0.0524>, rot=<-0.0044 0.0485 0.0924 -0.9945>}
 index_02_l {parent=index_01_l, loc=<3.3587 -0.3004 0.0452>, rot=<-0.002 0.0027 0.3225 -0.9466>}
 index_03_l {parent=index_02_l, loc=<2.5578 -0.123 -0.0252>, rot=<-0.002 0.0067 0.2558 -0.9667>}
 middle_metacarpal_l {parent=hand_l, loc=<2.1856 -0.1725 -0.5984>, rot=<0.006 0.0168 -0.0329 -0.9993>}
 middle_01_l {parent=middle_metacarpal_l, loc=<7.3275 -0.2447 0.0129>, rot=<-0.0208 0.0042 0.0858 -0.9961>}
 middle_02_l {parent=middle_01_l, loc=<4.0123 0.1401 0.1175>, rot=<-0.0218 0.0258 0.3512 -0.9357>}
 middle_03_l {parent=middle_02_l, loc=<2.6077 0.0534 0.1373>, rot=<0.0082 0.0188 0.1708 -0.9851>}
 ring_metacarpal_l {parent=hand_l, loc=<2.2074 -0.3625 1.3638>, rot=<-0.0284 0.0722 -0.0458 -0.9959>}
 ring_01_l {parent=ring_metacarpal_l, loc=<7.0295 -0.6197 -0.1446>, rot=<-0.0489 -0.0252 0.1892 -0.9804>}
 ring_02_l {parent=ring_01_l, loc=<3.6174 0.4647 -0.1027>, rot=<0 0 0.2806 -0.9598>}
 ring_03_l {parent=ring_02_l, loc=<2.4152 0.114 0.0199>, rot=<0 0 0.1189 -0.9929>}
 pinky_metacarpal_l {parent=hand_l, loc=<1.7616 -0.9753 3.3814>, rot=<-0.0397 0.0814 -0.069 -0.9935>}
 pinky_01_l {parent=pinky_metacarpal_l, loc=<7.1709 -0.4831 -0.1094>, rot=<-0.0654 -0.0035 0.2242 -0.9723>}
 pinky_02_l {parent=pinky_01_l, loc=<2.8274 0.0893 -0.1043>, rot=<0.0037 0.0011 0.2854 -0.9584>}
 pinky_03_l {parent=pinky_02_l, loc=<2.2046 0.0307 -0.0078>, rot=<-0.0002 0.0098 0.1445 -0.9895>}
 deform_wrist_dn_base_l {parent=hand_l, loc=<0 -2.2342 0>, rot=<0 0 0 -1>}
 deform_wrist_dn_l {parent=deform_wrist_dn_base_l, loc=<0 0 0>, rot=<0 0 0 1>}
 deform_wrist_up_base_l {parent=hand_l, loc=<0 2.2682 0>, rot=<0 0 0 -1>}
 deform_wrist_up_l {parent=deform_wrist_up_base_l, loc=<0 0 0>, rot=<0 0 0 1>}
 lowerarm_twist_02_l {parent=lowerarm_l, loc=<9.6199 0 0>, rot=<0 0 0 -1>}
 lowerarm_twist_01_l {parent=lowerarm_l, loc=<17.8549 0 0>, rot=<0 0 0 -1>}
 deform_elbow_in_base_l {parent=lowerarm_l, loc=<0.2112 5.259 0.8274>, rot=<0 0 0 1>}
 deform_elbow_in_l {parent=deform_elbow_in_base_l, loc=<0 0 0>, rot=<0 0 0 1>}
 deform_elbow_out_base_l {parent=lowerarm_l, loc=<0 -4 0>, rot=<0 0 0 1>}
 deform_elbow_out_l {parent=deform_elbow_out_base_l, loc=<0 0 0>, rot=<0 0 0 1>}
 upperarm_twist_01_l {parent=upperarm_l, loc=<0.0041 0 0>, rot=<0 0 0 -1>}
 upperarm_twist_02_l {parent=upperarm_l, loc=<13.7941 0 0>, rot=<0 0 0 -1>}
 upperarm_bicep_l_base {parent=upperarm_twist_02_l, loc=<0.1682 4.9626 0>, rot=<0 0 -0.695 -0.719>}
 upperarm_bicep_l {parent=upperarm_bicep_l_base, loc=<0 0 0>, rot=<0 0 0 1>}
 upperarm_tricep_l_base {parent=upperarm_twist_02_l, loc=<0.2941 -6.3929 0>, rot=<0 0 -0.7232 -0.6907>}
 upperarm_tricep_l {parent=upperarm_tricep_l_base, loc=<0 0 0>, rot=<0 0 0 1>}
 upperarm_l_null {parent=clavicle_l, type=Null, loc=<15.5111 0 0>, rot=<0.0226 -0.3234 -0.0088 -0.9459>}
 clavicle_r {parent=spine_05, loc=<4.3575 -0.1466 2.3223>, rot=<0.7289 0.0465 -0.6829 0.009>}
 upperarm_r {parent=clavicle_r, loc=<-15.5111 0 0>, rot=<0.0226 -0.3234 -0.0088 -0.9459>}
 lowerarm_r {parent=upperarm_r, loc=<-26.6514 0 0>, rot=<0.0391 -0.0099 -0.2453 -0.9686>}
 hand_r {parent=lowerarm_r, loc=<-22.3491 0 0>, rot=<-0.7081 -0.0507 0.0975 -0.6975>}
 thumb_01_r {parent=hand_r, loc=<-1.3118 0.3191 2.459>, rot=<0.6546 -0.4042 0.0664 -0.6354>}
 thumb_02_r {parent=thumb_01_r, loc=<-5.9336 0.0448 -0.0401>, rot=<-0.0037 0.0213 0.298 -0.9543>}
 thumb_03_r {parent=thumb_02_r, loc=<-2.2223 -0.3682 0.0907>, rot=<0.0118 -0.013 0.0066 -0.9998>}
 index_metacarpal_r {parent=hand_r, loc=<-1.6535 -0.0183 2.96>, rot=<0.0158 -0.0124 -0.0173 -0.9996>}
 index_01_r {parent=index_metacarpal_r, loc=<-8.3008 -0.0357 0.0524>, rot=<-0.0044 0.0485 0.0924 -0.9945>}
 index_02_r {parent=index_01_r, loc=<-3.3587 0.3004 -0.0452>, rot=<-0.002 0.0027 0.3225 -0.9466>}
 index_03_r {parent=index_02_r, loc=<-2.5578 0.123 0.0252>, rot=<-0.002 0.0067 0.2558 -0.9667>}
 middle_metacarpal_r {parent=hand_r, loc=<-2.1856 0.1725 0.5984>, rot=<0.006 0.0168 -0.0329 -0.9993>}
 middle_01_r {parent=middle_metacarpal_r, loc=<-7.3275 0.2447 -0.0129>, rot=<-0.0208 0.0042 0.0858 -0.9961>}
 middle_02_r {parent=middle_01_r, loc=<-4.0123 -0.1401 -0.1175>, rot=<-0.0218 0.0258 0.3512 -0.9357>}
 middle_03_r {parent=middle_02_r, loc=<-2.6077 -0.0534 -0.1373>, rot=<0.0082 0.0188 0.1708 -0.9851>}
 ring_metacarpal_r {parent=hand_r, loc=<-2.2074 0.3625 -1.3638>, rot=<-0.0284 0.0722 -0.0458 -0.9959>}
 ring_01_r {parent=ring_metacarpal_r, loc=<-7.0295 0.6197 0.1446>, rot=<-0.0489 -0.0252 0.1892 -0.9804>}
 ring_02_r {parent=ring_01_r, loc=<-3.6174 -0.4647 0.1027>, rot=<0 0 0.2806 -0.9598>}
 ring_03_r {parent=ring_02_r, loc=<-2.4152 -0.114 -0.0199>, rot=<0 0 0.1189 -0.9929>}
 pinky_metacarpal_r {parent=hand_r, loc=<-1.7616 0.9753 -3.3814>, rot=<-0.0397 0.0814 -0.069 -0.9935>}
 pinky_01_r {parent=pinky_metacarpal_r, loc=<-7.1709 0.4831 0.1094>, rot=<-0.0654 -0.0035 0.2242 -0.9723>}
 pinky_02_r {parent=pinky_01_r, loc=<-2.8274 -0.0893 0.1043>, rot=<0.0037 0.0011 0.2854 -0.9584>}
 pinky_03_r {parent=pinky_02_r, loc=<-2.2046 -0.0307 0.0078>, rot=<-0.0002 0.0098 0.1445 -0.9895>}
 lowerarm_twist_02_r {parent=lowerarm_r, loc=<-9.6199 0 0>, rot=<0 0 0 -1>}
 lowerarm_twist_01_r {parent=lowerarm_r, loc=<-17.8549 0 0>, rot=<0 0 0 -1>}
 deform_elbow_in_base_r {parent=lowerarm_r, loc=<-0.2111 -5.2591 -0.8275>, rot=<0 0 0 -1>}
 deform_elbow_in_r {parent=deform_elbow_in_base_r, loc=<-0.0001 0 0.0002>, rot=<0 0 0 1>}
 deform_elbow_out_base_r {parent=lowerarm_r, loc=<0 4 -0.0001>, rot=<0 0 0 -1>}
 deform_elbow_out_r {parent=deform_elbow_out_base_r, loc=<0 0 0>, rot=<0 0 0 1>}
 upperarm_twist_01_r {parent=upperarm_r, loc=<-0.0041 0 0>, rot=<0 0 0 -1>}
 deform_upperarm_bk_base_r {parent=upperarm_twist_01_r, loc=<-1.0814 7.2192 1.9455>, rot=<0 0 0 1>}
 deform_upperarm_bk_r {parent=deform_upperarm_bk_base_r, loc=<0 0 0>, rot=<0 0 0 1>}
 deform_upperarm_fr_base_r {parent=upperarm_twist_01_r, loc=<0.3689 -7.0506 2.7652>, rot=<0 0 0 1>}
 deform_upperarm_fr_r {parent=deform_upperarm_fr_base_r, loc=<0 0 0>, rot=<0 0 0 1>}
 deform_upperarm_up_base_r {parent=upperarm_twist_01_r, loc=<3.2214 1.1391 -3.3308>, rot=<0 0 0 1>}
 deform_upperarm_up_r {parent=deform_upperarm_up_base_r, loc=<0 0 0>, rot=<0 0 0 -1>}
 upperarm_twist_02_r {parent=upperarm_r, loc=<-13.7941 0 0>, rot=<0 0 0 -1>}
 upperarm_bicep_r_base {parent=upperarm_twist_02_r, loc=<-0.1682 -4.9626 0>, rot=<0 0 0.719 -0.695>}
 upperarm_bicep_r {parent=upperarm_bicep_r_base, loc=<0 0 0>, rot=<0 0 0 1>}
 upperarm_tricep_r_base {parent=upperarm_twist_02_r, loc=<-0.2941 6.3929 0>, rot=<0 0 0.6907 -0.7232>}
 upperarm_tricep_r {parent=upperarm_tricep_r_base, loc=<0 0 0>, rot=<0 0 0 1>}
 upperarm_r_helper {parent=upperarm_r, loc=<0 0 0>, rot=<0 0 0 1>}
 upperarm_r_null {parent=clavicle_r, type=Null, loc=<-15.5111 0 0>, rot=<0.0226 -0.3234 -0.0088 -0.9459>}
 clavicle_r_helper {parent=clavicle_r, loc=<0 0 0>, rot=<0 0 0 1>}
 base_SideArmor_r {parent=spine_05, loc=<13.5441 -2.5519 12.4701>, rot=<-0.1907 -0.0077 0.9808 -0.0394>}
 dyn_SideArmor_r {parent=base_SideArmor_r, loc=<0 0 0>, rot=<0 0 0 1>}
 base_SideArmor_l {parent=spine_05, loc=<13.5445 -2.5519 -12.4701>, rot=<0.0394 -0.9808 0.0077 0.1907>}
 dyn_SideArmor_l {parent=base_SideArmor_l, loc=<0 0 0>, rot=<0 0 0 1>}
 clavicle_l_null {parent=spine_05, type=Null, loc=<4.3575 -0.1466 -2.3223>, rot=<0.0465 -0.7289 -0.009 -0.6829>}
 clavicle_r_null {parent=spine_05, type=Null, loc=<4.3575 -0.1466 2.3223>, rot=<0.7289 0.0465 -0.6829 0.009>}
 base_SideArmor_l_front {parent=spine_05, loc=<6.7605 9.328 -9.4129>, rot=<0.3859 -0.9079 -0.1175 0.1137>}
 dyn_SideArmor_l_front {parent=base_SideArmor_l_front, loc=<0 0 0>, rot=<0 0 0 1>}
 base_SideArmor_l_rear {parent=spine_05, loc=<10.8735 -9.497 -11.1411>, rot=<-0.3952 -0.892 0.108 0.191>}
 dyn_SideArmor_l_rear {parent=base_SideArmor_l_rear, loc=<0 0 0>, rot=<0 0 0 1>}
 base_SideArmor_r_front {parent=spine_05, loc=<6.7602 9.328 9.4129>, rot=<0.3859 -0.9079 0.1175 -0.1137>}
 dyn_SideArmor_r_front {parent=base_SideArmor_r_front, loc=<0 0 0>, rot=<0 0 0 1>}
 base_SideArmor_r_rear {parent=spine_05, loc=<10.873 -9.4971 11.1411>, rot=<-0.3952 -0.892 -0.108 -0.191>}
 dyn_SideArmor_r_rear {parent=base_SideArmor_r_rear, loc=<0 0 0>, rot=<0 0 0 1>}
 base_deform_r_back_helper {parent=spine_04, loc=<1.5968 -6.9834 18.9066>, rot=<-0.1493 0.7944 0.0362 -0.5876>}
 deform_r_back_helper {parent=base_deform_r_back_helper, loc=<0 0 0>, rot=<0 0 0 1>}
 deform_pec_aim_base_l {parent=spine_04, loc=<7.8732 12.6388 -2.4123>, rot=<0.0102 -0.4602 0.3979 -0.7936>}
 deform_pec_aim_l {parent=deform_pec_aim_base_l, loc=<0 0 0>, rot=<0 0 0 1>}
 deform_pec_base_l {parent=deform_pec_aim_l, loc=<10.766 0 0>, rot=<0 0 0 1>}
 deform_pec_l {parent=deform_pec_base_l, loc=<0.055 0 0>, rot=<0 0 0 1>}
 deform_pec_aim_base_r {parent=spine_04, loc=<7.8727 12.6388 2.4123>, rot=<-0.4602 -0.0102 0.7936 0.3979>}
 deform_pec_aim_r {parent=deform_pec_aim_base_r, loc=<0 0 0>, rot=<0 0 0 1>}
 deform_pec_base_r {parent=deform_pec_aim_r, loc=<-10.7662 -0.0006 -0.0007>, rot=<0 0 0 1>}
 deform_pec_r {parent=deform_pec_base_r, loc=<0 0 0>, rot=<0 0 0 1>}
 thigh_l {parent=pelvis, loc=<-2.5693 -0.0109 -7.7501>, rot=<0.0012 0.0572 0.0155 -0.9982>}
 calf_l {parent=thigh_l, loc=<-39.0484 0 0>, rot=<0.0006 -0.0161 -0.0391 -0.9991>}
 foot_l {parent=calf_l, loc=<-42.561 0.1077 0.4292>, rot=<0.0418 -0.0455 0.0343 -0.9975>}
 ball_l {parent=foot_l, loc=<-6.1599 13.8264 0.6806>, rot=<0.0001 -0.0001 -0.7063 -0.7079>}
 calf_twist_02_l {parent=calf_l, loc=<-18.8064 -0.0967 0.0397>, rot=<0.0237 0.0095 -0.0035 -0.9997>}
 calf_twist_01_l {parent=calf_l, loc=<-35.813 -0.0967 0.0397>, rot=<0.0237 0.0095 -0.0035 -0.9997>}
 deform_knee_in_base_l {parent=calf_l, loc=<1.298 -5.2957 -0.0496>, rot=<0 0 0 -1>}
 deform_knee_in_l {parent=deform_knee_in_base_l, loc=<0 0 0>, rot=<0 0 0 1>}
 Base_l_leg_outer {parent=calf_l, loc=<-0.1075 -0.3539 -5.4126>, rot=<0.0004 0.6962 0.059 -0.7154>}
 dyn_l_leg_outer {parent=Base_l_leg_outer, loc=<0 0 0>, rot=<0 0 0 1>}
 Base_l_leg_inner {parent=calf_l, loc=<-1.0857 -0.0784 6.0156>, rot=<0.7742 -0.0553 -0.6305 -0.0077>}
 dyn_l_leg_inner {parent=Base_l_leg_inner, loc=<0 0 0>, rot=<0 0 0 1>}
 thigh_twist_01_l {parent=thigh_l, loc=<-21.6303 0 0>, rot=<0.0013 0 0 -1>}
 thigh_hamstrings_l_base {parent=thigh_twist_01_l, loc=<0 -6.6112 0>, rot=<0 0 -0.7071 -0.7071>}
 thigh_hamstrings_l {parent=thigh_hamstrings_l_base, loc=<0 0 0>, rot=<0 0 0 1>}
 thigh_quads_l_base {parent=thigh_twist_01_l, loc=<0 6.0114 0>, rot=<0 0 -0.7071 -0.7071>}
 thigh_quads_l {parent=thigh_quads_l_base, loc=<0 0 0>, rot=<0 0 0 1>}
 thigh_l_helper {parent=thigh_l, loc=<0 0 0>, rot=<0 0 0 1>}
 thigh_r {parent=pelvis, loc=<-2.5693 -0.0109 7.7501>, rot=<0.0572 -0.0012 0.9982 0.0155>}
 calf_r {parent=thigh_r, loc=<39.0484 0 0>, rot=<0.0006 -0.0161 -0.0391 -0.9991>}
 foot_r {parent=calf_r, loc=<42.561 -0.1077 -0.4292>, rot=<0.0418 -0.0455 0.0343 -0.9975>}
 ball_r {parent=foot_r, loc=<6.1599 -13.8264 -0.6806>, rot=<0.0001 -0.0001 -0.7063 -0.7079>}
 calf_twist_02_r {parent=calf_r, loc=<18.8064 0.0967 -0.0397>, rot=<0.0237 0.0095 -0.0035 -0.9997>}
 calf_twist_01_r {parent=calf_r, loc=<35.813 0.0967 -0.0397>, rot=<0.0237 0.0095 -0.0035 -0.9997>}
 deform_knee_in_base_r {parent=calf_r, loc=<-1.298 5.2957 0.0496>, rot=<0 0 0 -1>}
 deform_knee_in_r {parent=deform_knee_in_base_r, loc=<0 0 0>, rot=<0 0 0 1>}
 Base_r_leg_outer {parent=calf_r, loc=<0.1075 0.3539 5.4126>, rot=<-0.059 -0.7154 0.0004 -0.6962>}
 dyn_r_leg_outer {parent=Base_r_leg_outer, loc=<0 0 0>, rot=<0 0 0 1>}
 Base_r_leg_inner {parent=calf_r, loc=<1.0857 0.0784 -6.0156>, rot=<0.6305 -0.0077 0.7742 0.0553>}
 dyn_r_leg_inner {parent=Base_r_leg_inner, loc=<0 0 0>, rot=<0 0 0 1>}
 thigh_twist_01_r {parent=thigh_r, loc=<21.6303 0 0>, rot=<0.0013 0 0 -1>}
 thigh_quads_r_base {parent=thigh_twist_01_r, loc=<0 -6.0114 0>, rot=<0 0 0.7071 -0.7071>}
 thigh_quads_r {parent=thigh_quads_r_base, loc=<0 0 0>, rot=<0 0 0 1>}
 thigh_hamstrings_r_base {parent=thigh_twist_01_r, loc=<0 6.6112 0>, rot=<0 0 0.7071 -0.7071>}
 thigh_hamstrings_r {parent=thigh_hamstrings_r_base, loc=<0 0 0>, rot=<0 0 0 1>}
 thigh_r_helper {parent=thigh_r, loc=<0 0 0>, rot=<0 0 0 1>}
 thigh_l_null {parent=pelvis, type=Null, loc=<-2.5693 -0.0109 -7.7501>, rot=<0.0012 0.0572 0.0155 -0.9982>}
 thigh_r_null {parent=pelvis, type=Null, loc=<-2.5693 -0.0109 7.7501>, rot=<0.0572 -0.0012 0.9982 0.0155>}
 base_bk_01_skirt_l {parent=pelvis, loc=<-0.8221 -12.3249 -8.7248>, rot=<0.0134 -0.7036 0.7078 -0.0612>}
 dyn_bk_01_skirt_l {parent=base_bk_01_skirt_l, loc=<0 0 0>, rot=<0 0 0 1>}
 dyn_bk_02_skirt_l {parent=dyn_bk_01_skirt_l, loc=<6.403 0 0>, rot=<0 0 0 1>}
 base_bk_01_skirt_c {parent=pelvis, loc=<-3.5424 -13.7873 0>, rot=<-0.0117 -0.707 0.707 0.0117>}
 dyn_bk_01_skirt_c {parent=base_bk_01_skirt_c, loc=<0 0 0>, rot=<0 0 0 1>}
 dyn_bk_02_skirt_c {parent=dyn_bk_01_skirt_c, loc=<7.1766 0 0>, rot=<0 0 0 1>}
 base_bk_01_skirt_r {parent=pelvis, loc=<-0.8222 -12.3249 8.7248>, rot=<0.0612 -0.7078 0.7036 -0.0134>}
 dyn_bk_01_skirt_r {parent=base_bk_01_skirt_r, loc=<0 0 0>, rot=<0 0 0 1>}
 dyn_bk_02_skirt_r {parent=dyn_bk_01_skirt_r, loc=<6.403 0 0>, rot=<0 0 0 1>}
 base_skirt_fr_c {parent=pelvis, loc=<-2.6968 10.2692 0>, rot=<-0.0247 -0.7067 -0.7067 -0.0247>}
 dyn_skirt_fr_01_c {parent=base_skirt_fr_c, loc=<0 0 0>, rot=<0 0 0 1>}
 dyn_skirt_fr_02_c {parent=dyn_skirt_fr_01_c, loc=<7.2638 0 0>, rot=<0 0 0 1>}
 base_skirt_fr_l {parent=pelvis, loc=<1.9933 6.485 -9.1752>, rot=<0.0915 0.6977 0.7007 0.1182>}
 dyn_skirt_fr_01_l {parent=base_skirt_fr_l, loc=<0 0 0>, rot=<0 0 0 1>}
 dyn_skirt_fr_02_l {parent=dyn_skirt_fr_01_l, loc=<5.337 0 0>, rot=<0 0 0 1>}
 base_skirt_fr_r {parent=pelvis, loc=<1.9933 6.485 9.1752>, rot=<-0.1182 -0.7007 -0.6977 -0.0915>}
 dyn_skirt_fr_01_r {parent=base_skirt_fr_r, loc=<0 0 0>, rot=<0 0 0 1>}
 dyn_skirt_fr_02_r {parent=dyn_skirt_fr_01_r, loc=<5.337 0 0>, rot=<0 0 0 -1>}
 base_skirt_sd_l {parent=pelvis, loc=<4.2537 -1.9004 -14.1717>, rot=<-0.0195 -0.9908 0.0022 -0.1339>}
 dyn_skirt_sd_01_l {parent=base_skirt_sd_l, loc=<0 0 0>, rot=<0 0 0 1>}
 dyn_skirt_sd_02_l {parent=dyn_skirt_sd_01_l, loc=<4.6141 0 0>, rot=<0.0002 -0.0023 -0.0004 -1>}
 base_skirt_sd_r {parent=pelvis, loc=<4.2536 -1.9004 14.1717>, rot=<0.1339 -0.0022 0.9908 0.0195>}
 dyn_skirt_sd_01_r {parent=base_skirt_sd_r, loc=<0 0 0>, rot=<0 0 0 1>}
 dyn_skirt_sd_02_r {parent=dyn_skirt_sd_01_r, loc=<4.614 0 0>, rot=<0 0 0 -1>}
]
variables [
 Baseball Sliding {type=double, default=0}
 Crouching {type=double, default=0}
 CrouchingBool {type=bool, default=False}
 Idle {type=double, default=0}
 Leg_BackSide_L {type=double, default=0}
 Leg_BackSide_R {type=double, default=0}
 Leg_Back_L {type=double, default=0}
 Leg_Back_R {type=double, default=0}
 Leg_FrontInside_L {type=double, default=0}
 Leg_FrontInside_R {type=double, default=0}
 Leg_FrontSide_L {type=double, default=0}
 Leg_FrontSide_R {type=double, default=0}
 Leg_Front_L {type=double, default=0}
 Leg_Front_R {type=double, default=0}
 Leg_Side_L {type=double, default=0}
 Leg_Side_R {type=double, default=0}
 Parachute_opened {type=bool, default=False}
 SMG_P90 {type=bool, default=False}
 SPR_Clavicle_l_Back {type=double, default=0}
 SPR_Clavicle_l_Fwd {type=double, default=0}
 SPR_Clavicle_l_Up {type=double, default=0}
 SPR_Clavicle_l_Up_Parachuting {type=double, default=0}
 SPR_Clavicle_r_Back {type=double, default=0}
 SPR_Clavicle_r_Fwd {type=double, default=0}
 SPR_Clavicle_r_Up {type=double, default=0}
 SPR_Clavicle_r_Up_Parachuting {type=double, default=0}
 SPR_Leg_R_Fwd_Ext {type=double, default=0}
 Skydiving {type=double, default=0}
 isAR Equipped lerp {type=double, default=0}
 isFrontEnd {type=bool, default=False}
 isJumping {type=double, default=0}
 isPA_Swing {type=double, default=0}
 isPistol Equipped lerp {type=double, default=0}
 isTargeting lerp {type=double, default=0}
 is_OutLaw_SG {type=bool, default=False}
 onGround {type=bool, default=False}
 pub isAR Equipped {type=bool, default=False}
 pub isPistol Equipped {type=bool, default=False}
 pub isTargeting {type=bool, default=False}
 right_Upperarm {type=double, default=0}
 swingPickaxe {type=bool, default=False}
]
