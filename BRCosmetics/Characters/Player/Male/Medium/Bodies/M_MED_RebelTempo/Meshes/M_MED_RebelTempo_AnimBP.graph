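# Post-process style Anim Blueprint for the M_MED_RebelTempo player body, adding procedural secondary motion to the pose copied from the parent mesh.
# A Control Rig driven by movement, weapon and state flags feeds chained AnimDynamics for skirt bones.
# An optional stage, selected by bPbBlendEnabled, adds arm, thigh and pec jiggle plus a weighted attach_backpack constraint.
# EventGraph updates rig inputs, plus cloth-alpha, emote and backpack variables the AnimGraph never reads.
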AnimBlueprint /BRCosmetics/Characters/Player/Male/Medium/Bodies/M_MED_RebelTempo/Meshes/M_MED_RebelTempo_AnimBP
target_skeleton=&/M_MED_RebelTempo_Skeleton
graphs [
 AnimGraph {nodes [
   Output Pose(Result=Blend Poses by bool(
     BlendPose_0=Use cached pose 'Post_Body_Dynamics'()
     BlendPose_1=Use cached pose 'Post_AnimDynamics'()
     BlendTime_0=0
     BlendTime_1=0
     bActiveValue=Get bPbBlendEnabled() -> (bPbBlendEnabled:bool)
    )) {layerGroup=None}
   n1 = Control Rig!(
    Source=Copy Pose From Mesh() {bUseAttachedParent=true}
    Alpha
    Parachute_opened:bool=Get bIsParachuteOpened()
    Crouching:double=Get Crouching() -> (Crouching:double)
    swingPickaxe:bool=Get bIsPlayingMeleeAnim()
    pub isTargeting:bool=Get bIsTargeting()
    Baseball Sliding:double=Get Baseball_Sliding() -> (Baseball_Sliding:double)
    Skydiving:double="Get SkyDiving "() -> ("SkyDiving ":double)
    onGround:bool=Get bIsOnGround() -> (bIsOnGround:bool)
    isJumping:double=Get is_Jumping() -> (is_Jumping:double)
    pub isAR Equipped:bool=Get isAR Equipped() -> (isAR Equipped:bool)
    pub isPistol Equipped:bool=Get isPistol Equipped() -> (isPistol Equipped:bool)
    is_OutLaw_SG:bool=Get Is_outLaw_SG() -> (Is_outLaw_SG:bool)
    SMG_P90:bool=Get SMG_SG() -> (SMG_SG:bool)
    isFrontEnd:bool=Get bIsInFrontEnd()
   ) -> (Pose) {controlRigAssetReference {blueprintRigClass=&/M_MED_RebelTempo_CtrlRig.M_MED_RebelTempo_CtrlRig_C}}
   Save cached pose 'Control Rig cache'(Pose=@n1) {cacheName=Control Rig cache}
   n2 = Anim Dynamics - Bone: dyn_bk_01_skirt_l(ComponentPose=Local To Component(LocalPose=Use cached pose 'Control Rig cache'()), Alpha=1, ExternalForce="0,0,0") {
    angularDampingOverride=1
    angularSpringConstant=10000
    bAngularSpring=true
    bOverrideAngularDamping=true
    bUseGravityOverride=true
    gravityScale=0.35
    simulationSpace=RootRelative
    boundBone {boneName=dyn_bk_01_skirt_l}
    componentAppliedLinearAccClamp=<4000 4000 4000>
    componentLinearAccScale=<0.4 0.4 0.45>
    componentLinearVelScale=<0.4 0.4 0.45>
    gravityOverride {z=-980}
    physicsBodyDefinitions [{
      boundBone {boneName=dyn_bk_01_skirt_l}
      boxExtents=<25 35 35>
      constraintSetup {angularLimitsMax=<0 0 5>, angularLimitsMin=<0 -20 -5>, angularTarget=<1 0 0>}
      localJointOffset=<4 0 0>
     }]
    relativeSpaceBone {boneName=pelvis}
    simSpaceSettings {maxAngularAcceleration=100, maxAngularVelocity=10, simSpaceAngularAlpha=1}
   }
   n3 = Anim Dynamics - Bone: dyn_bk_01_skirt_c(ComponentPose=@n2, Alpha=1, ExternalForce="0,0,0") {
    angularDampingOverride=1
    angularSpringConstant=10000
    bAngularSpring=true
    bOverrideAngularDamping=true
    bUseGravityOverride=true
    gravityScale=0.35
    simulationSpace=RootRelative
    boundBone {boneName=dyn_bk_01_skirt_c}
    componentAppliedLinearAccClamp=<4000 4000 4000>
    componentLinearAccScale=<0.4 0.4 0.45>
    componentLinearVelScale=<0.4 0.4 0.45>
    gravityOverride {z=-980}
    physicsBodyDefinitions [{
      boundBone {boneName=dyn_bk_01_skirt_c}
      boxExtents=<25 35 35>
      constraintSetup {angularLimitsMax=<0 5 5>, angularLimitsMin=<0 -20 -5>, angularTarget=<1 0 0>}
      localJointOffset=<4 0 0>
     }]
    relativeSpaceBone {boneName=pelvis}
    simSpaceSettings {maxAngularAcceleration=100, maxAngularVelocity=10, simSpaceAngularAlpha=1}
   }
   n4 = Anim Dynamics - Bone: dyn_bk_02_skirt_c(ComponentPose=@n3, Alpha=1, ExternalForce="0,0,0") {
    angularDampingOverride=1
    angularSpringConstant=10000
    bAngularSpring=true
    bOverrideAngularDamping=true
    bUseGravityOverride=true
    gravityScale=0.35
    simulationSpace=RootRelative
    boundBone {boneName=dyn_bk_02_skirt_c}
    componentAppliedLinearAccClamp=<4000 4000 4000>
    componentLinearAccScale=<0.4 0.4 0.45>
    componentLinearVelScale=<0.4 0.4 0.45>
    gravityOverride {z=-980}
    physicsBodyDefinitions [{
      boundBone {boneName=dyn_bk_02_skirt_c}
      boxExtents=<25 35 35>
      constraintSetup {angularLimitsMax=<0 5 8>, angularLimitsMin=<0 -20 -8>, angularTarget=<1 0 0>}
      localJointOffset=<4 0 0>
     }]
    relativeSpaceBone {boneName=pelvis}
    simSpaceSettings {maxAngularAcceleration=100, maxAngularVelocity=10, simSpaceAngularAlpha=1}
   }
   n5 = Anim Dynamics - Bone: dyn_bk_01_skirt_r(ComponentPose=@n4, Alpha=1, ExternalForce="0,0,0") {
    angularDampingOverride=1
    angularSpringConstant=10000
    bAngularSpring=true
    bOverrideAngularDamping=true
    bUseGravityOverride=true
    gravityScale=0.35
    simulationSpace=RootRelative
    boundBone {boneName=dyn_bk_01_skirt_r}
    componentAppliedLinearAccClamp=<4000 4000 4000>
    componentLinearAccScale=<0.4 0.4 0.45>
    componentLinearVelScale=<0.4 0.4 0.45>
    gravityOverride {z=-980}
    physicsBodyDefinitions [{
      boundBone {boneName=dyn_bk_01_skirt_r}
      boxExtents=<25 35 35>
      constraintSetup {angularLimitsMax=<0 5 5>, angularLimitsMin=<0 -20 -5>, angularTarget=<1 0 0>}
      localJointOffset=<4 0 0>
     }]
    relativeSpaceBone {boneName=pelvis}
    simSpaceSettings {maxAngularAcceleration=100, maxAngularVelocity=10, simSpaceAngularAlpha=1}
   }
   n6 = Anim Dynamics - Bone: dyn_skirt_sd_01_r(ComponentPose=@n5, Alpha=1, ExternalForce="0,0,0") {
    angularDampingOverride=1
    angularSpringConstant=10000
    bAngularSpring=true
    bOverrideAngularDamping=true
    bUseGravityOverride=true
    gravityScale=0.35
    simulationSpace=RootRelative
    boundBone {boneName=dyn_skirt_sd_01_r}
    componentAppliedLinearAccClamp=<4000 4000 4000>
    componentLinearAccScale=<0.4 0.4 0.45>
    componentLinearVelScale=<0.4 0.4 0.45>
    gravityOverride {z=-980}
    physicsBodyDefinitions [{
      boundBone {boneName=dyn_skirt_sd_01_r}
      boxExtents=<25 35 35>
      constraintSetup {angularLimitsMax=<0 5 5>, angularLimitsMin=<0 -20 -5>, angularTarget=<1 0 0>}
      localJointOffset=<4 0 0>
     }]
    relativeSpaceBone {boneName=pelvis}
    simSpaceSettings {maxAngularAcceleration=100, maxAngularVelocity=10, simSpaceAngularAlpha=1}
   }
   n7 = Anim Dynamics - Bone: dyn_skirt_fr_01_r(ComponentPose=@n6, Alpha=1, ExternalForce="0,0,0") {
    angularDampingOverride=1
    angularSpringConstant=10000
    bAngularSpring=true
    bOverrideAngularDamping=true
    bUseGravityOverride=true
    gravityScale=0.35
    simulationSpace=RootRelative
    boundBone {boneName=dyn_skirt_fr_01_r}
    componentAppliedLinearAccClamp=<4000 4000 4000>
    componentLinearAccScale=<0.4 0.4 0.45>
    componentLinearVelScale=<0.4 0.4 0.45>
    gravityOverride {z=-980}
    physicsBodyDefinitions [{
      boundBone {boneName=dyn_skirt_fr_01_r}
      boxExtents=<25 35 35>
      constraintSetup {angularLimitsMax=<0 5 5>, angularLimitsMin=<0 -10 -5>, angularTarget=<1 0 0>}
      localJointOffset=<4 0 0>
     }]
    relativeSpaceBone {boneName=pelvis}
    simSpaceSettings {maxAngularAcceleration=100, maxAngularVelocity=10, simSpaceAngularAlpha=1}
   }
   n8 = Anim Dynamics - Bone: dyn_skirt_fr_01_c(ComponentPose=@n7, Alpha=1, ExternalForce="0,0,0") {
    angularDampingOverride=1
    angularSpringConstant=10000
    bAngularSpring=true
    bOverrideAngularDamping=true
    bUseGravityOverride=true
    gravityScale=0.35
    simulationSpace=RootRelative
    boundBone {boneName=dyn_skirt_fr_01_c}
    componentAppliedLinearAccClamp=<4000 4000 4000>
    componentLinearAccScale=<0.4 0.4 0.45>
    componentLinearVelScale=<0.4 0.4 0.45>
    gravityOverride {z=-980}
    physicsBodyDefinitions [{
      boundBone {boneName=dyn_skirt_fr_01_c}
      boxExtents=<25 35 35>
      constraintSetup {angularLimitsMax=<0 30 0>, angularLimitsMin=<0 -1 0>, angularTarget=<1 0 0>}
      localJointOffset=<4 0 0>
     }]
    relativeSpaceBone {boneName=pelvis}
    simSpaceSettings {maxAngularAcceleration=100, maxAngularVelocity=10, simSpaceAngularAlpha=1}
   }
   n9 = Anim Dynamics - Bone: dyn_skirt_fr_02_c(ComponentPose=@n8, Alpha=1, ExternalForce="0,0,0") {
    angularDampingOverride=1
    angularSpringConstant=10000
    bAngularSpring=true
    bOverrideAngularDamping=true
    bUseGravityOverride=true
    gravityScale=0.35
    simulationSpace=RootRelative
    boundBone {boneName=dyn_skirt_fr_02_c}
    componentAppliedLinearAccClamp=<4000 4000 4000>
    componentLinearAccScale=<0.4 0.4 0.45>
    componentLinearVelScale=<0.4 0.4 0.45>
    gravityOverride {z=-980}
    physicsBodyDefinitions [{
      boundBone {boneName=dyn_skirt_fr_02_c}
      boxExtents=<25 35 35>
      constraintSetup {angularLimitsMax=<0 30 0>, angularLimitsMin=<0 -10 0>, angularTarget=<1 0 0>}
      localJointOffset=<4 0 0>
     }]
    relativeSpaceBone {boneName=pelvis}
    simSpaceSettings {maxAngularAcceleration=100, maxAngularVelocity=10, simSpaceAngularAlpha=1}
   }
   n10 = Anim Dynamics - Bone: dyn_skirt_fr_01_l(ComponentPose=@n9, Alpha=1, ExternalForce="0,0,0") {
    angularDampingOverride=1
    angularSpringConstant=10000
    bAngularSpring=true
    bOverrideAngularDamping=true
    bUseGravityOverride=true
    gravityScale=0.35
    simulationSpace=RootRelative
    boundBone {boneName=dyn_skirt_fr_01_l}
    componentAppliedLinearAccClamp=<4000 4000 4000>
    componentLinearAccScale=<0.4 0.4 0.45>
    componentLinearVelScale=<0.4 0.4 0.45>
    gravityOverride {z=-980}
    physicsBodyDefinitions [{
      boundBone {boneName=dyn_skirt_fr_01_l}
      boxExtents=<25 35 35>
      constraintSetup {angularLimitsMax=<0 5 5>, angularLimitsMin=<0 -10 -5>, angularTarget=<1 0 0>}
      localJointOffset=<4 0 0>
     }]
    relativeSpaceBone {boneName=pelvis}
    simSpaceSettings {maxAngularAcceleration=100, maxAngularVelocity=10, simSpaceAngularAlpha=1}
   }
   n11 = Anim Dynamics - Bone: dyn_skirt_sd_01_l(ComponentPose=@n10, Alpha=1, ExternalForce="0,0,0") {
    angularDampingOverride=1
    angularSpringConstant=10000
    bAngularSpring=true
    bOverrideAngularDamping=true
    bUseGravityOverride=true
    gravityScale=0.35
    simulationSpace=RootRelative
    boundBone {boneName=dyn_skirt_sd_01_l}
    componentAppliedLinearAccClamp=<4000 4000 4000>
    componentLinearAccScale=<0.4 0.4 0.45>
    componentLinearVelScale=<0.4 0.4 0.45>
    gravityOverride {z=-980}
    physicsBodyDefinitions [{
      boundBone {boneName=dyn_skirt_sd_01_l}
      boxExtents=<25 35 35>
      constraintSetup {angularLimitsMax=<0 5 5>, angularLimitsMin=<0 -20 -5>, angularTarget=<1 0 0>}
      localJointOffset=<4 0 0>
     }]
    relativeSpaceBone {boneName=pelvis}
    simSpaceSettings {maxAngularAcceleration=100, maxAngularVelocity=10, simSpaceAngularAlpha=1}
   }
   Save cached pose 'Post_AnimDynamics'(Pose=Component To Local(ComponentPose=@n11)) {cacheName=Post_AnimDynamics}
   n12 = Anim Dynamics - Bone: upperarm_bicep_l(ComponentPose=Local To Component(LocalPose=Use cached pose 'Post_AnimDynamics'()), Alpha=1, ExternalForce="0,0,0") {
    angularSpringConstant=15000
    bLinearSpring=true
    bOverrideLinearDamping=true
    bUseGravityOverride=true
    bUsePlanarLimit=false
    linearDampingOverride=1
    linearSpringConstant=500
    simulationSpace=RootRelative
    boundBone {boneName=upperarm_bicep_l}
    componentLinearAccScale=<0.2 0.2 0.2>
    componentLinearVelScale=<0.2 0.2 0.2>
    gravityOverride {z=-200}
    physicsBodyDefinitions [{
      boundBone {boneName=upperarm_bicep_l}
      boxExtents=<18 18 18>
      constraintSetup {angularLimitsMax=<0 0 0>, angularLimitsMin=<0 0 0>, angularTarget=<1 0 0>, linearAxesMax=<0 0 0.5>, linearAxesMin=<0 0 -0.5>}
      localJointOffset=<-2 0 0>
     }]
    relativeSpaceBone {boneName=upperarm_l}
    simSpaceSettings {maxAngularAcceleration=40, maxAngularVelocity=770, simSpaceAngularAlpha=1}
   }
   n13 = Anim Dynamics - Bone: upperarm_bicep_l(ComponentPose=@n12, Alpha=1, ExternalForce="0,0,0") {
    angularDampingOverride=0.9999
    angularSpringConstant=15000
    bAngularSpring=true
    bOverrideAngularDamping=true
    bUseGravityOverride=true
    bUsePlanarLimit=false
    simulationSpace=RootRelative
    boundBone {boneName=upperarm_bicep_l}
    componentLinearAccScale=<0.2 0.2 0.2>
    componentLinearVelScale=<0.2 0.2 0.2>
    gravityOverride {z=-980}
    physicsBodyDefinitions [{
      boundBone {boneName=upperarm_bicep_l}
      boxExtents=<6 18 6>
      constraintSetup {angularLimitsMax=<1 10 1>, angularLimitsMin=<-1 -10 -1>, angularTarget=<1 0 0>}
      localJointOffset=<-5 0 0>
     }]
    relativeSpaceBone {boneName=upperarm_l}
    simSpaceSettings {maxAngularAcceleration=10, maxAngularVelocity=300, simSpaceAngularAlpha=1}
   }
   n14 = Anim Dynamics - Bone: upperarm_bicep_r(ComponentPose=@n13, Alpha=1, ExternalForce="0,0,0") {
    angularSpringConstant=15000
    bLinearSpring=true
    bOverrideLinearDamping=true
    bUseGravityOverride=true
    bUsePlanarLimit=false
    linearDampingOverride=1
    linearSpringConstant=500
    simulationSpace=RootRelative
    boundBone {boneName=upperarm_bicep_r}
    componentLinearAccScale=<0.2 0.2 0.2>
    componentLinearVelScale=<0.2 0.2 0.2>
    gravityOverride {z=-200}
    physicsBodyDefinitions [{
      boundBone {boneName=upperarm_bicep_r}
      boxExtents=<18 18 18>
      constraintSetup {angularLimitsMax=<0 0 0>, angularLimitsMin=<0 0 0>, angularTarget=<1 0 0>, linearAxesMax=<0 0 0.5>, linearAxesMin=<0 0 -0.5>}
      localJointOffset=<2 0 0>
     }]
    relativeSpaceBone {boneName=upperarm_r}
    simSpaceSettings {maxAngularAcceleration=40, maxAngularVelocity=770, simSpaceAngularAlpha=1}
   }
   n15 = Anim Dynamics - Bone: upperarm_bicep_r(ComponentPose=@n14, Alpha=1, ExternalForce="0,0,0") {
    angularDampingOverride=0.9999
    angularSpringConstant=15000
    bAngularSpring=true
    bOverrideAngularDamping=true
    bUseGravityOverride=true
    bUsePlanarLimit=false
    simulationSpace=RootRelative
    boundBone {boneName=upperarm_bicep_r}
    componentLinearAccScale=<0.2 0.2 0.2>
    componentLinearVelScale=<0.2 0.2 0.2>
    gravityOverride {z=-980}
    physicsBodyDefinitions [{
      boundBone {boneName=upperarm_bicep_r}
      boxExtents=<6 18 6>
      constraintSetup {angularLimitsMax=<1 10 1>, angularLimitsMin=<-1 -10 -1>, angularTarget=<1 0 0>}
      localJointOffset=<-5 0 0>
     }]
    relativeSpaceBone {boneName=upperarm_r}
    simSpaceSettings {maxAngularAcceleration=10, maxAngularVelocity=300, simSpaceAngularAlpha=1}
   }
   n16 = Anim Dynamics - Bone: upperarm_tricep_l(ComponentPose=@n15, Alpha=1, ExternalForce="0,0,0") {
    angularSpringConstant=15000
    bLinearSpring=true
    bOverrideLinearDamping=true
    bUseGravityOverride=true
    bUsePlanarLimit=false
    linearDampingOverride=1
    linearSpringConstant=450
    simulationSpace=RootRelative
    boundBone {boneName=upperarm_tricep_l}
    componentLinearAccScale=<0.15 0.15 0.15>
    componentLinearVelScale=<0.15 0.15 0.15>
    gravityOverride {z=-100}
    physicsBodyDefinitions [{
      boundBone {boneName=upperarm_tricep_l}
      boxExtents=<30 40 30>
      constraintSetup {angularLimitsMax=<0 0 0>, angularLimitsMin=<0 0 0>, angularTarget=<1 0 0>, linearAxesMax=<0 0.1 0.75>, linearAxesMin=<0 -0.1 -0.75>}
      localJointOffset=<-3 0 0>
     }]
    relativeSpaceBone {boneName=upperarm_l}
    simSpaceSettings {maxAngularAcceleration=30, maxAngularVelocity=300, simSpaceAngularAlpha=1}
   }
   n17 = Anim Dynamics - Bone: upperarm_tricep_l(ComponentPose=@n16, Alpha=1, ExternalForce="0,0,0") {
    angularDampingOverride=0.9999
    angularSpringConstant=50000
    bAngularSpring=true
    bOverrideAngularDamping=true
    bUseGravityOverride=true
    bUsePlanarLimit=false
    simulationSpace=BoneRelative
    boundBone {boneName=upperarm_tricep_l}
    componentLinearAccScale=<0.15 0.15 0.15>
    componentLinearVelScale=<0.15 0.15 0.15>
    gravityOverride {z=-980}
    physicsBodyDefinitions [{
      boundBone {boneName=upperarm_tricep_l}
      boxExtents=<12 20 12>
      constraintSetup {angularLimitsMax=<1 10 1>, angularLimitsMin=<-1 -10 -1>, angularTarget=<1 0 0>}
      localJointOffset=<5 0 0>
     }]
    relativeSpaceBone {boneName=upperarm_l}
    simSpaceSettings {maxAngularAcceleration=10, maxAngularVelocity=300, simSpaceAngularAlpha=1}
   }
   n18 = Anim Dynamics - Bone: upperarm_tricep_r(ComponentPose=@n17, Alpha=1, ExternalForce="0,0,0") {
    angularSpringConstant=15000
    bLinearSpring=true
    bOverrideLinearDamping=true
    bUseGravityOverride=true
    bUsePlanarLimit=false
    linearDampingOverride=1
    linearSpringConstant=450
    simulationSpace=RootRelative
    boundBone {boneName=upperarm_tricep_r}
    componentLinearAccScale=<0.15 0.15 0.15>
    componentLinearVelScale=<0.15 0.15 0.15>
    gravityOverride {z=-100}
    physicsBodyDefinitions [{
      boundBone {boneName=upperarm_tricep_r}
      boxExtents=<30 40 30>
      constraintSetup {angularLimitsMax=<0 0 0>, angularLimitsMin=<0 0 0>, angularTarget=<1 0 0>, linearAxesMax=<0 0.1 0.75>, linearAxesMin=<0 -0.1 -0.75>}
      localJointOffset=<3 0 0>
     }]
    relativeSpaceBone {boneName=upperarm_r}
    simSpaceSettings {maxAngularAcceleration=30, maxAngularVelocity=300, simSpaceAngularAlpha=1}
   }
   n19 = Anim Dynamics - Bone: upperarm_tricep_r(ComponentPose=@n18, Alpha=1, ExternalForce="0,0,0") {
    angularDampingOverride=0.9999
    angularSpringConstant=50000
    bAngularSpring=true
    bOverrideAngularDamping=true
    bUseGravityOverride=true
    bUsePlanarLimit=false
    simulationSpace=BoneRelative
    boundBone {boneName=upperarm_tricep_r}
    componentLinearAccScale=<0.15 0.15 0.15>
    componentLinearVelScale=<0.15 0.15 0.15>
    gravityOverride {z=-980}
    physicsBodyDefinitions [{
      boundBone {boneName=upperarm_tricep_r}
      boxExtents=<12 20 12>
      constraintSetup {angularLimitsMax=<1 10 1>, angularLimitsMin=<-1 -10 -1>, angularTarget=<1 0 0>}
      localJointOffset=<5 0 0>
     }]
    relativeSpaceBone {boneName=upperarm_r}
    simSpaceSettings {maxAngularAcceleration=10, maxAngularVelocity=300, simSpaceAngularAlpha=1}
   }
   n20 = Anim Dynamics - Bone: thigh_quads_l(ComponentPose=@n19, Alpha=1, ExternalForce="0,0,0") {
    angularSpringConstant=15000
    bLinearSpring=true
    bOverrideLinearDamping=true
    bUseGravityOverride=true
    bUsePlanarLimit=false
    linearDampingOverride=1
    linearSpringConstant=350
    simulationSpace=RootRelative
    boundBone {boneName=thigh_quads_l}
    componentLinearAccScale=<0.42 0.42 0.42>
    componentLinearVelScale=<0.42 0.42 0.42>
    gravityOverride {z=-380}
    physicsBodyDefinitions [{
      boundBone {boneName=thigh_quads_l}
      boxExtents=<8 8 8>
      constraintSetup {angularLimitsMax=<0 0 0>, angularLimitsMin=<0 0 0>, angularTarget=<1 0 0>, linearAxesMax=<0 0 0.5>, linearAxesMin=<0 0 -0.5>}
      localJointOffset=<-2 0 0>
     }]
    relativeSpaceBone {boneName=upperarm_l}
    simSpaceSettings {maxAngularAcceleration=40, maxAngularVelocity=770, simSpaceAngularAlpha=1}
   }
   n21 = Anim Dynamics - Bone: thigh_quads_l(ComponentPose=@n20, Alpha=1, ExternalForce="0,0,0") {
    angularDampingOverride=0.9999
    angularSpringConstant=10000
    bAngularSpring=true
    bOverrideAngularDamping=true
    bUseGravityOverride=true
    bUsePlanarLimit=false
    simulationSpace=RootRelative
    boundBone {boneName=thigh_quads_l}
    componentLinearAccScale=<0.42 0.42 0.42>
    componentLinearVelScale=<0.42 0.42 0.42>
    gravityOverride {z=-980}
    physicsBodyDefinitions [{
      boundBone {boneName=thigh_quads_l}
      boxExtents=<6 11 6>
      constraintSetup {angularLimitsMax=<1 10 1>, angularLimitsMin=<-1 -10 -1>, angularTarget=<1 0 0>}
      localJointOffset=<-5 0 0>
     }]
    relativeSpaceBone {boneName=upperarm_l}
    simSpaceSettings {maxAngularAcceleration=10, maxAngularVelocity=300, simSpaceAngularAlpha=1}
   }
   n22 = Anim Dynamics - Bone: thigh_hamstrings_l(ComponentPose=@n21, Alpha=1, ExternalForce="0,0,0") {
    angularSpringConstant=15000
    bLinearSpring=true
    bOverrideLinearDamping=true
    bUseGravityOverride=true
    bUsePlanarLimit=false
    linearDampingOverride=1
    linearSpringConstant=350
    simulationSpace=RootRelative
    boundBone {boneName=thigh_hamstrings_l}
    componentLinearAccScale=<0.42 0.42 0.42>
    componentLinearVelScale=<0.42 0.42 0.42>
    gravityOverride {z=-380}
    physicsBodyDefinitions [{
      boundBone {boneName=thigh_hamstrings_l}
      boxExtents=<8 8 8>
      constraintSetup {angularLimitsMax=<0 0 0>, angularLimitsMin=<0 0 0>, angularTarget=<1 0 0>, linearAxesMax=<0 0 0.5>, linearAxesMin=<0 0 -0.5>}
      localJointOffset=<2 0 0>
     }]
    relativeSpaceBone {boneName=upperarm_r}
    simSpaceSettings {maxAngularAcceleration=40, maxAngularVelocity=770, simSpaceAngularAlpha=1}
   }
   n23 = Anim Dynamics - Bone: thigh_hamstrings_l(ComponentPose=@n22, Alpha=1, ExternalForce="0,0,0") {
    angularDampingOverride=0.9999
    angularSpringConstant=10000
    bAngularSpring=true
    bOverrideAngularDamping=true
    bUseGravityOverride=true
    bUsePlanarLimit=false
    simulationSpace=RootRelative
    boundBone {boneName=thigh_hamstrings_l}
    componentLinearAccScale=<0.42 0.42 0.42>
    componentLinearVelScale=<0.42 0.42 0.42>
    gravityOverride {z=-980}
    physicsBodyDefinitions [{
      boundBone {boneName=thigh_hamstrings_l}
      boxExtents=<6 11 6>
      constraintSetup {angularLimitsMax=<1 10 1>, angularLimitsMin=<-1 -10 -1>, angularTarget=<1 0 0>}
      localJointOffset=<-5 0 0>
     }]
    relativeSpaceBone {boneName=upperarm_r}
    simSpaceSettings {maxAngularAcceleration=10, maxAngularVelocity=300, simSpaceAngularAlpha=1}
   }
   n24 = Anim Dynamics - Bone: thigh_quads_r(ComponentPose=@n23, Alpha=1, ExternalForce="0,0,0") {
    angularSpringConstant=15000
    bLinearSpring=true
    bOverrideLinearDamping=true
    bUseGravityOverride=true
    bUsePlanarLimit=false
    linearDampingOverride=1
    linearSpringConstant=350
    simulationSpace=RootRelative
    boundBone {boneName=thigh_quads_r}
    componentLinearAccScale=<0.42 0.42 0.42>
    componentLinearVelScale=<0.42 0.42 0.42>
    gravityOverride {z=-380}
    physicsBodyDefinitions [{
      boundBone {boneName=thigh_quads_r}
      boxExtents=<6 6 6>
      constraintSetup {angularLimitsMax=<0 0 0>, angularLimitsMin=<0 0 0>, angularTarget=<1 0 0>, linearAxesMax=<0 0 0.5>, linearAxesMin=<0 0 -0.5>}
      localJointOffset=<-2 0 0>
     }]
    relativeSpaceBone {boneName=upperarm_l}
    simSpaceSettings {maxAngularAcceleration=40, maxAngularVelocity=770, simSpaceAngularAlpha=1}
   }
   n25 = Anim Dynamics - Bone: thigh_quads_r(ComponentPose=@n24, Alpha=1, ExternalForce="0,0,0") {
    angularDampingOverride=0.9999
    angularSpringConstant=10000
    bAngularSpring=true
    bOverrideAngularDamping=true
    bUseGravityOverride=true
    bUsePlanarLimit=false
    simulationSpace=RootRelative
    boundBone {boneName=thigh_quads_r}
    componentLinearAccScale=<0.4 0.4 0.4>
    componentLinearVelScale=<0.4 0.4 0.4>
    gravityOverride {z=-980}
    physicsBodyDefinitions [{
      boundBone {boneName=thigh_quads_r}
      boxExtents=<5 8 5>
      constraintSetup {angularLimitsMax=<1 10 1>, angularLimitsMin=<-1 -10 -1>, angularTarget=<1 0 0>}
      localJointOffset=<-5 0 0>
     }]
    relativeSpaceBone {boneName=upperarm_l}
    simSpaceSettings {maxAngularAcceleration=10, maxAngularVelocity=300, simSpaceAngularAlpha=1}
   }
   n26 = Anim Dynamics - Bone: thigh_hamstrings_r(ComponentPose=@n25, Alpha=1, ExternalForce="0,0,0") {
    angularSpringConstant=15000
    bLinearSpring=true
    bOverrideLinearDamping=true
    bUseGravityOverride=true
    bUsePlanarLimit=false
    linearDampingOverride=1
    linearSpringConstant=350
    simulationSpace=RootRelative
    boundBone {boneName=thigh_hamstrings_r}
    componentLinearAccScale=<0.42 0.42 0.42>
    componentLinearVelScale=<0.42 0.42 0.42>
    gravityOverride {z=-380}
    physicsBodyDefinitions [{
      boundBone {boneName=thigh_hamstrings_r}
      boxExtents=<6 6 6>
      constraintSetup {angularLimitsMax=<0 0 0>, angularLimitsMin=<0 0 0>, angularTarget=<1 0 0>, linearAxesMax=<0 0 0.5>, linearAxesMin=<0 0 -0.5>}
      localJointOffset=<2 0 0>
     }]
    relativeSpaceBone {boneName=upperarm_r}
    simSpaceSettings {maxAngularAcceleration=40, maxAngularVelocity=770, simSpaceAngularAlpha=1}
   }
   n27 = Anim Dynamics - Bone: thigh_hamstrings_r(ComponentPose=@n26, Alpha=1, ExternalForce="0,0,0") {
    angularDampingOverride=0.9999
    angularSpringConstant=10000
    bAngularSpring=true
    bOverrideAngularDamping=true
    bUseGravityOverride=true
    bUsePlanarLimit=false
    simulationSpace=RootRelative
    boundBone {boneName=thigh_hamstrings_r}
    componentLinearAccScale=<0.4 0.4 0.4>
    componentLinearVelScale=<0.4 0.4 0.4>
    gravityOverride {z=-980}
    physicsBodyDefinitions [{
      boundBone {boneName=thigh_hamstrings_r}
      boxExtents=<5 8 5>
      constraintSetup {angularLimitsMax=<1 10 1>, angularLimitsMin=<-1 -10 -1>, angularTarget=<1 0 0>}
      localJointOffset=<-5 0 0>
     }]
    relativeSpaceBone {boneName=upperarm_r}
    simSpaceSettings {maxAngularAcceleration=10, maxAngularVelocity=300, simSpaceAngularAlpha=1}
   }
   n28 = Anim Dynamics - Bone: deform_pec_l!(ComponentPose=@n27, ExternalForce="0,0,0") -> (Pose) {
    alpha=0.8
    angularSpringConstant=15000
    bLinearSpring=true
    bOverrideLinearDamping=true
    bUseGravityOverride=true
    bUsePlanarLimit=false
    linearDampingOverride=0.9995
    linearSpringConstant=400
    simulationSpace=RootRelative
    boundBone {boneName=deform_pec_l}
    componentAppliedLinearAccClamp=<1000 1000 1000>
    gravityOverride {z=-200}
    physicsBodyDefinitions [{
      boundBone {boneName=deform_pec_l}
      boxExtents=<50 55 60>
      constraintSetup {angularLimitsMax=<0 0 0>, angularLimitsMin=<0 0 0>, angularTarget=<1 0 0>, linearAxesMax=<1 0 3>, linearAxesMin=<-1 0 -3>}
      localJointOffset=<-4 0 0>
     }]
    relativeSpaceBone {boneName=pelvis}
    simSpaceSettings {maxAngularAcceleration=10, maxAngularVelocity=770, simSpaceAngularAlpha=0.4}
   }
   n29 = Anim Dynamics - Bone: deform_pec_r!(ComponentPose=@n28, ExternalForce="0,0,0") -> (Pose) {
    alpha=0.8
    angularSpringConstant=15000
    bLinearSpring=true
    bOverrideLinearDamping=true
    bUseGravityOverride=true
    bUsePlanarLimit=false
    linearDampingOverride=0.9995
    linearSpringConstant=400
    simulationSpace=RootRelative
    boundBone {boneName=deform_pec_r}
    componentAppliedLinearAccClamp=<1000 1000 1000>
    gravityOverride {z=-200}
    physicsBodyDefinitions [{
      boundBone {boneName=deform_pec_r}
      boxExtents=<50 55 60>
      constraintSetup {angularLimitsMax=<0 0 0>, angularLimitsMin=<0 0 0>, angularTarget=<1 0 0>, linearAxesMax=<1 0 3>, linearAxesMin=<-1 0 -3>}
      localJointOffset=<4 0 0>
     }]
    relativeSpaceBone {boneName=pelvis}
    simSpaceSettings {maxAngularAcceleration=10, maxAngularVelocity=770, simSpaceAngularAlpha=0.4}
   }
   n30 = "Constraint - attach_backpack (PPPPPPP)"!(
    ComponentPose=@n29
    Alpha=1
    ConstraintWeights_0=0.7184
    ConstraintWeights_1:float=0.1706
    ConstraintWeights_2:float=0.1052
    ConstraintWeights_3:float=0.0023
    ConstraintWeights_4:float=0.002
    ConstraintWeights_5:float=0.0013
    ConstraintWeights_6:float=0.0003
   ) -> (Pose) {
    boneToModify {boneName=attach_backpack}
    constraintSetup [
     {targetBone {boneName=spine_05}}
     {targetBone {boneName=spine_04}}
     {targetBone {boneName=spine_03}}
     {targetBone {boneName=neck_02}}
     {targetBone {boneName=clavicle_l}}
     {targetBone {boneName=neck_01}}
     {targetBone {boneName=spine_02}}
    ]
    constraintWeights [1,1,1,1,1,1,1]
   }
   n31 = "Transform (Modify) Bone - Bone: attach_backpack"(ComponentPose=@n30, Alpha=1, Translation="0,0.85,0", Rotation="0,0,1", Scale="1,1,1") {
    rotationMode=BMM_Additive
    rotationSpace=BCS_BoneSpace
    translationMode=BMM_Additive
    translationSpace=BCS_BoneSpace
    boneToModify {boneName=attach_backpack}
    rotation {roll=1}
    translation {y=0.85}
   }
   Save cached pose 'Post_Body_Dynamics'(Pose=Component To Local(ComponentPose=@n31)) {cacheName=Post_Body_Dynamics}
  ]}
 EventGraph {nodes [
   Event BlueprintUpdateAnimation()
   > n1 = Sequence!(execute) -> (then_0, then_1, then_2:Exec, then_3:Exec, then_4:Exec)
   > n2 = Branch(Condition=Get bIsCrouching())
   > Set Crouching(Crouching:double=1) -> (Output_Get:double)
   Set Crouching(execute=@n2.else, Crouching=0)
   n3 = Branch(execute=@n1.then_2, Condition=Get bIsPlayingMeleeAnim())
   > Set AnimDyn_Cloth_alplha(AnimDyn_Cloth_alplha:double=0.7) -> (Output_Get:double)
   n4 = Branch(execute=@n3.else, Condition=Get bIsDBNOCarrying() -> (bIsDBNOCarrying:bool))
   > Set AnimDyn_Cloth_alplha(AnimDyn_Cloth_alplha=0.65)
   n5 = Branch(execute=@n4.else, Condition=Get bIsCrouching())
   > Set AnimDyn_Cloth_alplha(AnimDyn_Cloth_alplha=0.65)
   n6 = Branch(execute=@n5.else, Condition=Get bIsMoving2D())
   > Set AnimDyn_Cloth_alplha(AnimDyn_Cloth_alplha=0.5)
   n7 = Branch(execute=@n6.else, Condition=Get bIsInFrontEnd())
   > Set AnimDyn_Cloth_alplha(AnimDyn_Cloth_alplha=0.4)
   Set AnimDyn_Cloth_alplha(execute=@n7.else, AnimDyn_Cloth_alplha=1)
   n8 = Get bIsParachuteOpened()
   n9 = Get bIsSkyDiving()
   n10 = Get bIsSkydiveFloating()
   n11 = Branch(execute=@n1.then_1, Condition=AND Boolean(A=OR Boolean(A=@n9, B=@n10), B=NOT Boolean(A=@n8)))
   > "Set SkyDiving "("SkyDiving ":double=1) -> (Output_Get:double)
   "Set SkyDiving "(execute=@n11.else, "SkyDiving "=0)
   n12 = NOT Boolean(A=Get bIsParachuteOpened())
   n13 = OR Boolean(A=Get bIsJumping(), B=Get bIsFalling())
   Set is_Jumping(execute=@n1.then_3, is_Jumping:double="To Float (Boolean)"(InBool=AND Boolean(A=@n13, B=@n12))) -> (Output_Get:double)
   n14 = Branch(execute=@n1.then_4, Condition=Get bIsPlayingEmote())
   > Set SimlationSpace(SimlationSpace:AnimPhysSimSpaceType Enum) -> (Output_Get:AnimPhysSimSpaceType Enum)
   > Set AlphaPlayingEmote(AlphaPlayingEmote:double=1) -> (Output_Get:double)
   Set SimlationSpace(execute=@n14.else)
   > Set AlphaPlayingEmote(AlphaPlayingEmote=0.45)
   Event OnBeginBaseballSliding() {is=title K2Node_Event}
   > Set Baseball_Sliding(Baseball_Sliding:double=1) -> (Output_Get:double)
   Event OnCharacterCustomizationLoaded()
   > Set Own Backpack Equiped(Own Backpack Equiped:bool=BackpackItemDefinitionHasAnyMetaTag!(Tags:Array of Gameplay Tag Structures=Get Tags() -> (Tags:Array of Gameplay Tag Structures), bRequiresEquipped:bool=false) -> (ReturnValue:bool)) -> (Output_Get:bool)
   Event OnEndBaseballSliding() {is=title K2Node_Event}
   > Set Baseball_Sliding(Baseball_Sliding=0)
   Event OnWeaponChanged()
   > Set isAR Equipped(isAR Equipped:bool=false) -> (Output_Get:bool)
   > Set isPistol Equipped(isPistol Equipped:bool=false) -> (Output_Get:bool)
   > Set Is_outLaw_SG(Is_outLaw_SG:bool=false) -> (Output_Get:bool)
   > Set SMG_SG(SMG_SG:bool=false) -> (Output_Get:bool)
   > n15 = Switch on EFortWeaponCoreAnimation!(execute, Selection:EFortWeaponCoreAnimation Enum=Get WeaponCoreAnim!() -> (WeaponCoreAnim)) -> (
    Melee:Exec
    Pistol:Exec
    Shotgun:Exec
    PaperBlueprint:Exec
    Rifle:Exec
    MeleeOneHand:Exec
    MachinePistol:Exec
    RocketLauncher:Exec
    GrenadeLauncher:Exec
    GoingCommando:Exec
    AssaultRifle:Exec
    TacticalShotgun:Exec
    SniperRifle:Exec
    TrapPlacement:Exec
    ShoulderLauncher:Exec
    AbilityDecoTool:Exec
    Crossbow:Exec
    C4:Exec
    RemoteControl:Exec
    DualWield:Exec
    AR_BullPup:Exec
    AR_ForwardGrip:Exec
    MedPackPaddles:Exec
    SMG_P90:Exec
    AR_DrumGun:Exec
    Consumable_Small:Exec
    Consumable_Large:Exec
    Balloon:Exec
    MountedTurret:Exec
    CreativeTool:Exec
    ExplosiveBow:Exec
    AshtonIndigo:Exec
    AshtonChicago:Exec
    MeleeDualWield:Exec
    Unarmed:Exec
   )
   Set isAR Equipped(execute=@n15.AssaultRifle, isAR Equipped=true)
   Set isPistol Equipped(execute=@n15.Pistol, isPistol Equipped=true)
   Set SMG_SG(execute=@n15.SMG_P90, SMG_SG=true)
   Set Is_outLaw_SG(execute=@n15.TacticalShotgun, Is_outLaw_SG=true)
  ]}
]
variables [
 AlphaPlayingEmote {type=number}
 AnimDyn_Cloth_alplha {type=number}
 Baseball_Sliding {type=number}
 Crouching {type=number}
 is_Jumping {type=number}
 Is_outLaw_SG {type=boolean}
 isAR Equipped {type=boolean}
 isPistol Equipped {type=boolean}
 Own Backpack Equiped {type=boolean}
 SimlationSpace {type=AnimPhysSimSpaceType}
 "SkyDiving " {type=number}
 SMG_SG {type=boolean}
 Tags {type="array<GameplayTag>", default [{tagName=Cosmetics.MetaTags.RebelTempo}]}
]
