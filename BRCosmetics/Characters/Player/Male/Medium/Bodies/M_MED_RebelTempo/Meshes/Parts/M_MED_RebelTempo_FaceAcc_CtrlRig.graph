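# Control Rig for the RebelTempo face accessory that procedurally offsets the hair base bones of this medium male body to suit gameplay states.
# Hair_Setup chains weighted ModifyTransforms, driven by interpolated head_up, sky floating and celebrating values plus a parachute weight.
# A DeltaRotation function converts head rotation relative to head_null into weights for further offsets in two opposite directions.
# Head-up and head-rotation offsets fade out while sky floating.
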ControlRigBlueprint /BRCosmetics/Characters/Player/Male/Medium/Bodies/M_MED_RebelTempo/Meshes/Parts/M_MED_RebelTempo_FaceAcc_CtrlRig
graphs [
 New Function_ContainedGraph {nodes [
   RigVMFunctionEntryNode()
   > RigVMFunction_Sequence = RigVMAggregateNode()
   Remap = &RigVMFunction_MathDoubleRemap(SourceMaximum=1, SourceMinimum=0, TargetMaximum=0, TargetMinimum=1, bClamp=False, Value=RigVMVariableNode() {variable=isSky_Floating}.Value)
   Multiply = &RigVMFunction_MathDoubleMul(A=@Remap.Result, B=RigVMVariableNode() {variable=head_up}.Value)
   n1 = &RigVMFunction_AlphaInterp(InterpSpeedDecreasing=10, InterpSpeedIncreasing=17, Scale=1, bClampResult=False, Value=@Multiply.Result)
   &RigUnit_ModifyTransforms(
    ItemToModify=((Item=(Type=Bone,Name="base_front_l"),Transform=(Rotation=(X=0,Y=0,Z=0,W=1),Translation=(X=0,Y=0,Z=0),Scale3D=(X=1,Y=1,Z=1))),(Item=(Type=Bone,Name="base_side_l"),Transform=(Rotation=(X=0,Y=0,Z=0,W=1),Translation=(X=0,Y=0,Z=0),Scale3D=(X=1,Y=1,Z=1))))
    WeightMinimum=0
    Weight=@n1.Result
    ExecutePin=@RigVMFunction_Sequence.A
   )
   > &RigUnit_ModifyTransforms(
    ItemToModify="((Item=(Type=Bone,Name=\"base_hair_l\"),Transform=(Rotation=(X=-0.0076,Y=-0.0868,Z=-0.0868,W=0.9924),Translation=(X=-4,Y=2,Z=0),Scale3D=(X=1,Y=1,Z=1))),(Item=(Type=Bone,Name=\"base_hair_m\"),Transform=(Rotation=(X=0.0114,Y=-0.0864,Z=0.13,W=0.9877),Translation= ...(truncated)"
    WeightMinimum=0
    Weight=@n1.Result
   )
   > &RigUnit_ModifyTransforms(
    ItemToModify="((Item=(Type=Bone,Name=\"base_hair_r\"),Transform=(Rotation=(X=0.0189,Y=-0.0851,Z=0.2156,W=0.9726),Translation=(X=-4,Y=0,Z=0),Scale3D=(X=1,Y=1,Z=1))),(Item=(Type=Bone,Name=\"base_side_r\"),Transform=(Rotation=(X=0,Y=0,Z=0,W=1),Translation=(X=-3,Y=0,Z=0),Scale3 ...(truncated)"
    WeightMinimum=0
    Weight=@n1.Result
   )
   > &RigUnit_ModifyTransforms(
    ItemToModify=((Item=(Type=Bone,Name="base_hair"),Transform=(Rotation=(X=0,Y=0,Z=0,W=1),Translation=(X=0,Y=0,Z=0),Scale3D=(X=0.9,Y=1,Z=1))))
    WeightMinimum=0
    Weight=@n1.Result
   )
   AlphaInterp = &RigVMFunction_AlphaInterp(InterpSpeedDecreasing=5, InterpSpeedIncreasing=10, Scale=1, bClampResult=False, Value=RigVMVariableNode() {variable=isSky_Floating}.Value)
   > &RigUnit_ModifyTransforms(
    ItemToModify="((Item=(Type=Bone,Name=\"base_front_l\"),Transform=(Rotation=(X=-0.0885,Y=-0.2432,Z=-0.3304,W=0.9077),Translation=(X=0,Y=0,Z=0),Scale3D=(X=1,Y=1,Z=1))),(Item=(Type=Bone,Name=\"base_side_l\"),Transform=(Rotation=(X=-0.1028,Y=-0.2826,Z=-0.3262,W=0.8962),Translat ...(truncated)"
    WeightMinimum=0
    Weight=@AlphaInterp.Result
   )
   > &RigUnit_ModifyTransforms(
    ItemToModify="((Item=(Type=Bone,Name=\"base_hair_l\"),Transform=(Rotation=(X=0,Y=-0.4617,Z=0,W=0.887),Translation=(X=-5,Y=0,Z=0),Scale3D=(X=1,Y=1,Z=1))),(Item=(Type=Bone,Name=\"base_hair_m\"),Transform=(Rotation=(X=-0.0996,Y=-0.5649,Z=-0.1422,W=0.8067),Translation=(X=-1,Y=0 ...(truncated)"
    WeightMinimum=0
    Weight=@AlphaInterp.Result
   )
   > &RigUnit_ModifyTransforms(
    ItemToModify="((Item=(Type=Bone,Name=\"base_hair_r\"),Transform=(Rotation=(X=-0.1318,Y=-0.5943,Z=-0.1717,W=0.7745),Translation=(X=-3,Y=0,Z=0),Scale3D=(X=1,Y=1,Z=1))),(Item=(Type=Bone,Name=\"base_side_r\"),Transform=(Rotation=(X=0.1094,Y=-0.4082,Z=0.2346,W=0.8754),Translatio ...(truncated)"
    WeightMinimum=0
    Weight=@AlphaInterp.Result
   )
   > &RigUnit_ModifyTransforms(
    ItemToModify=((Item=(Type=Bone,Name="base_hair"),Transform=(Rotation=(X=0,Y=0,Z=0,W=1),Translation=(X=0,Y=0,Z=0),Scale3D=(X=0.9,Y=1,Z=0.9))))
    WeightMinimum=0
    Weight=@AlphaInterp.Result
   )
   n2 = RigVMVariableNode() {variable=isParachuteOpened}
   > &RigUnit_ModifyTransforms(
    ItemToModify=((Item=(Type=Bone,Name="base_front_l"),Transform=(Rotation=(X=0,Y=0,Z=0,W=1),Translation=(X=-2,Y=0,Z=0),Scale3D=(X=1,Y=1,Z=1))),(Item=(Type=Bone,Name="base_side_l"),Transform=(Rotation=(X=0,Y=0,Z=0,W=1),Translation=(X=0,Y=0,Z=0),Scale3D=(X=1,Y=1,Z=1))))
    WeightMinimum=0
    Weight=@n2.Value
   )
   > &RigUnit_ModifyTransforms(
    ItemToModify=((Item=(Type=Bone,Name="base_hair_l"),Transform=(Rotation=(X=0,Y=0,Z=0,W=1),Translation=(X=0,Y=0,Z=0),Scale3D=(X=1,Y=1,Z=1))),(Item=(Type=Bone,Name="base_hair_m"),Transform=(Rotation=(X=0,Y=0,Z=0,W=1),Translation=(X=0,Y=0,Z=0),Scale3D=(X=1,Y=1,Z=1))))
    WeightMinimum=0
    Weight=@n2.Value
   )
   > &RigUnit_ModifyTransforms(
    ItemToModify=((Item=(Type=Bone,Name="base_hair_r"),Transform=(Rotation=(X=0,Y=0,Z=0,W=1),Translation=(X=0,Y=0,Z=0),Scale3D=(X=1,Y=1,Z=1))),(Item=(Type=Bone,Name="base_side_r"),Transform=(Rotation=(X=0,Y=0,Z=0,W=1),Translation=(X=0,Y=0,Z=0),Scale3D=(X=1,Y=1,Z=1))))
    WeightMinimum=0
    Weight=@n2.Value
   )
   > &RigUnit_ModifyTransforms(
    ItemToModify=((Item=(Type=Bone,Name="base_front_r"),Transform=(Rotation=(X=0,Y=0,Z=0.1736,W=0.9848),Translation=(X=0,Y=0,Z=0),Scale3D=(X=1,Y=1,Z=1))))
    WeightMinimum=0
    Weight=@n2.Value
   )
   n3 = &RigVMFunction_AlphaInterp(InterpSpeedDecreasing=3, InterpSpeedIncreasing=3, Scale=1, bClampResult=False, Value=RigVMVariableNode() {variable=isCelebrating}.Value)
   > &RigUnit_ModifyTransforms(
    ItemToModify="((Item=(Type=Bone,Name=\"base_front_l\"),Transform=(Rotation=(X=0,Y=0,Z=0,W=1),Translation=(X=0,Y=0,Z=0),Scale3D=(X=1,Y=1,Z=1))),(Item=(Type=Bone,Name=\"base_side_l\"),Transform=(Rotation=(X=0,Y=0,Z=0,W=1),Translation=(X=0,Y=0,Z=0),Scale3D=(X=1,Y=1,Z=1))),(Ite ...(truncated)"
    WeightMinimum=0
    Weight=@n3.Result
   )
   RigVMFunctionReturnNode(ExecuteContext=@RigVMFunction_Sequence.C)
   DeltaRotation_1 = RigVMFunctionReferenceNode(
    Rotation Order=ZYX
    child=(Type=Bone,Name="head")
    debugRotation=false
    outputMinMax=(X=0,Y=1)
    parent=(Type=Null,Name="head_null")
    premult=1
    rotationMinMax=(X=0,Y=10)
    triggerAxis=Z
    twistAxis=(X=1,Y=0,Z=0)
    ExecuteContext=@RigVMFunction_Sequence.B
   )
   n4 = &RigVMFunction_AlphaInterp(InterpSpeedDecreasing=3, InterpSpeedIncreasing=0.5, Scale=1, bClampResult=False, Value=@DeltaRotation_1.outputWeight)
   n5 = RigVMVariableNode() {variable=isSky_Floating}
   > &RigUnit_ModifyTransforms(
    ItemToModify="((Item=(Type=Bone,Name=\"base_front_l\"),Transform=(Rotation=(X=0,Y=0.0436,Z=0,W=0.999),Translation=(X=-1.5,Y=0,Z=0),Scale3D=(X=1,Y=1,Z=1))),(Item=(Type=Bone,Name=\"base_side_l\"),Transform=(Rotation=(X=0,Y=0,Z=0,W=1),Translation=(X=-1,Y=0,Z=0),Scale3D=(X=1,Y= ...(truncated)"
    WeightMinimum=0
    Weight=&RigVMFunction_MathDoubleMul(
     A=@n4.Result
     B=&RigVMFunction_MathDoubleRemap(SourceMaximum=1, SourceMinimum=0, TargetMaximum=0, TargetMinimum=1, bClamp=False, Value=@n5.Value).Result
    ).Result
   )
   > DeltaRotation = RigVMFunctionReferenceNode(
    Rotation Order=ZYX
    child=(Type=Bone,Name="head")
    debugRotation=false
    outputMinMax=(X=0,Y=1)
    parent=(Type=Null,Name="head_null")
    premult=1
    rotationMinMax=(X=-5,Y=-29)
    triggerAxis=Z
    twistAxis=(X=1,Y=0,Z=0)
   )
   n6 = &RigVMFunction_AlphaInterp(InterpSpeedDecreasing=3, InterpSpeedIncreasing=0.25, Scale=1, bClampResult=False, Value=@DeltaRotation.outputWeight)
   n7 = RigVMVariableNode() {variable=isSky_Floating}
   > &RigUnit_ModifyTransforms(
    ItemToModify="((Item=(Type=Bone,Name=\"base_front_l\"),Transform=(Rotation=(X=0,Y=0,Z=0,W=1),Translation=(X=0,Y=0,Z=0),Scale3D=(X=1,Y=1,Z=1))),(Item=(Type=Bone,Name=\"base_side_l\"),Transform=(Rotation=(X=0,Y=0,Z=0,W=1),Translation=(X=0,Y=0,Z=0),Scale3D=(X=1,Y=1,Z=1))),(Ite ...(truncated)"
    WeightMinimum=0
    Weight=&RigVMFunction_MathDoubleMul(
     A=@n6.Result
     B=&RigVMFunction_MathDoubleRemap(SourceMaximum=1, SourceMinimum=0, TargetMaximum=0, TargetMinimum=1, bClamp=False, Value=@n7.Value).Result
    ).Result
   )
  ]}
 New Function_ContainedGraph {nodes [
   n1 = RigVMFunctionEntryNode()
   RigVMFunction_MathBoolToInteger = &RigVMFunction_MathBoolToInteger(Value=@n1.onlyTwist)
   RigVMFunction_MathQuaternionSwingTwist = &RigVMFunction_MathQuaternionSwingTwist(
    twistAxis=@n1.twistAxis
    Input=&RigUnit_GetRelativeTransformForItem(bChildInitial=False, bParentInitial=False, child=@n1.child, parent=@n1.parent).RelativeTransform.Rotation
   )
   n2 = RigVMDispatchNode(
    Values=((),())
    Index=@RigVMFunction_MathBoolToInteger.Result
    Values.0=@RigVMFunction_MathQuaternionSwingTwist.Swing
    Values.1=@RigVMFunction_MathQuaternionSwingTwist.Twist
   ) {template="DISPATCH_RigVMDispatch_SelectInt32(in Index,in Values,out Result)"}
   RigVMFunction_MathQuaternionToEuler = &RigVMFunction_MathQuaternionToEuler(Value=@n2.Result, RotationOrder=@n1.Rotation Order)
   > DISPATCH_RigVMDispatch_Print = RigVMDispatchNode(ScreenColor=(R=1,G=1,B=1,A=1), ScreenDuration=1.5, Enabled=@n1.debugRotation, Value=@RigVMFunction_MathQuaternionToEuler.Result) {template="DISPATCH_RigVMDispatch_Print(in Prefix,in Value,in Enabled,in ScreenDuration,in ScreenColor)"}
   Cast = RigVMDispatchNode(Value=@n1.triggerAxis) {template="DISPATCH_RigVMDispatch_CastEnumToInt(in Value,out Result)"}
   DISPATCH_RigVMDispatch_SelectInt32 = RigVMDispatchNode(
    Values=(0,0,0)
    Index=@Cast.Result
    Values.0=@RigVMFunction_MathQuaternionToEuler.Result.X
    Values.1=@RigVMFunction_MathQuaternionToEuler.Result.Y
    Values.2=@RigVMFunction_MathQuaternionToEuler.Result.Z
   ) {template="DISPATCH_RigVMDispatch_SelectInt32(in Index,in Values,out Result)"}
   Remap = &RigVMFunction_MathDoubleRemap(
    bClamp=False
    Value=@DISPATCH_RigVMDispatch_SelectInt32.Result
    TargetMinimum=@n1.outputMinMax.X
    TargetMaximum=@n1.outputMinMax.Y
    SourceMinimum=@n1.rotationMinMax.X
    SourceMaximum=@n1.rotationMinMax.Y
   )
   Absolute = &RigVMFunction_MathDoubleAbs(Value=@Remap.Result)
   Multiply = &RigVMFunction_MathDoubleMul(
    A=&RigVMFunction_MathDoubleClamp(Value=RigVMDispatchNode(True=@Absolute.Result, Condition=@n1.absoluteValue, False=@Remap.Result) {template="DISPATCH_RigVMDispatch_If(in Condition,in True,in False,out Result)"}.Result).Result
    B=@n1.premult
   )
   > RigVMFunctionReturnNode(outputQuat=@n2.Result, outputWeight=@Multiply.Result)
  ]}
 RigVMFunctionLibrary {nodes [
   DeltaRotation = RigVMCollapseNode(
    ExecuteContext=()
    Rotation Order=ZYX
    child=(Type=Bone,Name="None")
    outputMinMax=(X=0,Y=1)
    parent=(Type=Bone,Name="None")
    premult=1
    rotationMinMax=(X=0,Y=0)
    twistAxis=(X=1,Y=0,Z=0)
   )
   Hair_Setup = RigVMCollapseNode(ExecuteContext=())
  ]}
 RigVMFunction_Sequence_1_ContainedGraph {nodes [
   n1 = RigVMFunctionEntryNode()
   RigVMFunction_Sequence = &RigVMFunction_Sequence(ExecuteContext=@n1.ExecuteContext)
   n2 = &RigVMFunction_Sequence(ExecuteContext=@RigVMFunction_Sequence.B)
   RigVMFunctionReturnNode(A=@RigVMFunction_Sequence.A, B=@n2.A, C=@n2.B)
  ]}
 RigVMModel {nodes [
   n1 = &RigUnit_PrepareForExecution()
   &RigUnit_HierarchyImportFromSkeleton(NameSpace=None, bIncludeCurves=False, bIncludeMeshSockets=False, bIncludeVirtualBones=True, ExecutePin=@n1.ExecutePin)
   > Control Stack at Item = RigVMFunctionReferenceNode(
    Add Bottom Null=False
    Add Null=False
    Add Secondary Control=False
    Control Suffix=_ctrl
    Item=(Name="base_hair",Type=Bone)
    Null Suffix=_null
    Parent=(Name="spine_05",Type=Bone)
    Secondary Settings="(Shape=(bVisible=True,Name=\"Default\",Color=(R=1,G=0,B=0,A=1),Transform=(Rotation=(X=0,Y=0,Z=0,W=1),Translation=(X=0,Y=0,Z=0),Scale3D=(X=1,Y=1,Z=1))),Proxy=(bIsProxy=False,ShapeVisibility=BasedOnSelection),DisplayName=\"None\",InitialSpace=LocalSpace,bUsePref ...(truncated)"
    Settings="(Shape=(bVisible=True,Name=\"Default\",Color=(R=0,G=0,B=1,A=1),Transform=(Rotation=(X=0,Y=0,Z=0,W=1),Translation=(X=0,Y=0,Z=0),Scale3D=(X=1,Y=1,Z=1))),Proxy=(bIsProxy=False,ShapeVisibility=BasedOnSelection),DisplayName=\"None\",InitialSpace=LocalSpace,bUsePref ...(truncated)"
    Tag Controls=true
    Zero Controls=true
   )
   RigUnit_BeginExecution = &RigUnit_BeginExecution()
   setFKCtrl = RigVMFunctionReferenceNode(setFK_Ctrl=true, ExecuteContext=@RigUnit_BeginExecution.ExecutePin)
   > Hair_Setup_1 = RigVMFunctionReferenceNode()
   Hair_Setup_2 = RigVMFunctionReferenceNode(ExecuteContext=())
  ]}
]
hierarchy [
 root {loc=<0 0 0>, rot=<0 0 0 1>}
 pelvis {parent=root, loc=<0 -0.7071 93.1191>, rot=<0.0127 0.707 -0.0127 -0.707>}
 spine_01 {parent=pelvis, loc=<6.7494 -0.6388 0>, rot=<0 0 0.034 -0.9994>}
 spine_02 {parent=spine_01, loc=<7.2518 -0.035 0>, rot=<0 0 0.0238 -0.9997>}
 spine_03 {parent=spine_02, loc=<6.0104 0 0>, rot=<0 0 0.0149 -0.9999>}
 spine_04 {parent=spine_03, loc=<8.1024 0 0>, rot=<0 0 -0.0132 -0.9999>}
 spine_05 {parent=spine_04, loc=<10.0146 0 0>, rot=<0 0 -0.0015 -1>}
 neck_01 {parent=spine_05, loc=<11.3138 0.1529 0>, rot=<0 0 -0.2003 -0.9797>}
 neck_02 {parent=neck_01, loc=<4.5627 0 0>, rot=<0 0 0.0504 -0.9987>}
 head {parent=neck_02, loc=<5.3914 0 0>, rot=<0 0 0.115 -0.9934>}
 base_hair {parent=head, loc=<5.156 0 0>, rot=<0 0 0 1>}
 base_front_l {parent=base_hair, loc=<2.5667 3.0473 -6.5249>, rot=<0.0513 -0.9922 -0.0319 -0.1092>}
 dyn_front_l_01 {parent=base_front_l, loc=<0 0 0>, rot=<0 0 0 1>}
 dyn_front_l_02 {parent=dyn_front_l_01, loc=<4.5065 -0.1573 -0.0998>, rot=<0.0305 0.023 0.0403 -0.9985>}
 dyn_front_l_03 {parent=dyn_front_l_02, loc=<4.3113 0.0001 -0.8828>, rot=<0 -0.1311 0 -0.9914>}
 base_side_l {parent=base_hair, loc=<2.9211 -2.4328 -6.7695>, rot=<0.0525 -0.9838 0.1274 -0.1144>}
 dyn_side_l_01 {parent=base_side_l, loc=<0 0 0>, rot=<0 0 0 1>}
 dyn_side_l_02 {parent=dyn_side_l_01, loc=<4.3018 -0.2082 0.0819>, rot=<0 -0.1142 0 -0.9935>}
 dyn_side_l_03 {parent=dyn_side_l_02, loc=<3.9031 -0.2227 0.2673>, rot=<0 0.0108 0 -0.9999>}
 base_hair_l {parent=base_hair, loc=<2.6201 -6.2373 -4.5867>, rot=<0.0621 -0.8616 0.4977 -0.0782>}
 dyn_hair_l_01 {parent=base_hair_l, loc=<0 0 0>, rot=<0 0 0 1>}
 dyn_hair_l_02 {parent=dyn_hair_l_01, loc=<4.8461 0 0.0277>, rot=<0 0.012 0 -0.9999>}
 dyn_hair_l_03 {parent=dyn_hair_l_02, loc=<4.4339 0 0.4166>, rot=<0 -0.1957 0 -0.9807>}
 base_hair_m {parent=base_hair, loc=<2.8423 -7.254 0>, rot=<0.0875 -0.7017 0.7017 -0.0875>}
 dyn_hair_m_01 {parent=base_hair_m, loc=<0 0 0>, rot=<0 0 0 1>}
 dyn_hair_m_02 {parent=dyn_hair_m_01, loc=<4.2432 0 0.0133>, rot=<0 0.012 0 -0.9999>}
 dyn_hair_m_03 {parent=dyn_hair_m_02, loc=<4.1999 0 0.2367>, rot=<0 0.1729 0 -0.9849>}
 base_hair_r {parent=base_hair, loc=<2.6202 -6.2373 4.5867>, rot=<0.0782 -0.4977 0.8616 -0.0621>}
 dyn_hair_r_01 {parent=base_hair_r, loc=<0 0 0>, rot=<0 0 0 1>}
 dyn_hair_r_02 {parent=dyn_hair_r_01, loc=<4.8458 0 0.0277>, rot=<0 0.012 0 -0.9999>}
 dyn_hair_r_03 {parent=dyn_hair_r_02, loc=<4.4342 0 0.4165>, rot=<0 -0.1957 0 -0.9807>}
 base_side_r {parent=base_hair, loc=<2.9211 -2.4328 6.7695>, rot=<0.1144 -0.1274 0.9838 -0.0525>}
 dyn_side_r_01 {parent=base_side_r, loc=<0 0 0>, rot=<0 0 0 1>}
 dyn_side_r_02 {parent=dyn_side_r_01, loc=<4.302 0.2082 0.0818>, rot=<0 -0.1142 0 -0.9935>}
 dyn_side_r_03 {parent=dyn_side_r_02, loc=<3.9033 0.2226 0.2673>, rot=<0 0.0108 0 -0.9999>}
 base_front_r {parent=base_hair, loc=<2.567 3.0473 6.5249>, rot=<0.1092 0.0319 0.9922 -0.0513>}
 dyn_front_r_01 {parent=base_front_r, loc=<0 0 0>, rot=<0 0 0 1>}
 dyn_front_r_02 {parent=dyn_front_r_01, loc=<4.507 0.1572 -0.0999>, rot=<-0.0305 0.023 -0.0403 -0.9985>}
 dyn_front_r_03 {parent=dyn_front_r_02, loc=<4.3106 0 -0.8827>, rot=<0 -0.1311 0 -0.9914>}
 head_null_2 {parent=head, type=Null, loc=<0 0 0>, rot=<0 0 0 1>}
 head_null {parent=neck_02, type=Null, loc=<5.3914 0 0>, rot=<0 0 0.115 -0.9934>}
 head_ctrl {controlType=EULER_TRANSFORM, parent=neck_02, type=Control, loc=<0 0 0>, rot=<0 0 0 1>}
 clavicle_l {parent=spine_05, loc=<4.3575 -0.1466 -2.3223>, rot=<0.0465 -0.7289 -0.009 -0.6829>}
 upperarm_l {parent=clavicle_l, loc=<15.5111 0 0>, rot=<0.0226 -0.3234 -0.0088 -0.9459>}
 clavicle_r {parent=spine_05, loc=<4.3575 -0.1466 2.3223>, rot=<0.7289 0.0465 -0.6829 0.009>}
 upperarm_r {parent=clavicle_r, loc=<-15.5111 0 0>, rot=<0.0226 -0.3234 -0.0088 -0.9459>}
]
variables [
 IsPlayingEmote {type=bool, default=False}
 headSwing_rot {type=FVector, default=(X=0,Y=0,Z=0)}
 headSwing_twist {type=FVector, default=(X=0,Y=0,Z=0)}
 head_up {type=double, default=0}
 isCelebrating {type=double, default=0}
 isCrouching {type=double, default=0}
 isParachuteOpened {type=double, default=0}
 isPistol Equipped lerp {type=double, default=0}
 isPistolASDUp Extreme {type=double, default=0}
 isSky_Floating {type=double, default=0}
 isTargeting lerp {type=double, default=0}
 pub isPistol Equipped {type=bool, default=False}
 pub isTargeting {type=bool, default=False}
]
