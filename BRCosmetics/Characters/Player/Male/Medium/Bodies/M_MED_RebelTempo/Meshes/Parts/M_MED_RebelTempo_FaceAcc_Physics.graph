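# Physics asset for the RebelTempo male face accessory mesh, simulating its dangling hair strands.
# Seven three-segment capsule chains (front, side and hair) hang from small base bodies, with narrow swing and near-locked twist limits.
# Extra constraints with free rotation tie the left and right hair tips to the middle tip, linking those chains.
# Head, neck, upper spine, clavicle and upper-arm capsules supply body shapes for collision.
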PhysicsAsset /BRCosmetics/Characters/Player/Male/Medium/Bodies/M_MED_RebelTempo/Meshes/Parts/M_MED_RebelTempo_FaceAcc_Physics
bodies {
 base_front_l [{shapeType=Capsule, length=0.505, radius=0.505, shapeName=base_front_l_capsule, rotation=<rot 0 -90 0>}]
 base_front_r [{shapeType=Capsule, length=0.505, radius=0.505, shapeName=base_front_r_capsule, rotation=<rot 0 -90 0>}]
 base_hair_l [{shapeType=Capsule, length=0.505, radius=0.505, shapeName=base_hair_l_capsule, rotation=<rot 0 -90 0>}]
 base_hair_m [{shapeType=Capsule, length=0.505, radius=0.505, shapeName=base_hair_m_capsule, rotation=<rot 0 -90 0>}]
 base_hair_r [{shapeType=Capsule, length=0.505, radius=0.505, shapeName=base_hair_r_capsule, rotation=<rot 0 -90 0>}]
 base_side_l [{shapeType=Capsule, length=0.505, radius=0.505, shapeName=base_side_l_capsule, rotation=<rot 0 -90 0>}]
 base_side_r [{shapeType=Capsule, length=0.505, radius=0.505, shapeName=base_side_r_capsule, rotation=<rot 0 -90 0>}]
 clavicle_l [{
   shapeType=Capsule
   length=10
   radius=10.25
   shapeName=clavicle_l_capsule
   center=<8.1933 -3.1196 0.7258>
   rotation=<rot 81.6306 -4.0436 -13.9636>
  }]
 clavicle_r [
  {
   shapeType=Capsule
   length=8
   radius=10.5
   shapeName=clavicle_r_capsule
   center=<-8.6319 1.4852 0.347>
   rotation=<rot 79.8449 -0.0037 175.7113>
  }
  {
   shapeType=Capsule
   length=20
   radius=3
   shapeName=clavicle_r_capsule
   center=<-20.281 11.439 -10.7021>
   rotation=<rot 79.8449 -0.0034 110.9805>
  }
 ]
 dyn_front_l_01 [{shapeType=Capsule, length=2, radius=2, shapeName=dyn_front_l_01_capsule, center=<1 0 0>, rotation=<rot 90 0 0>}]
 dyn_front_l_02 [{shapeType=Capsule, length=2, radius=2, shapeName=dyn_front_l_02_capsule, center=<1 0 0>, rotation=<rot 90 0 0>}]
 dyn_front_l_03 [{shapeType=Capsule, length=3, radius=1.5, shapeName=dyn_front_l_03_capsule, center=<1.25 0 0>, rotation=<rot 90 0 0>}]
 dyn_front_r_01 [{shapeType=Capsule, length=2, radius=2, shapeName=dyn_front_r_01_capsule, center=<1 0 0>, rotation=<rot 90 0 0>}]
 dyn_front_r_02 [{shapeType=Capsule, length=2, radius=2, shapeName=dyn_front_r_02_capsule, center=<1 0 0>, rotation=<rot 90 0 0>}]
 dyn_front_r_03 [{shapeType=Capsule, length=3, radius=1.5, shapeName=dyn_front_r_03_capsule, center=<1.5 0 0>, rotation=<rot 90 0 0>}]
 dyn_hair_l_01 [{shapeType=Capsule, length=2, radius=2, shapeName=dyn_hair_l_01_capsule, center=<1 0 0>, rotation=<rot 90 0 0>}]
 dyn_hair_l_02 [{shapeType=Capsule, length=2, radius=2, shapeName=dyn_hair_l_02_capsule, center=<1 0 0>, rotation=<rot 90 0 0>}]
 dyn_hair_l_03 [{shapeType=Capsule, length=3, radius=1.5, shapeName=dyn_hair_l_03_capsule, center=<1.2 0 1.2>, rotation=<rot 56.8269 -179.9999 180>}]
 dyn_hair_m_01 [{shapeType=Capsule, length=2, radius=2, shapeName=dyn_hair_m_01_capsule, center=<1 0 0>, rotation=<rot 90 0 0>}]
 dyn_hair_m_02 [{shapeType=Capsule, length=2, radius=2, shapeName=dyn_hair_m_02_capsule, center=<1 0 0>, rotation=<rot 90 0 0>}]
 dyn_hair_m_03 [{shapeType=Capsule, length=4, radius=1.5, shapeName=dyn_hair_m_03_capsule, center=<2 0 0>, rotation=<rot 71.3926 0 0>}]
 dyn_hair_r_01 [{shapeType=Capsule, length=2, radius=2, shapeName=dyn_hair_r_01_capsule, center=<1 0 0>, rotation=<rot 90 0 0>}]
 dyn_hair_r_02 [{shapeType=Capsule, length=2, radius=1.75, shapeName=dyn_hair_r_02_capsule, center=<1 0 0>, rotation=<rot 90 0 0>}]
 dyn_hair_r_03 [{shapeType=Capsule, length=3.5, radius=1.2, shapeName=dyn_hair_r_03_capsule, center=<1 0 1.5>, rotation=<rot 60.5505 179.9989 -179.9992>}]
 dyn_side_l_01 [{shapeType=Capsule, length=2, radius=2, shapeName=dyn_side_l_01_capsule, center=<1 0 0>, rotation=<rot 90 0 0>}]
 dyn_side_l_02 [{shapeType=Capsule, length=2, radius=2, shapeName=dyn_side_l_02_capsule, center=<1 0 0>, rotation=<rot 90 0 0>}]
 dyn_side_l_03 [{shapeType=Capsule, length=3.5, radius=1.25, shapeName=dyn_side_l_03_capsule, center=<0.5 0 0>, rotation=<rot 90 0 0>}]
 dyn_side_r_01 [{shapeType=Capsule, length=2, radius=2, shapeName=dyn_side_r_01_capsule, center=<1 0 0>, rotation=<rot 90 0 0>}]
 dyn_side_r_02 [{shapeType=Capsule, length=2, radius=1.75, shapeName=dyn_side_r_02_capsule, center=<1 0 0>, rotation=<rot 90 0 0>}]
 dyn_side_r_03 [{shapeType=Capsule, length=1.5, radius=1.25, shapeName=dyn_side_r_03_capsule, center=<0.75 0 0>, rotation=<rot 90 0 0>}]
 head [{shapeType=Capsule, length=0, radius=6.75, shapeName=head_capsule, center=<2.1265 -0.5849 -0.2>}]
 neck_01 [{shapeType=Capsule, length=0.5, radius=8, shapeName=neck_01_capsule, center=<1 0 0.374>}]
 neck_02 [{shapeType=Capsule, length=0, radius=7.5, shapeName=neck_02_capsule, center=<1 -1.274 0.2302>, rotation=<rot 0 -90 0>}]
 spine_05 [{shapeType=Capsule, length=16, radius=14, shapeName=spine_05_capsule, center=<1.8199 -1.3678 -0.0075>, rotation=<rot 1.689 0.0415 1.7786>}]
 upperarm_l [{
   shapeType=Capsule
   length=20
   radius=8
   shapeName=upperarm_l_capsule
   center=<5.4994 -2.134 -0.0003>
   rotation=<rot 87.6399 -179.3937 177.0224>
  }]
 upperarm_r [{
   shapeType=Capsule
   length=23
   radius=8
   shapeName=upperarm_r_capsule
   center=<-9.2569 1.2172 -0.7349>
   rotation=<rot 89.2451 0.6114 179.9978>
  }]
}
constraints [
 {bone1Name=dyn_front_r_03, bone2Name=dyn_front_r_02, swing1LimitDegrees=14, swing2LimitDegrees=8, twistLimitDegrees=1}
 {bone1Name=dyn_hair_l_01, bone2Name=base_hair_l, swing1LimitDegrees=8, swing2LimitDegrees=8, twistLimitDegrees=1}
 {bone1Name=dyn_hair_l_02, bone2Name=dyn_hair_l_01, swing1LimitDegrees=8, swing2LimitDegrees=8, twistLimitDegrees=1}
 {bone1Name=dyn_hair_l_03, bone2Name=dyn_hair_l_02, swing1LimitDegrees=8, swing2LimitDegrees=12, twistLimitDegrees=1}
 {bone1Name=dyn_hair_m_01, bone2Name=base_hair_m, swing1LimitDegrees=8, swing2LimitDegrees=8, twistLimitDegrees=1}
 {bone1Name=dyn_hair_m_02, bone2Name=dyn_hair_m_01, swing1LimitDegrees=8, swing2LimitDegrees=8, twistLimitDegrees=1}
 {bone1Name=dyn_hair_m_03, bone2Name=dyn_hair_m_02, swing1LimitDegrees=8, swing2LimitDegrees=12, twistLimitDegrees=1}
 {bone1Name=dyn_hair_r_01, bone2Name=base_hair_r, swing1LimitDegrees=8, swing2LimitDegrees=8, twistLimitDegrees=1}
 {bone1Name=dyn_hair_r_02, bone2Name=dyn_hair_r_01, swing1LimitDegrees=8, swing2LimitDegrees=8, twistLimitDegrees=1}
 {bone1Name=dyn_hair_r_03, bone2Name=dyn_hair_r_02, swing1LimitDegrees=8, swing2LimitDegrees=10, twistLimitDegrees=1}
 {bone1Name=dyn_side_l_01, bone2Name=base_side_l, swing1LimitDegrees=8, swing2LimitDegrees=8, twistLimitDegrees=1}
 {bone1Name=dyn_side_l_02, bone2Name=dyn_side_l_01, swing1LimitDegrees=8, swing2LimitDegrees=8, twistLimitDegrees=1}
 {bone1Name=dyn_side_l_03, bone2Name=dyn_side_l_02, swing1LimitDegrees=8, swing2LimitDegrees=8, twistLimitDegrees=1}
 {bone1Name=dyn_side_r_01, bone2Name=base_side_r, swing1LimitDegrees=8, swing2LimitDegrees=8, twistLimitDegrees=1}
 {bone1Name=dyn_side_r_02, bone2Name=dyn_side_r_01, swing1LimitDegrees=8, swing2LimitDegrees=8, twistLimitDegrees=1}
 {bone1Name=dyn_side_r_03, bone2Name=dyn_side_r_02, swing1LimitDegrees=8, swing2LimitDegrees=8, twistLimitDegrees=1}
 {bone1Name=dyn_front_l_01, bone2Name=base_front_l, swing1LimitDegrees=14, swing2LimitDegrees=8, twistLimitDegrees=1}
 {bone1Name=dyn_front_l_02, bone2Name=dyn_front_l_01, swing1LimitDegrees=14, swing2LimitDegrees=8, twistLimitDegrees=1}
 {bone1Name=dyn_front_l_03, bone2Name=dyn_front_l_02, swing1LimitDegrees=14, swing2LimitDegrees=8, twistLimitDegrees=1}
 {bone1Name=dyn_front_r_01, bone2Name=base_front_r, swing1LimitDegrees=14, swing2LimitDegrees=8, twistLimitDegrees=1}
 {bone1Name=dyn_front_r_02, bone2Name=dyn_front_r_01, swing1LimitDegrees=14, swing2LimitDegrees=8, twistLimitDegrees=1}
 {
  bone1Name=dyn_hair_l_03
  bone2Name=dyn_hair_m_03
  swing1LimitDegrees=45
  swing1Motion=Free
  swing2LimitDegrees=45
  swing2Motion=Free
  twistLimitDegrees=45
  twistMotion=Free
 }
 {
  bone1Name=dyn_hair_r_03
  bone2Name=dyn_hair_m_03
  swing1LimitDegrees=45
  swing1Motion=Free
  swing2LimitDegrees=45
  swing2Motion=Free
  twistLimitDegrees=45
  twistMotion=Free
 }
]
