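# Physics asset for the skydive face accessory of the male medium RebelTempo cosmetic body, setting up secondary motion for its hair, front and side strands.
# Seven chains each link a small base anchor capsule to three dynamic capsules through constraints with uniform, tight swing and twist limits.
# Larger unconstrained capsules on the head, neck, upper spine, clavicles and upper arms provide body collision shapes around the strands.
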PhysicsAsset /BRCosmetics/Characters/Player/Male/Medium/Bodies/M_MED_RebelTempo/Meshes/Parts/M_MED_RebelTempo_FaceAcc_SkyDive_Physics
bodies {
 base_front_l [{shapeType=Capsule, length=0.505, radius=0.505, shapeName=base_front_l_capsule, rotation=<rot 0 -90 0>}]
 base_front_r [{shapeType=Capsule, length=0.505, radius=0.505, shapeName=base_front_r_capsule, rotation=<rot 0 -90 0>}]
 base_hair_l [{shapeType=Capsule, length=0.505, radius=0.505, shapeName=base_hair_l_capsule, rotation=<rot 0 -90 0>}]
 base_hair_m [{shapeType=Capsule, length=0.505, radius=0.505, shapeName=base_hair_m_capsule, rotation=<rot 0 -90 0>}]
 base_hair_r [{shapeType=Capsule, length=0.505, radius=0.505, shapeName=base_hair_r_capsule, rotation=<rot 0 -90 0>}]
 base_side_l [{shapeType=Capsule, length=0.505, radius=0.505, shapeName=base_side_l_capsule, rotation=<rot 0 -90 0>}]
 base_side_r [{shapeType=Capsule, length=0.505, radius=0.505, shapeName=base_side_r_capsule, rotation=<rot 0 -90 0>}]
 clavicle_l [{
   shapeType=Capsule
   length=10
   radius=11.75
   shapeName=clavicle_l_capsule
   center=<7.8021 0.317 -0.3686>
   rotation=<rot 85.4622 -12.0265 -16.2363>
  }]
 clavicle_r [{
   shapeType=Capsule
   length=10
   radius=11.75
   shapeName=clavicle_r_capsule
   center=<-7.8021 -0.317 0.3686>
   rotation=<rot 85.4622 -12.0265 163.7637>
  }]
 dyn_front_l_01 [{shapeType=Capsule, length=2, radius=2, shapeName=dyn_front_l_01_capsule, center=<1 0 0>, rotation=<rot 90 0 0>}]
 dyn_front_l_02 [{shapeType=Capsule, length=2, radius=2, shapeName=dyn_front_l_02_capsule, center=<1 0 0>, rotation=<rot 90 0 0>}]
 dyn_front_l_03 [{shapeType=Capsule, length=3, radius=1.5, shapeName=dyn_front_l_03_capsule, center=<0.5 0 0>, rotation=<rot 90 0 0>}]
 dyn_front_r_01 [{shapeType=Capsule, length=2, radius=2, shapeName=dyn_front_r_01_capsule, center=<1 0 0>, rotation=<rot 90 0 0>}]
 dyn_front_r_02 [{shapeType=Capsule, length=2, radius=2, shapeName=dyn_front_r_02_capsule, center=<1 0 0>, rotation=<rot 90 0 0>}]
 dyn_front_r_03 [{shapeType=Capsule, length=3, radius=1.5, shapeName=dyn_front_r_03_capsule, center=<1.5 0 0>, rotation=<rot 90 0 0>}]
 dyn_hair_l_01 [{shapeType=Capsule, length=2, radius=2, shapeName=dyn_hair_l_01_capsule, center=<1 0 0>, rotation=<rot 90 0 0>}]
 dyn_hair_l_02 [{shapeType=Capsule, length=2, radius=2, shapeName=dyn_hair_l_02_capsule, center=<1 0 0>, rotation=<rot 90 0 0>}]
 dyn_hair_l_03 [{shapeType=Capsule, length=3, radius=1.5, shapeName=dyn_hair_l_03_capsule, center=<1 0 1>, rotation=<rot 63.1096 -179.9999 179.9999>}]
 dyn_hair_m_01 [{shapeType=Capsule, length=2, radius=2, shapeName=dyn_hair_m_01_capsule, center=<1 0 0>, rotation=<rot 90 0 0>}]
 dyn_hair_m_02 [{shapeType=Capsule, length=2, radius=2, shapeName=dyn_hair_m_02_capsule, center=<1 0 0>, rotation=<rot 90 0 0>}]
 dyn_hair_m_03 [{shapeType=Capsule, length=4, radius=1.5, shapeName=dyn_hair_m_03_capsule, center=<2 0 0>, rotation=<rot 71.3926 0 0>}]
 dyn_hair_r_01 [{shapeType=Capsule, length=2, radius=2, shapeName=dyn_hair_r_01_capsule, center=<1 0 0>, rotation=<rot 90 0 0>}]
 dyn_hair_r_02 [{shapeType=Capsule, length=2, radius=1.75, shapeName=dyn_hair_r_02_capsule, center=<1 0 0>, rotation=<rot 90 0 0>}]
 dyn_hair_r_03 [{shapeType=Capsule, length=3.5, radius=1.5, shapeName=dyn_hair_r_03_capsule, center=<1.25 0 1>, rotation=<rot 60.4295 179.9989 -179.9991>}]
 dyn_side_l_01 [{shapeType=Capsule, length=2, radius=2, shapeName=dyn_side_l_01_capsule, center=<1 0 0>, rotation=<rot 90 0 0>}]
 dyn_side_l_02 [{shapeType=Capsule, length=2, radius=2, shapeName=dyn_side_l_02_capsule, center=<1 0 0>, rotation=<rot 83.7858 -180 180>}]
 dyn_side_l_03 [{shapeType=Capsule, length=3.5, radius=1.5, shapeName=dyn_side_l_03_capsule, center=<1.5 0 0>, rotation=<rot 90 0 0>}]
 dyn_side_r_01 [{shapeType=Capsule, length=2, radius=2, shapeName=dyn_side_r_01_capsule, center=<1 0 0>, rotation=<rot 90 0 0>}]
 dyn_side_r_02 [{shapeType=Capsule, length=2, radius=1.75, shapeName=dyn_side_r_02_capsule, center=<1 0 0>, rotation=<rot 90 0 0>}]
 dyn_side_r_03 [{shapeType=Capsule, length=1.5, radius=1.25, shapeName=dyn_side_r_03_capsule, center=<0.75 0 0>, rotation=<rot 90 0 0>}]
 head [{shapeType=Capsule, length=0.5, radius=7, shapeName=head_capsule, center=<3.4787 0.9201 0>}]
 neck_01 [{shapeType=Capsule, length=0.5, radius=10, shapeName=neck_01_capsule}]
 neck_02 [{shapeType=Capsule, length=0, radius=9, shapeName=neck_02_capsule, center=<1 -1.274 0>, rotation=<rot 0 -90 0>}]
 spine_05 [{shapeType=Capsule, length=16, radius=15.5, shapeName=spine_05_capsule, center=<2.0823 -7.7942 0>}]
 upperarm_l [{
   shapeType=Capsule
   length=23
   radius=7.75
   shapeName=upperarm_l_capsule
   center=<11.2092 -1.1963 1.4328>
   rotation=<rot 88.8177 0.6089 -0.0051>
  }]
 upperarm_r [{shapeType=Capsule, length=23, radius=7.5, shapeName=upperarm_r_capsule, center=<-11.3601 1.1947 -1.2176>, rotation=<rot 90 -179.3861 0>}]
}
constraints [
 {bone1Name=dyn_front_r_03, bone2Name=dyn_front_r_02, swing1LimitDegrees=15, swing2LimitDegrees=15, twistLimitDegrees=2.5}
 {bone1Name=dyn_hair_l_01, bone2Name=base_hair_l, swing1LimitDegrees=15, swing2LimitDegrees=15, twistLimitDegrees=2.5}
 {bone1Name=dyn_hair_l_02, bone2Name=dyn_hair_l_01, swing1LimitDegrees=15, swing2LimitDegrees=15, twistLimitDegrees=2.5}
 {bone1Name=dyn_hair_l_03, bone2Name=dyn_hair_l_02, swing1LimitDegrees=15, swing2LimitDegrees=15, twistLimitDegrees=2.5}
 {bone1Name=dyn_hair_m_01, bone2Name=base_hair_m, swing1LimitDegrees=15, swing2LimitDegrees=15, twistLimitDegrees=2.5}
 {bone1Name=dyn_hair_m_02, bone2Name=dyn_hair_m_01, swing1LimitDegrees=15, swing2LimitDegrees=15, twistLimitDegrees=2.5}
 {bone1Name=dyn_hair_m_03, bone2Name=dyn_hair_m_02, swing1LimitDegrees=15, swing2LimitDegrees=15, twistLimitDegrees=2.5}
 {bone1Name=dyn_hair_r_01, bone2Name=base_hair_r, swing1LimitDegrees=15, swing2LimitDegrees=15, twistLimitDegrees=2.5}
 {bone1Name=dyn_hair_r_02, bone2Name=dyn_hair_r_01, swing1LimitDegrees=15, swing2LimitDegrees=15, twistLimitDegrees=2.5}
 {bone1Name=dyn_hair_r_03, bone2Name=dyn_hair_r_02, swing1LimitDegrees=15, swing2LimitDegrees=15, twistLimitDegrees=2.5}
 {bone1Name=dyn_side_l_01, bone2Name=base_side_l, swing1LimitDegrees=15, swing2LimitDegrees=15, twistLimitDegrees=2.5}
 {bone1Name=dyn_side_l_02, bone2Name=dyn_side_l_01, swing1LimitDegrees=15, swing2LimitDegrees=15, twistLimitDegrees=2.5}
 {bone1Name=dyn_side_l_03, bone2Name=dyn_side_l_02, swing1LimitDegrees=15, swing2LimitDegrees=15, twistLimitDegrees=2.5}
 {bone1Name=dyn_side_r_01, bone2Name=base_side_r, swing1LimitDegrees=15, swing2LimitDegrees=15, twistLimitDegrees=2.5}
 {bone1Name=dyn_side_r_02, bone2Name=dyn_side_r_01, swing1LimitDegrees=15, swing2LimitDegrees=15, twistLimitDegrees=2.5}
 {bone1Name=dyn_side_r_03, bone2Name=dyn_side_r_02, swing1LimitDegrees=15, swing2LimitDegrees=15, twistLimitDegrees=2.5}
 {bone1Name=dyn_front_l_01, bone2Name=base_front_l, swing1LimitDegrees=15, swing2LimitDegrees=15, twistLimitDegrees=2.5}
 {bone1Name=dyn_front_l_02, bone2Name=dyn_front_l_01, swing1LimitDegrees=15, swing2LimitDegrees=15, twistLimitDegrees=2.5}
 {bone1Name=dyn_front_l_03, bone2Name=dyn_front_l_02, swing1LimitDegrees=15, swing2LimitDegrees=15, twistLimitDegrees=2.5}
 {bone1Name=dyn_front_r_01, bone2Name=base_front_r, swing1LimitDegrees=15, swing2LimitDegrees=15, twistLimitDegrees=2.5}
 {bone1Name=dyn_front_r_02, bone2Name=dyn_front_r_01, swing1LimitDegrees=15, swing2LimitDegrees=15, twistLimitDegrees=2.5}
]
