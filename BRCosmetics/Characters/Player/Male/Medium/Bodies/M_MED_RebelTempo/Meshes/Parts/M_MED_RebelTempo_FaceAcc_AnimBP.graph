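# Post-copy AnimBP for the male RebelTempo face accessory, adding procedural hair and bang motion.
# Merges parent and body mesh poses, then runs a Control Rig driven by head tilt, skydive, parachute, crouch and celebration values from the EventGraph.
# That feeds two head-space RigidBody sims, blended on skydiving to swap physics assets and drag.
# An additive bone tweak and four chained AnimDynamics bang bones finish the pose.
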AnimBlueprint /BRCosmetics/Characters/Player/Male/Medium/Bodies/M_MED_RebelTempo/Meshes/Parts/M_MED_RebelTempo_FaceAcc_AnimBP
target_skeleton=&/M_MED_RebelTempo_FaceAcc_Skeleton
graphs [
 AnimGraph {nodes [
   n1 = RigidBody!(
    ComponentPose=Local To Component(LocalPose=Use cached pose 'Initia Rig Cache'())
    Alpha
    OverrideWorldGravity:Vector="0,0,0"
    ExternalForce
    SimSpaceSettings:Sim Space Settings Structure=(WorldAlpha=0.7,VelocityScaleZ=0.6,DampingAlpha=0.2,MaxLinearVelocity=600,MaxAngularVelocity=4,MaxLinearAcceleration=750,MaxAngularAcceleration=1000,ExternalLinearDragV=(X=3,Y=3,Z=3),ExternalLinearVelocity=(X=0,Y=0,Z=0),ExternalAngularVelocity=(X=0,Y=0,Z=0))
   ) -> (Pose) {
    bOverrideWorldGravity=true
    overridePhysicsAsset=&/M_MED_RebelTempo_FaceAcc_SkyDive_Physics
    simulationSpace=BaseBoneSpace
    baseBoneRef {boneName=head}
    simSpaceSettings {
     dampingAlpha=0.2
     maxAngularAcceleration=1000
     maxAngularVelocity=4
     maxLinearAcceleration=750
     maxLinearVelocity=600
     velocityScaleZ=0.6
     worldAlpha=0.7
     externalLinearDragV=<3 3 3>
    }
   }
   n2 = RigidBody(
    ComponentPose=Local To Component(LocalPose=Use cached pose 'Initia Rig Cache'())
    OverrideWorldGravity="0,0,0"
    SimSpaceSettings=(WorldAlpha=0.7,VelocityScaleZ=0.6,DampingAlpha=0.2,MaxLinearVelocity=600,MaxAngularVelocity=4,MaxLinearAcceleration=750,MaxAngularAcceleration=1000,ExternalLinearDragV=(X=1,Y=1,Z=1),ExternalLinearVelocity=(X=0,Y=0,Z=0),ExternalAngularVelocity=(X=0,Y=0,Z=0))
   ) {
    bOverrideWorldGravity=true
    overridePhysicsAsset=&/M_MED_RebelTempo_FaceAcc_Physics
    simulationSpace=BaseBoneSpace
    baseBoneRef {boneName=head}
    simSpaceSettings {
     dampingAlpha=0.2
     maxAngularAcceleration=1000
     maxAngularVelocity=4
     maxLinearAcceleration=750
     maxLinearVelocity=600
     velocityScaleZ=0.6
     worldAlpha=0.7
     externalLinearDragV=<1 1 1>
    }
   }
   n3 = Blend Poses by bool(
    BlendPose_0=Component To Local(ComponentPose=@n1)
    BlendPose_1=Component To Local(ComponentPose=@n2)
    BlendTime_0=0.5
    BlendTime_1=0.5
    bActiveValue=Get bIsSkyDiving()
   )
   n4 = Anim Dynamics - Bone: dyn_bang_01(
    ComponentPose="Transform (Modify) Bone - Bone: base_bang_02"(ComponentPose=Local To Component(LocalPose=@n3), Alpha=1, Translation="0,0,0", Rotation="0,0,-8", Scale="1,1,1") {rotationMode=BMM_Additive, boneToModify {boneName=base_bang_02}, rotation {roll=-8}}
    Alpha=1
    ExternalForce="0,0,0"
   ) {
    angularDampingOverride=1
    angularSpringConstant=20000
    bAngularSpring=true
    bOverrideAngularDamping=true
    bUseGravityOverride=true
    gravityScale=0.35
    simulationSpace=RootRelative
    boundBone {boneName=dyn_bang_01}
    componentLinearAccScale=<1 1 1>
    componentLinearVelScale=<1 1 1>
    gravityOverride {z=-980}
    physicsBodyDefinitions [{
      boundBone {boneName=dyn_bang_01}
      boxExtents=<20 30 30>
      constraintSetup {angularLimitsMax=<0 10 20>, angularLimitsMin=<0 -10 0>, angularTarget=<1 0 0>}
      localJointOffset=<4 0 0>
     }]
   }
   n5 = Anim Dynamics - Bone: dyn_bang_02(ComponentPose=@n4, Alpha=1, ExternalForce="0,0,0") {
    angularDampingOverride=1
    angularSpringConstant=20000
    bAngularSpring=true
    bOverrideAngularDamping=true
    bUseGravityOverride=true
    gravityScale=0.35
    simulationSpace=RootRelative
    boundBone {boneName=dyn_bang_02}
    componentLinearAccScale=<1 1 1>
    componentLinearVelScale=<1 1 1>
    gravityOverride {z=-980}
    physicsBodyDefinitions [{
      boundBone {boneName=dyn_bang_02}
      boxExtents=<20 30 30>
      constraintSetup {angularLimitsMax=<0 10 25>, angularLimitsMin=<0 -10 10>, angularTarget=<1 0 0>}
      localJointOffset=<4 0 0>
     }]
   }
   n6 = Anim Dynamics - Bone: dyn_bang_03(ComponentPose=@n5, Alpha=1, ExternalForce="0,0,0") {
    angularDampingOverride=1
    angularSpringConstant=20000
    bAngularSpring=true
    bOverrideAngularDamping=true
    bUseGravityOverride=true
    gravityScale=0.35
    simulationSpace=RootRelative
    boundBone {boneName=dyn_bang_03}
    componentLinearAccScale=<1 1 1>
    componentLinearVelScale=<1 1 1>
    gravityOverride {z=-980}
    physicsBodyDefinitions [{
      boundBone {boneName=dyn_bang_03}
      boxExtents=<20 30 30>
      constraintSetup {angularLimitsMax=<0 10 20>, angularLimitsMin=<0 -10 10>, angularTarget=<1 0 0>}
      localJointOffset=<4 0 0>
     }]
   }
   n7 = Anim Dynamics - Bone: dyn_bang_04(ComponentPose=@n6, Alpha=1, ExternalForce="0,0,0") {
    angularDampingOverride=1
    angularSpringConstant=20000
    bAngularSpring=true
    bOverrideAngularDamping=true
    bUseGravityOverride=true
    gravityScale=0.35
    simulationSpace=RootRelative
    boundBone {boneName=dyn_bang_04}
    componentLinearAccScale=<1 1 1>
    componentLinearVelScale=<1 1 1>
    gravityOverride {z=-980}
    physicsBodyDefinitions [{
      boundBone {boneName=dyn_bang_04}
      boxExtents=<20 30 30>
      constraintSetup {angularLimitsMax=<0 10 20>, angularLimitsMin=<0 -10 10>, angularTarget=<1 0 0>}
      localJointOffset=<4 0 0>
     }]
   }
   Output Pose(Result=Component To Local(ComponentPose=@n7)) {layerGroup=None}
   n8 = Control Rig!(
    Source=Use cached pose 'Start Cache'()
    Alpha
    head_up:double=Get Head Up() -> (Head Up:float)
    isSky_Floating:double=Get Skydive Floating() -> (Skydive Floating:double)
    isParachuteOpened:double=Get isParachuteOpened() -> (isParachuteOpened:double)
    isCrouching:double=Get crouchingSprin() -> (crouchingSprin:double)
    isCelebrating:double=Get isCelebrating() -> (isCelebrating:double)
   ) -> (Pose) {controlRigAssetReference {blueprintRigClass=&/M_MED_RebelTempo_FaceAcc_CtrlRig.M_MED_RebelTempo_FaceAcc_CtrlRig_C}}
   Save cached pose 'Initia Rig Cache'(Pose=@n8) {cacheName=Initia Rig Cache}
   n9 = "Get Body Mesh "() -> ("Body Mesh ":Skeletal Mesh Component Object Reference)
   Save cached pose 'Start Cache'(Pose=Layered blend per bone(
     BasePose=Copy Pose From Mesh() {bUseAttachedParent=true}
     BlendPoses_0=Copy Pose From Mesh(SourceMeshComponent=@n9) {bUseAttachedParent=true}
    )) {cacheName=Start Cache}
  ]}
 EventGraph {nodes [
   Event BlueprintUpdateAnimation()
   > n1 = Sequence!(execute) -> (then_0, then_1, then_2:Exec, then_3:Exec, then_4:Exec, then_5:Exec)
   n2 = GetAnimInstance(self=GetOwnerSkeletalMeshComponent())
   Is Valid(exec=@n1.then_5, InputObject=@n2)
   > n3 = Set is Celebrating(is Celebrating:bool=Montage_IsPlaying(self=@n2, Montage=/Game/Animation/Game/MainPlayer/Emotes/Emote_Celebration_M)) -> (Output_Get:bool)
   > n4 = Branch(Condition=@n3)
   > Set isCelebrating(isCelebrating:double=1) -> (Output_Get:double)
   Set isCelebrating(execute=@n4.else, isCelebrating=0)
   n5 = Get bIsParachuteOpened()
   n6 = Get bIsSkydiveFloating()
   n7 = Get bIsSkyDiving()
   n8 = Branch(execute=@n1.then_1, Condition=AND Boolean(A=OR Boolean(A=@n6, B=@n7), B=NOT Boolean(A=@n5)))
   > Set Skydive Floating(Skydive Floating:double=1) -> (Output_Get:double)
   Set Skydive Floating(execute=@n8.else, Skydive Floating=0)
   n9 = Branch(execute=@n1.then_2, Condition=Get bIsParachuteOpened())
   > Set isParachuteOpened(isParachuteOpened:double=1) -> (Output_Get:double)
   Set isParachuteOpened(execute=@n9.else, isParachuteOpened=0)
   n10 = Branch(execute=@n1.then_3, Condition=Get bIsPlayingMeleeAnim())
   > Set alpha hairRigidBody(alpha hairRigidBody:double=0.75) -> (Output_Get:double)
   n11 = Branch(execute=@n10.else, Condition=Get bIsSkydiveFloating())
   > Set alpha hairRigidBody(alpha hairRigidBody=0.65)
   n12 = Branch(execute=@n11.else, Condition=Get bIsParachuteOpened())
   > Set alpha hairRigidBody(alpha hairRigidBody=0.5)
   Set alpha hairRigidBody(execute=@n12.else, alpha hairRigidBody=1)
   n13 = Branch(execute=@n1.then_4, Condition=OR Boolean(A=Get bIsCrouching(), B=Get bIsCrouchSprinting()))
   > Set crouchingSprin(crouchingSprin:double=1) -> (Output_Get:double)
   Set crouchingSprin(execute=@n13.else, crouchingSprin=0)
   n14 = "Gravity Override State Function (Parameter Asset) (Nativized)"(
    execute=@n1.then_0
    CharacterPartAnimInstance=Self-Reference()
    GameTimeInSeconds=GetGameTimeInSeconds()
    DebugJointName=spine_05
    Parameters=&/F_MED_RebelTempo_Parameters
   )
   > Set Gravity Override(Gravity Override=@n14)
   > Set RBAN Settings Out(RBAN Settings Out=@n14.OutRigidBodyAnimNodeParameters)
   > Set Head Up(Head Up:float=GetAlphaFromReferenceTransformDelta(SkeletalMeshComponent=Get SkelMesh(), BoneName=spine_04, BaseName=head, InputRange=(X=15,Y=60), TwistAxis="0,0,1")) -> (Output_Get:float)
   Event OnCharacterCustomizationLoaded()
   > n15 = Cast To PlayerPawn_Athena(Object=TryGetPawnOwner())
   > "Set Face Mesh "("Face Mesh ":Skeletal Mesh Component Object Reference=GetSkeletalMeshForPartType(self=@n15, PartType=Head)) -> (Output_Get:Skeletal Mesh Component Object Reference)
   > "Set Body Mesh "("Body Mesh ":Skeletal Mesh Component Object Reference=GetSkeletalMeshForPartType(self=@n15)) -> (Output_Get:Skeletal Mesh Component Object Reference)
   Event BlueprintInitializeAnimation()
   > Set SkelMesh(SkelMesh=GetOwnerSkeletalMeshComponent())
  ]}
]
variables [
 alpha hairRigidBody {type=number}
 "Body Mesh " {type=SkeletalMeshComponent}
 crouchingSprin {type=number}
 "Face Mesh " {type=SkeletalMeshComponent}
 Gravity Override {type=Vector}
 Head Up {type=number}
 is Celebrating {type=boolean}
 isCelebrating {type=number}
 isParachuteOpened {type=number}
 RBAN Settings Out {type=FortRigidBodyAnimNodeParameters}
 SkelMesh {type=SkeletalMeshComponent}
 Skydive Floating {type=number}
]
